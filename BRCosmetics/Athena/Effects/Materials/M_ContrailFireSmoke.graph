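# Translucent unlit contrail material for Niagara ribbons and beam trails, rendering fiery smoke.
# A panned smoke normal map's blue channel blends hot and smoke color pairs, with dynamic parameters driving pan speed, hot-to-smoke transition and erosion.
# Opacity combines eroded smoke, particle alpha and a U-axis mask broken up by two scrolling fire textures.
# Distance fade is skipped on Low quality, and mobile scales emissive separately.
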Material /BRCosmetics/Athena/Effects/Materials/M_ContrailFireSmoke
graph {
 properties {
  bUsedWithBeamTrails=true
  bUsedWithNiagaraRibbons=true
  blendMode=BLEND_Translucent
  pixelDepthOffsetMode=PDOM_Legacy
  shadingModel=MSM_Unlit
  usedShadingModels=Unlit
 }
 nodes [
  ConstantBiasScale(None=Sine(None=Time()))
  n1 = VectorParameter() {parameterName=Hot Color B, defaultValue=<rgba 100 5.9748 1.9948 1>}
  n2 = VectorParameter() {parameterName=Hot Color A, defaultValue=<rgba 4 0.3524 0.0208 0>}
  n3 = Panner(
   Coordinate=TextureCoordinate() {coordinateIndex=1}
   Speed=AppendVector(A=DynamicParameter() {paramNames [PannerSpeedU,Transition,Erode,Param4]}.PannerSpeedU, B=Constant())
  )
  n4 = ConstantBiasScale(None=TextureSample(UVs=@n3) {samplerType=SAMPLERTYPE_Normal, texture=/Game/Effects/Fort_Effects/Gameplay/SmokeDustExplosions/Textures/T_RibbonSmokeMasked_01}.B)
  n5 = OneMinus(None=@n4)
  n6 = VectorParameter() {parameterName=Smoke Color B, defaultValue=<rgba 0.05 0.0464 0.0461 0>}
  n7 = VectorParameter() {parameterName=Smoke Color A, defaultValue=<rgba 0.01 0.0094 0.0093 0>}
  n8 = Add(A=DynamicParameter() {paramNames [PannerSpeedU,Transition,Erode,Param4]}.Transition, B=@n5)
  n9 = ParticleColor()
  n10 = Multiply(
   A=LinearInterpolate(A=LinearInterpolate(A=@n1.RGB, B=@n2.RGB, Alpha=@n5), B=LinearInterpolate(A=@n6.RGB, B=@n7.RGB, Alpha=@n5), Alpha=Saturate(None=@n8))
   B=@n9.RGB
  )
  n11 = ScalarParameter() {defaultValue=0.8, parameterName=MobileEmissiveScalar}
  MaterialFunctionCall(Emissive=ShadingPathSwitch(Default=@n10, Mobile=Multiply(A=@n10, B=@n11)) {desc=Migrated from FeatureLevelSwitch}) {materialFunction=&MF_FX_EmissiveAdjustments}
  n12 = TextureSample(UVs=@n3) {texture=/Game/Effects/Fort_Effects/Gameplay/SmokeDustExplosions/Textures/T_RibbonSmokeMasked_02}
  n13 = Multiply(A=@n4, B=DynamicParameter() {paramNames [PannerSpeedU,Transition,Erode,Param4]}.Erode)
  n14 = OneMinus(None=ComponentMask(None=TextureCoordinate()) {r=true})
  n15 = ComponentMask(None=TextureCoordinate()) {r=true}
  n16 = Clamp(None=LinearInterpolate(Alpha=OneMinus(None=@n15)) {constB=8})
  n17 = Multiply(A=OneMinus(None=LinearInterpolate(Alpha=@n14) {constA=-3}), B=@n16)
  n18 = TextureSample!(UVs=Panner() {speedY=0.5}, Tex, Level, Apply View MipBias) {constMipValue=2, mipValueMode=TMVM_MipLevel, texture=/Game/Effects/Fort_Effects/Textures/Fire/T_FireTile}
  n19 = TextureSample(UVs=Panner() {speedY=-0.125}) {constMipValue=2, mipValueMode=TMVM_MipLevel, texture=&T_FireTile}
  n20 = LinearInterpolate(A=Multiply(A=@n17, B=Add(A=@n18.A, B=@n19.A)), B=@n17, Alpha=@n16)
  n21 = Multiply(
   A=Multiply(A=@n9.A, B=Subtract(A=@n12.RGB, B=@n13))
   B=LinearInterpolate(A=Constant3Vector() {constant=<rgba 1 1 1 0>}, B=Clamp(None=@n20), Alpha=ScalarParameter() {defaultValue=1, parameterName=Clamping_Mask})
  )
  QualitySwitch(
   Default=MaterialFunctionCall(
    Opacity In=@n21
    Position=WorldPosition().XYZ
    Fade_Start_Distance=ScalarParameter() {defaultValue=100, parameterName=Start Fade Distance}
    Fade_End_Distance=ScalarParameter() {defaultValue=20, parameterName=End Fade Distance}
   ) {materialFunction=&MF_Distance_SyncedFade_FX}.Result
   Low=@n21
  )
 ]
}
