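# Translucent unlit Niagara ribbon material for the BattleSuit contrail, drawing a smoky trail with a hot band and a Mirror switch.
# Panned smoke noise, skewed across the width and offset by a dynamic parameter, combines with edge masks into an alpha.
# That alpha indexes a curve-atlas gradient for HDR emissive and feeds layered smoothstep bands with near-camera fade.
# A second panned noise, scaled by U and particle red, builds a sideways displacement vector.
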Material /BRCosmetics/Athena/Effects/Materials/M_Contrail_BattleSuit
graph {
 properties {
  bAutomaticallySetUsageInEditor=false
  bUsedWithNiagaraRibbons=true
  bUsedWithStaticMesh=false
  blendMode=BLEND_Translucent
  pixelDepthOffsetMode=PDOM_Legacy
  shadingModel=MSM_Unlit
  usedShadingModels=Unlit
 }
 nodes [
  Divide(A=ComponentMask(None=WorldPosition().XYZ) {b=true}, B=Constant() {r=3000})
  n1 = Constant2Vector() {g=-0.25}
  n2 = Constant2Vector() {g=0.25}
  n3 = ComponentMask(None=Add(A=TextureCoordinate(), B=StaticSwitchParameter(True=@n1, False=@n2) {defaultValue=true, parameterName=Mirror})) {g=true}
  n4 = Multiply(A=OneMinus(None=@n3), B=@n3)
  n5 = ScalarParameter() {defaultValue=2, group=Opacity, parameterName=Nosie EdgeFade Strength}
  n6 = Multiply(A=Power(Base=Multiply(A=@n4, B=@n5) {constB=2}, Exp=Constant() {r=10})) {constB=200}
  n7 = TextureCoordinate() {coordinateIndex=1}
  n8 = ComponentMask(None=@n7) {r=true}
  n9 = ScalarParameter() {defaultValue=1, parameterName=Mask Exponent}
  n10 = TextureCoordinate()
  n11 = StaticSwitchParameter(True=Constant() {r=0.5}, False=Constant()) {defaultValue=true, parameterName=Mirror}
  n12 = Add(
   A=ComponentMask(None=MaterialFunctionCall(UVs=@n10, "Rotation Angle (0-1)"=@n11) {materialFunction=&CustomRotator}.Rotated Values) {g=true}
   B=ScalarParameter() {parameterName=CenterOffset}
  )
  n13 = ConstantBiasScale(None=@n12) {bias=-0.5, scale=2}
  n14 = Multiply(A=@n13) {constB=-1}
  n15 = LinearInterpolate(
   A=ScalarParameter() {defaultValue=0.5, parameterName=UV Skew U}
   B=ScalarParameter() {defaultValue=0.525, parameterName=Skew}
   Alpha=Sine(None=Power(Base=@n14, Exp=@n9)) {period=2.5}
  )
  n16 = ScalarParameter() {defaultValue=4, group=Opacity, parameterName=Tile NoiseU}
  n17 = DynamicParameter() {paramNames [NoiseUOffset,Param2,Param3,Param4]}
  n18 = AppendVector(
   A=Multiply(A=Add(A=@n8, B=@n15), B=Add(A=@n16, B=@n17.NoiseUOffset))
   B=Multiply(A=ComponentMask(None=@n7) {g=true}, B=ScalarParameter() {defaultValue=6, group=Opacity, parameterName=Tile NoiseV})
  )
  n19 = ScalarParameter() {defaultValue=1, group=Opacity, parameterName=Noise PanV Speed}
  n20 = AppendVector(
   A=ScalarParameter() {group=Opacity, parameterName=Noise PanU Speed}
   B=StaticSwitchParameter(True=Multiply(A=@n19) {constB=-1}, False=@n19) {defaultValue=true, parameterName=Mirror}
  )
  n21 = Saturate(None=Sign(None=@n13))
  n22 = Subtract(
   A=TextureSampleParameter2D(UVs=Panner(Coordinate=@n18, Speed=@n20) {speedX=0.25}) {group=Noise, parameterName=Noise Texture, texture=&T_Inky_Smoke_Tile}.RGB
   B=@n21
  )
  n23 = OneMinus(None=@n21)
  n24 = Multiply(A=Multiply(A=@n12) {constB=2}, B=@n23)
  n25 = Add(A=@n24, B=ScalarParameter() {defaultValue=-0.9, parameterName=HotBand_Width, sliderMax=0, sliderMin=-1}) {constB=-0.9}
  n26 = Add(A=Multiply(A=Saturate(None=@n6), B=Saturate(None=@n22)), B=Saturate(None=Multiply(A=@n25) {constB=50}))
  n27 = LinearInterpolate(
   A=ScalarParameter() {defaultValue=1, group=Opacity, parameterName=Contrast_Low}
   B=ScalarParameter() {defaultValue=3, group=Opacity, parameterName=Contrast_High}
   Alpha=Multiply(A=OneMinus(None=@n24), B=@n23)
  ) {constA=0.6, constB=4}
  n28 = Saturate(None=Multiply(A=@n14, B=ScalarParameter() {defaultValue=10, parameterName=Mask Contrast}))
  n29 = Multiply(A=ParticleColor().A, B=Multiply(A=Power(Base=@n26, Exp=@n27), B=@n28))
  n30 = ComponentMask(None=@n29) {g=true}
  n31 = CurveAtlasRowParameter(CurveTime=Power(Base=OneMinus(None=@n30), Exp=ScalarParameter() {defaultValue=0.45, group=Color, parameterName=Curve Contrast Controller, sortPriority=1})) {
   atlas=&AthenaCurveAtlas01
   curve=/Game/Characters/Player/Male/Medium/Bodies/M_MED_BattleSuit/FX/BattleSuit_Blue
   defaultValue=82
   group=Color
   parameterName=Curve Parameter
  }
  n32 = VectorParameter() {group=Color, parameterName=HDR_Tint, sortPriority=2, defaultValue=<rgba 1 1 1 0>}
  MaterialFunctionCall(Emissive=MaterialFunctionCall(
    LDR Input=Clamp(None=@n31)
    HDR Tint=Multiply(A=@n31, B=@n32.RGB)
    HDR Intensity Multiplier=ScalarParameter() {defaultValue=1.5, group=Color, parameterName=HDR_Intensity, sortPriority=3}
   ) {materialFunction=&DeriveHDRfromLDR}.RemappedRange) {materialFunction=&MF_FX_EmissiveAdjustments}
  MaterialFunctionCall!(MaxiumURange=Constant() {r=1000}, RangeRampHardness, MirrorTexCoordU) {materialFunction=/BRCosmetics/Weapons/FORT_Melee/Pickaxe_Venus/MF_LocalSpaceTrailUV}
  Multiply()
  n33 = ScalarParameter() {parameterName=Smoothstep Min Out}
  n34 = ScalarParameter() {defaultValue=0.2, parameterName=Smoothstep Smoothness Out}
  n35 = ScalarParameter() {defaultValue=0.2, parameterName=Smoothstep Min Mid}
  n36 = MaterialFunctionCall(Alpha=@n29, Min=@n35, Max=Add(A=@n35, B=ScalarParameter() {defaultValue=0.2, parameterName=Smoothstep Smoothness Mid})) {materialFunction=&SmoothStep}
  n37 = ScalarParameter() {defaultValue=0.5, parameterName=Smoothstep Min Inner}
  n38 = MaterialFunctionCall(Alpha=@n29, Min=@n37, Max=Add(A=@n37, B=ScalarParameter() {defaultValue=0.777, parameterName=Smoothstep Smoothness Inner})) {materialFunction=&SmoothStep}
  n39 = Subtract(
   A=MaterialFunctionCall(Alpha=@n29, Min=@n33, Max=Add(A=@n33, B=@n34)) {materialFunction=&SmoothStep}.Result
   B=Add(A=@n36.Result, B=@n38.Result)
  )
  n40 = Multiply(A=@n38.Result) {constB=10}
  n41 = Add(A=Add(A=@n29, B=Saturate(None=@n39)), B=Saturate(None=Subtract(A=@n36.Result, B=@n40)))
  Multiply(A=@n28, B=@n41)
  Multiply(
   A=@n41
   B=MaterialFunctionCall(Distance=ScalarParameter() {defaultValue=150, parameterName=CameraFadeLength}) {materialFunction=&Near_Camera_Fade}.Result
  )
  n42 = TextureCoordinate()
  n43 = ComponentMask(None=@n42) {r=true}
  n44 = Multiply(A=@n43, B=@n43)
  n45 = Multiply(A=@n42) {constB=0.5}
  n46 = Constant2Vector() {g=-0.3186, r=0.423}
  n47 = ConstantBiasScale(None=TextureSample(UVs=Panner(Coordinate=@n45, Speed=@n46)) {texture=&T_Inky_Smoke_Tile}.G) {bias=-0.5, scale=2}
  n48 = Transform(None=StaticSwitchParameter(True=Constant3Vector() {constant=<rgba -1 0 0 0>}, False=Constant3Vector() {constant=<rgba 1 0 0 0>}) {defaultValue=true, parameterName=Mirror})
  Multiply(A=Saturate(None=Multiply(A=@n44) {constB=5}), B=Multiply(A=Multiply(A=@n47, B=@n48), B=ParticleColor().R))
  Saturate(None=@n24)
 ]
}
