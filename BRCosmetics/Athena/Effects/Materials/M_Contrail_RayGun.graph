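# Translucent unlit ray-gun contrail material for beam trails and Niagara ribbons.
# Panning beam textures, some gated by thresholded noise, form a streak mask that combines with a distortion texture and sine band to drive a brightness lerp.
# Two colors blend along U, then particle color and dynamic HDR intensity drive emissive.
# Opacity sharpens the alpha-weighted mask, erodes both ends via dynamic parameters, and applies near-camera fade except on mobile.
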Material /BRCosmetics/Athena/Effects/Materials/M_Contrail_RayGun
graph {
 properties {
  bUsedWithBeamTrails=true
  bUsedWithNiagaraRibbons=true
  blendMode=BLEND_Translucent
  pixelDepthOffsetMode=PDOM_Legacy
  shadingModel=MSM_Unlit
  usedShadingModels=Unlit
 }
 nodes [
  n1 = StaticSwitchParameter(True=TextureCoordinate(), False=TextureCoordinate() {coordinateIndex=1}) {parameterName=Preview UVs}
  n2 = TextureSample(UVs=Panner(Coordinate=MaterialFunctionCall(XY=@n1) {materialFunction=&Swizzle}.YX) {speedY=-0.25}) {texture=/Game/Global/Textures/Default/Masks_and_Gradients/T_BlasterBeamDistort}
  n3 = TextureCoordinate()
  n4 = Power(Base=Sine(None=ComponentMask(None=@n3) {g=true}) {period=2}, Exp=ScalarParameter() {defaultValue=20, parameterName=Contrast})
  n5 = TextureCoordinate()
  n6 = TextureCoordinate() {coordinateIndex=1}
  n7 = TextureSample(UVs=Panner(Coordinate=StaticSwitchParameter(True=@n5, False=@n6) {parameterName=Preview UVs}) {bFractionalPart=true, speedX=-8, speedY=4}) {texture=&T_NoisePack_01}
  n8 = StaticSwitchParameter(True=TextureCoordinate(), False=TextureCoordinate() {coordinateIndex=1}) {parameterName=Preview UVs}
  n9 = AppendVector(
   A=ScalarParameter() {defaultValue=0.8, group=Phase 1, parameterName=Bits A U Repeat}
   B=ScalarParameter() {defaultValue=1, group=Phase 1, parameterName=Bits A V Repeat}
  )
  n10 = Multiply(A=Time(), B=ScalarParameter() {defaultValue=1, group=Phase 1, parameterName=Pan Speed Master A})
  n11 = TextureSampleParameter2D(UVs=Panner(Coordinate=Multiply(A=@n8, B=@n9), Time=@n10) {bFractionalPart=true, speedX=-13}) {
   group=Phase 1
   parameterName=BeamTexturePhase1
   samplerType=SAMPLERTYPE_LinearColor
   texture=/Game/Effects/Fort_Effects/Textures/Tile/T_Raygun_Beam_01
  }
  n12 = StaticSwitchParameter(True=TextureCoordinate(), False=TextureCoordinate() {coordinateIndex=1}) {parameterName=Preview UVs}
  n13 = AppendVector(
   A=ScalarParameter() {defaultValue=2, group=Phase 1, parameterName=Bits B U Repeat}
   B=ScalarParameter() {defaultValue=1, group=Phase 1, parameterName=Bits B V Repeat}
  )
  n14 = TextureSampleParameter2D(UVs=Panner(Coordinate=Multiply(A=@n12, B=@n13), Time=@n10) {bFractionalPart=true, speedX=-19}) {
   group=Phase 1
   parameterName=BeamTexturePhase1_1
   samplerType=SAMPLERTYPE_LinearColor
   texture=&T_Raygun_Beam_01
  }
  n15 = Multiply(A=Ceil(None=Add(A=@n7.B) {constB=-0.4}), B=Add(A=@n11.B, B=@n14.B))
  n16 = TextureCoordinate()
  n17 = TextureCoordinate() {coordinateIndex=1}
  n18 = TextureSampleParameter2D(UVs=Panner(Coordinate=StaticSwitchParameter(True=@n16, False=@n17) {parameterName=Preview UVs}) {bFractionalPart=true, speedX=-4}) {
   group=Phase 1
   parameterName=BeamTexturePhase1_2
   samplerType=SAMPLERTYPE_LinearColor
   texture=&T_Raygun_Beam_01
  }
  n19 = Saturate(None=ConstantBiasScale(None=@n18.R) {bias=-0.5, scale=2})
  n20 = StaticSwitchParameter(True=TextureCoordinate(), False=TextureCoordinate() {coordinateIndex=1}) {parameterName=Preview UVs}
  n21 = AppendVector(
   A=ScalarParameter() {defaultValue=0.25, group=Phase 1, parameterName=Bits Core U Repeat}
   B=ScalarParameter() {defaultValue=1, group=Phase 1, parameterName=Bits Core V Repeat}
  )
  n22 = TextureSampleParameter2D(UVs=Panner(Coordinate=Multiply(A=@n20, B=@n21)) {bFractionalPart=true, speedX=-4}) {
   group=Phase 1
   parameterName=BeamTexturePhase1_3
   samplerType=SAMPLERTYPE_LinearColor
   texture=&T_Raygun_Beam_01
  }
  n23 = Clamp(None=Add(A=Add(A=@n15, B=@n19), B=Add(A=@n22.R, B=@n22.R)) {constB=0})
  n24 = LinearInterpolate(
   A=VectorParameter() {parameterName=LerpA, defaultValue=<rgba 0.05 0.05 0.05 1>}.RGB
   B=VectorParameter() {parameterName=LerpB, defaultValue=<rgba 1 1 1 1>}.RGB
   Alpha=Multiply(A=Add(A=@n2.R, B=@n4), B=@n23)
  )
  n25 = VectorParameter() {parameterName=Color_B, defaultValue=<rgba 0.9748 1 0 1>}
  n26 = VectorParameter() {parameterName=Color_A, defaultValue=<rgba 0.1401 1 0 1>}
  n27 = OneMinus(None=ComponentMask(None=TextureCoordinate()) {r=true})
  n28 = ScalarParameter() {defaultValue=0.2, parameterName=VerticalContrast}
  n29 = ParticleColor()
  n30 = Multiply(A=LinearInterpolate(A=Multiply(A=@n24, B=@n25.RGB), B=Multiply(A=@n24, B=@n26.RGB), Alpha=Power(Base=@n27, Exp=@n28)), B=@n29.RGB)
  MaterialFunctionCall(Emissive=MaterialFunctionCall(
    LDR Input=Clamp(None=@n30)
    HDR Tint=@n29.RGB
    HDR Intensity Multiplier=DynamicParameter() {paramNames [Graphic Line Erode Exp, HDR_Intensity, Start Beam Erode, Param4]}.HDR_Intensity
   ) {materialFunction=&DeriveHDRfromLDR}.RemappedRange) {materialFunction=&MF_FX_EmissiveAdjustments}
  n31 = Multiply(A=@n29.A, B=@n23)
  n32 = ScalarParameter() {defaultValue=4, group=Erode, parameterName=Lerp Exponent}
  n33 = Saturate(None=Multiply(A=Power(Base=@n31, Exp=@n32) {constExponent=10}, B=ScalarParameter() {defaultValue=1000, group=Erode, parameterName=Lerp Brightness}))
  n34 = TextureCoordinate()
  n35 = DynamicParameter() {paramNames [Graphic Line Erode Exp, HDR_Intensity, Start Beam Erode, Param4]}
  n36 = Power(Base=OneMinus(None=ComponentMask(None=@n34) {r=true}), Exp=@n35.Graphic Line Erode Exp)
  n37 = ComponentMask(None=TextureCoordinate()) {r=true}
  n38 = MaterialFunctionCall(Distance=ScalarParameter() {defaultValue=128, parameterName=Fade Distance}) {materialFunction=&Near_Camera_Fade}
  Multiply(
   A=Multiply(A=Multiply(A=@n33, B=@n36), B=Power(Base=@n37, Exp=@n35.Start Beam Erode))
   B=ShadingPathSwitch(Default=@n38.Result, Deferred=@n38.Result, Mobile=Constant() {r=1}) {desc=Migrated from FeatureLevelSwitch}
  )
 ]
}
