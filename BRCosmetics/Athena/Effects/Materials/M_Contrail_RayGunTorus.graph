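# Additive unlit mesh-particle material, named for the torus in the Ray Gun contrail cosmetic, with Fresnel-based shading and noise erosion.
# An inverted-Fresnel color blend, multiplied by particle color, passes through HDR derivation and emissive adjustment functions.
# Particle alpha and inverted Fresnel multiply a hard mask from two summed, floored panning noise channels, blended in by an Erode dynamic parameter.
# An optional static switch offsets vertices along normals for shrinking.
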Material /BRCosmetics/Athena/Effects/Materials/M_Contrail_RayGunTorus
graph {
 properties {
  bUsedWithMeshParticles=true
  bUsedWithNiagaraMeshParticles=true
  blendMode=BLEND_Additive
  pixelDepthOffsetMode=PDOM_Legacy
  shadingModel=MSM_Unlit
  usedShadingModels=Unlit
 }
 nodes [
  n1 = Fresnel(ExponentIn=ScalarParameter() {defaultValue=0.25, parameterName=Fres_Exp_1})
  n2 = ParticleColor()
  n3 = Multiply(
   A=LinearInterpolate(
    A=VectorParameter() {parameterName=LerpA, defaultValue=<rgba 0.1 0.1 0.1 1>}.RGB
    B=VectorParameter() {parameterName=LerpB, defaultValue=<rgba 1 1 1 1>}.RGB
    Alpha=OneMinus(None=@n1)
   )
   B=@n2.RGB
  )
  MaterialFunctionCall(Emissive=MaterialFunctionCall(LDR Input=Clamp(None=@n3), HDR Tint=@n2.RGB) {materialFunction=&DeriveHDRfromLDR}.RemappedRange) {materialFunction=&MF_FX_EmissiveAdjustments}
  n4 = Fresnel(ExponentIn=ScalarParameter() {defaultValue=0.5, parameterName=Fres_Exp})
  n5 = ScalarParameter() {defaultValue=1.6, group=Erode, parameterName=Erode U Repeat A}
  n6 = ScalarParameter() {defaultValue=0.4, group=Erode, parameterName=Erode V Repeat A}
  n7 = StaticSwitchParameter(True=TextureCoordinate(), False=TextureCoordinate() {coordinateIndex=3}) {defaultValue=true, parameterName=Preview UVs}
  n8 = Constant()
  n9 = Panner(
   Coordinate=Multiply(A=AppendVector(A=@n5, B=@n6), B=@n7)
   Speed=AppendVector(A=@n8, B=ScalarParameter() {defaultValue=4.8, parameterName=PanSpeed1})
  ) {bFractionalPart=true, speedX=-0.4, speedY=4.8}
  n10 = ScalarParameter() {defaultValue=1.1, group=Erode, parameterName=Erode U Repeat B}
  n11 = ScalarParameter() {defaultValue=0.71, group=Erode, parameterName=Erode V Repeat B}
  n12 = Panner(
   Coordinate=Multiply(A=@n7, B=AppendVector(A=@n10, B=@n11))
   Speed=AppendVector(A=@n8, B=ScalarParameter() {defaultValue=2.3, parameterName=PanSpeed2})
  ) {bFractionalPart=true, speedX=0.3, speedY=2.3}
  n13 = Saturate(None=Add(
    A=TextureSampleParameter2D(UVs=@n9) {parameterName=NoiseA, texture=&T_NoisePack_01}.R
    B=TextureSampleParameter2D(UVs=@n12) {parameterName=NoiseB, texture=&T_NoisePack_01}.G
   ))
  Multiply(
   A=Multiply(A=@n2.A, B=OneMinus(None=@n4))
   B=LinearInterpolate(B=Floor(None=@n13), Alpha=DynamicParameter() {paramNames [Erode,Param2,Param3,Param4]}.Erode) {constA=1}
  )
  StaticSwitchParameter(True=Multiply(A=VertexNormalWS(), B=ScalarParameter() {parameterName=WPOShrinkAmnt}), False=Constant()) {parameterName=WPOShrink}
 ]
}
