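# Additive, unlit, two-sided material for a red tail-light contrail on beam trails and Niagara ribbons.
# UV falloffs and two ramp samples at fixed pan offsets build a streak mask, optionally adding a front edge.
# It blends red tints, scaled by particle color and time of day, and drives depth-faded opacity.
# BlendNoise adds panning smoke, but both brightness lerp inputs read LowBirghtness, so the noise cancels out.
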Material /BRCosmetics/Athena/Effects/Materials/M_Contrail_RedTailLightTrail
graph {
 properties {
  bUsedWithBeamTrails=true
  bUsedWithNiagaraRibbons=true
  blendMode=BLEND_Additive
  pixelDepthOffsetMode=PDOM_Legacy
  shadingModel=MSM_Unlit
  twoSided=true
  usedShadingModels=Unlit
 }
 nodes [
  Constant()
  n1 = Constant() {r=1}
  n2 = Constant() {r=0.5}
  n3 = Multiply(
   A=ScalarParameter() {defaultValue=3, parameterName=MaskUtile}
   B=MaterialFunctionCall(Float2=TextureCoordinate() {coordinateIndex=1}) {materialFunction=&BreakOutFloat2Components}.R
  )
  n4 = MaterialFunctionCall(Float2=TextureCoordinate()) {materialFunction=&BreakOutFloat2Components}
  n5 = Multiply(A=Multiply(A=@n3, B=@n3), B=Power(Base=OneMinus(None=@n4.R)) {constExponent=4})
  n6 = Divide(A=OneMinus(None=@n4.G), B=ScalarParameter() {defaultValue=5, parameterName=EmissiveAmount}) {constB=5}
  n7 = Multiply(A=@n6, B=@n6)
  n8 = Add(A=@n7, B=Power(Base=@n4.G) {constExponent=4})
  n9 = Panner(Coordinate=TextureCoordinate() {vTiling=8}, Time=Constant() {r=-14.3}) {speedY=0.5}
  n10 = Multiply(A=Multiply(
    A=@n5
    B=TextureSample(UVs=@n9) {samplerSource=SSM_Clamp_WorldGroupSettings, texture=/Game/Global/Textures/Default/Masks_and_Gradients/T_Water_Top_Ramp}.R
   )) {constB=3}
  n11 = TextureSample(UVs=Panner(Coordinate=TextureCoordinate() {vTiling=7}, Time=Constant() {r=0.5}) {speedY=0.5}) {samplerSource=SSM_Clamp_WorldGroupSettings, texture=&T_Water_Top_Ramp}
  n12 = StaticSwitchParameter(True=Constant(), False=Multiply(A=@n5, B=Divide(A=@n11.R) {constB=15})) {parameterName=BlendNoise}
  n13 = ScalarParameter() {defaultValue=1, parameterName=FrontAddMult}
  n14 = ScalarParameter() {defaultValue=1, parameterName=FrontAddExp}
  n15 = Add(
   A=Add(A=Multiply(A=@n5, B=@n8), B=Add(A=@n10, B=@n12) {constB=0})
   B=StaticSwitchParameter(
    True=MaterialFunctionCall!(UVs, Mult=@n13, Exp=@n14) {materialFunction=/Game/Effects/Fort_Effects/MaterialFunctions/MF_SquareEdgeMask}.RightMask
    False=Constant()
   ) {parameterName=UseFrontAdd}
  )
  n16 = LinearInterpolate(
   A=VectorParameter() {parameterName=LowColor, defaultValue=<rgba 1.1 0 0.007 1>}.RGB
   B=VectorParameter() {parameterName=HighColor, defaultValue=<rgba 1 0 0 1>}.RGB
   Alpha=@n15
  )
  n17 = ParticleColor()
  MaterialFunctionCall(Emissive=Multiply(
    A=MaterialFunctionCall(Color_Morning=@n1, Color_Day=@n1, Color_Evening=@n1, Color_Night=@n2) {materialFunction=&Color_Time_of_Day}.Result
    B=Multiply(A=@n16, B=@n17.RGB)
   )) {materialFunction=&MF_FX_EmissiveAdjustments}
  n18 = DynamicParameter() {paramNames [LowBirghtness, HighBrightness, U_PanSpeed, V_PanSpeed]}
  n19 = MaterialFunctionCall(Float2=TextureCoordinate() {coordinateIndex=1, uTiling=0.7, vTiling=0.4}) {materialFunction=&BreakOutFloat2Components}
  n20 = AppendVector(A=Multiply(A=@n19.R, B=@n19.R), B=@n19.G)
  n21 = AppendVector(A=@n18.U_PanSpeed, B=@n18.U_PanSpeed)
  n22 = LinearInterpolate(
   A=@n18.LowBirghtness
   B=@n18.LowBirghtness
   Alpha=TextureSampleParameter2D(UVs=Panner(Coordinate=@n20, Speed=@n21) {speedX=-3}) {parameterName=Noise, texture=&T_Inky_Smoke_Tile}.R
  ) {constA=0.5, constB=1.7}
  n23 = Multiply(A=@n17.A, B=StaticSwitchParameter(True=Multiply(A=@n15, B=@n22), False=@n15) {parameterName=BlendNoise})
  n24 = ScalarParameter() {defaultValue=40, parameterName=DepthFadeAlpha}
  ShadingPathSwitch(
   Default=StaticSwitchParameter(True=DepthFade(Opacity=@n23, FadeDistance=@n24), False=@n23) {defaultValue=true, parameterName=UseDepthFade}
   Mobile=@n23
  ) {desc=Migrated from FeatureLevelSwitch}
  StaticSwitchParameter(True=Constant(), False=@n7) {parameterName=BlendNoise}
 ]
}
