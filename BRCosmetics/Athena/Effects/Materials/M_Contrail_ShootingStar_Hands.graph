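# Unlit translucent particle sprite material for the hand effect of the Shooting Star contrail cosmetic, drawing a glowing radial energy burst.
# Swirl and smooth noise pan through radial coordinates to form two-color streaks inside sphere masks, with a center glow and panning dust debris added.
# The result gets an HDR boost and optional time-of-day tint.
# Opacity combines mask, streaks and particle alpha, with depth fade skipped on mobile.
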Material /BRCosmetics/Athena/Effects/Materials/M_Contrail_ShootingStar_Hands
graph {
 properties {
  bUsedWithNiagaraSprites=true
  bUsedWithParticleSprites=true
  blendMode=BLEND_Translucent
  pixelDepthOffsetMode=PDOM_Legacy
  shadingModel=MSM_Unlit
  twoSided=true
  usedShadingModels=Unlit
 }
 nodes [
  n1 = MaterialFunctionCall(
   Color_Morning=VectorParameter() {group=TOD, parameterName=MorningColor}.RGB
   Color_Day=VectorParameter() {group=TOD, parameterName=DayColor}.RGB
   Color_Evening=VectorParameter() {group=TOD, parameterName=EveningColor}.RGB
   Color_Night=VectorParameter() {group=TOD, parameterName=NightColor}.RGB
  ) {materialFunction=&Color_Time_of_Day}
  n2 = Multiply(
   A=SphereMask(A=TextureCoordinate(), B=Constant() {r=0.5}, Radius=ScalarParameter() {defaultValue=0.5, parameterName=Radius}) {hardnessPercent=0}
   B=VectorParameter() {parameterName=CenterColor, defaultValue=<rgba 0.4531 0.7178 1 0>}.RGB
  )
  n3 = Power(Base=SphereMask(A=TextureCoordinate(), B=Constant() {r=0.5}, Radius=ScalarParameter() {defaultValue=0.5, parameterName=Radius_2}) {hardnessPercent=0})
  n4 = ScalarParameter() {defaultValue=2, parameterName=U_Half}
  n5 = ScalarParameter() {defaultValue=2, parameterName=V_Half}
  n6 = MaterialFunctionCall(Vector or UVs=Subtract(A=TextureCoordinate() {uTiling=4, vTiling=4}, B=AppendVector(A=@n4, B=@n5))) {materialFunction=&VectorToRadialValue}
  n7 = Multiply(A=Time(), B=ScalarParameter() {defaultValue=0.6, parameterName=PanSpeed})
  n8 = ScalarParameter() {defaultValue=1.5, parameterName=U_Half_1}
  n9 = ScalarParameter() {defaultValue=1.5, parameterName=V_Half_1}
  n10 = MaterialFunctionCall(Vector or UVs=Subtract(A=TextureCoordinate() {uTiling=3, vTiling=3}, B=AppendVector(A=@n8, B=@n9))) {materialFunction=&VectorToRadialValue}
  n11 = LinearInterpolate(
   A=TextureSample(UVs=Panner(Coordinate=@n6.Radial Coordinates, Time=@n7) {speedX=1, speedY=-7}) {samplerType=SAMPLERTYPE_LinearGrayscale, texture=/Game/Global/Textures/Default/Noise/T_SwirlNoise_N_height}.R
   B=TextureSample(UVs=Panner(Coordinate=@n10.Radial Coordinates, Time=@n7) {speedX=-2, speedY=16}) {texture=&T_Smooth_tile}.R
  )
  n12 = ScalarParameter() {parameterName=Streaks_LerpA}
  n13 = ScalarParameter() {defaultValue=1, parameterName=Streaks_LerpB}
  n14 = OneMinus(None=SphereMask(A=TextureCoordinate(), B=Constant() {r=0.5}, Radius=ScalarParameter() {defaultValue=0.5, parameterName=Radius_1}) {hardnessPercent=0})
  n15 = Multiply(A=@n3, B=Power(Base=Multiply(A=@n11) {constB=1.5}, Exp=LinearInterpolate(A=@n12, B=@n13, Alpha=@n14) {constA=0.6, constB=4}))
  n16 = LinearInterpolate(
   A=VectorParameter() {parameterName=DarkColor, defaultValue=<rgba 0.1292 0.1853 0.4 0>}.RGB
   B=VectorParameter() {parameterName=LightColor, defaultValue=<rgba 0.0334 0.548 1.5 0>}.RGB
   Alpha=@n15
  )
  n17 = ScalarParameter() {defaultValue=1.5, parameterName=U_Half_2}
  n18 = ScalarParameter() {defaultValue=1.5, parameterName=V_Half_2}
  n19 = MaterialFunctionCall(Vector or UVs=Subtract(A=TextureCoordinate() {uTiling=3, vTiling=3}, B=AppendVector(A=@n17, B=@n18))) {materialFunction=&VectorToRadialValue}
  n20 = MaterialFunctionCall() {materialFunction=&VectorToRadialValue}
  n21 = Add(
   A=TextureSample(UVs=Panner(Coordinate=@n19.Radial Coordinates, Time=@n7) {speedX=-1, speedY=-2.5}) {texture=/Game/Global/Textures/Default/PointSplatter/T_Tiling_Dust}.RGB
   B=TextureSample(UVs=Panner(Coordinate=@n20.Radial Coordinates, Time=@n7) {speedX=1, speedY=-5}) {texture=&T_Tiling_Dust}.RGB
  )
  n22 = Add(
   A=Add(A=Power(Base=@n2) {constExponent=10}, B=Clamp(None=@n16) {maxDefault=10})
   B=Multiply(A=Multiply(A=@n21, B=@n21), B=VectorParameter() {parameterName=DebrisColor, defaultValue=<rgba 3.2956 3.6779 7 0>}.RGB)
  )
  n23 = MaterialFunctionCall(LDR Input=Clamp(None=@n22), HDR Tint=Multiply(A=@n22, B=ScalarParameter() {defaultValue=2, parameterName=HDRMult})) {materialFunction=&DeriveHDRfromLDR}
  MaterialFunctionCall(Emissive=StaticSwitchParameter(True=Multiply(A=@n1.Result, B=@n23.RemappedRange), False=@n23.RemappedRange) {group=TOD, parameterName=Modulate By Time of Day}) {materialFunction=&MF_FX_EmissiveAdjustments}
  OneMinus(None=ComponentMask(None=TextureCoordinate()) {g=true})
  n24 = Multiply(A=ParticleColor().A, B=Add(A=@n3, B=@n15))
  ShadingPathSwitch(Default=DepthFade(Opacity=@n24, FadeDistance=ScalarParameter() {parameterName=FadeDistance}), Mobile=@n24) {desc=Migrated from FeatureLevelSwitch}
 ]
}
