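# Translucent, unlit, two-sided material for an inky smoke contrail streak on beam trails and Niagara ribbons.
# Emissive blends particle color with a LowColor-tinted, desaturated copy.
# Opacity lighten-blends panning smoke and caustics textures, shaped by soft UV masks.
# Texture U comes from a world-space sphere mask around the offset object position, with particle alpha, near-camera fade and depth fade (skipped on mobile) applied.
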Material /BRCosmetics/Athena/Effects/Materials/M_Contrail_SteakInkySmoke
graph {
 properties {
  bUsedWithBeamTrails=true
  bUsedWithNiagaraRibbons=true
  blendMode=BLEND_Translucent
  pixelDepthOffsetMode=PDOM_Legacy
  shadingModel=MSM_Unlit
  twoSided=true
  usedShadingModels=Unlit
 }
 nodes [
  n1 = VectorParameter() {parameterName=LowColor, defaultValue=<rgba 0.1339 0.5363 0.558 1>}
  n2 = ParticleColor()
  n3 = Desaturation(None=@n2.RGB, Fraction=Constant() {r=0.5})
  MaterialFunctionCall(Emissive=LinearInterpolate(A=Multiply(A=@n1.RGB, B=@n3), B=@n2.RGB)) {materialFunction=&MF_FX_EmissiveAdjustments}
  n4 = MaterialFunctionCall(Float2=TextureCoordinate()) {materialFunction=&BreakOutFloat2Components}
  n5 = Constant() {r=0.5}
  n6 = TextureCoordinate()
  n7 = SphereMask(
   A=Power(Base=MaterialFunctionCall(Float2=@n6) {materialFunction=&BreakOutFloat2Components}.R) {constExponent=0.15}
   B=Constant() {r=0.5}
  ) {attenuationRadius=0.5, hardnessPercent=0}
  n8 = Multiply(
   A=Multiply(A=SphereMask(A=@n4.G, B=@n5) {attenuationRadius=0.5, hardnessPercent=0}, B=Multiply(A=@n7, B=@n7))
   B=SphereMask(A=TextureCoordinate(), B=Constant2Vector() {g=0.5, r=1.025}) {attenuationRadius=1.05, hardnessPercent=90}
  )
  n9 = SphereMask(
   A=WorldPosition().XYZ
   B=Add(A=ObjectPositionWS(), B=VectorParameter() {parameterName=Initial Offset, defaultValue=<rgba 0 0 -875 0>}.RGB)
   Radius=Constant() {r=2550}
   Hardness=Constant()
  )
  n10 = Add(A=Multiply(A=@n9, B=@n9), B=Constant() {r=-200}) {constB=-20}
  n11 = AppendVector(A=Multiply(A=@n10) {constB=7}, B=ComponentMask(None=TextureCoordinate()) {g=true})
  n12 = TextureSample(UVs=Panner(Coordinate=@n11) {speedX=2.2}) {samplerSource=SSM_Wrap_WorldGroupSettings, texture=&T_ShieldCaustics}
  n13 = MaterialFunctionCall(Float2=@n11) {materialFunction=&BreakOutFloat2Components}
  n14 = Add(A=@n13.G) {constB=0.5}
  n15 = TextureSample(UVs=Panner(Coordinate=AppendVector(A=@n13.R, B=@n14)) {speedX=1.3}) {samplerSource=SSM_Wrap_WorldGroupSettings, texture=&T_Inky_Smoke_Tile}
  n16 = Multiply(A=Add(A=LinearInterpolate(Alpha=@n12.R) {constA=0.2, constB=3}, B=LinearInterpolate(Alpha=@n15.R) {constB=3}), B=@n8)
  n17 = ComponentMask(None=TextureCoordinate()) {g=true}
  n18 = Panner(Coordinate=AppendVector(A=Multiply(A=@n10) {constB=7}, B=@n17)) {speedX=2.2}
  n19 = Add(A=@n10) {constB=0.6}
  n20 = Panner(Coordinate=AppendVector(A=Multiply(A=@n19) {constB=9}, B=@n17)) {speedX=2.2}
  n21 = Multiply(A=MaterialFunctionCall!(
    Blend=TextureSample(UVs=@n18) {samplerSource=SSM_Wrap_WorldGroupSettings, texture=&T_Inky_Smoke_Tile}.G
    Base=TextureSample(UVs=@n20) {samplerSource=SSM_Wrap_WorldGroupSettings, texture=&T_Inky_Smoke_Tile}.G
   ) {materialFunction=/Engine/Functions/Engine_MaterialFunctions03/Blends/Blend_Lighten}.Result) {constB=1.25}
  n22 = WorldPosition()
  n23 = CameraPositionWS()
  n24 = Constant() {r=100}
  n25 = Constant() {r=0.2}
  n26 = Multiply(
   A=Multiply(
    A=@n8
    B=MaterialFunctionCall(Blend=@n16, Base=@n21) {materialFunction=&Blend_Lighten}.Result
   )
   B=OneMinus(None=SphereMask(A=@n22.XYZ, B=@n23, Radius=@n24, Hardness=@n25))
  )
  Multiply(A=@n2.A, B=ShadingPathSwitch(Default=DepthFade(Opacity=@n26), Mobile=@n26) {desc=Migrated from FeatureLevelSwitch})
 ]
}
