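# Unlit translucent material for a thin smoke contrail, usable on beam trails, Niagara ribbons and Niagara sprites.
# A puffy smoke texture pans along UV1, is sharpened by a power and boosted, then shaped by a rounded cross-width falloff and a sphere mask along the length.
# Particle alpha, near-camera fade, depth fade and a blendable inverse Fresnel further control opacity.
# Particle color feeds a shared emissive-adjustment function.
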Material /BRCosmetics/Athena/Effects/Materials/M_Contrail_ThinSmoke
graph {
 properties {
  bUsedWithBeamTrails=true
  bUsedWithNiagaraRibbons=true
  bUsedWithNiagaraSprites=true
  blendMode=BLEND_Translucent
  ditherOpacityMask=true
  pixelDepthOffsetMode=PDOM_Legacy
  shadingModel=MSM_Unlit
  translucencyLightingMode=TLM_VolumetricPerVertexDirectional
  usedShadingModels=Unlit
 }
 nodes [
  n1 = ParticleColor()
  n2 = Multiply(
   A=TextureCoordinate() {coordinateIndex=1}
   B=ComponentMask(None=VectorParameter() {parameterName=MainSprayUVMulti, defaultValue=<rgba 0.5 1 0 0>}.RGB) {g=true, r=true}
  )
  n3 = AppendVector(A=ScalarParameter() {parameterName=MovementSpeed}, B=Constant())
  n4 = Power(Base=TextureSample(UVs=Panner(Coordinate=@n2, Speed=@n3)) {texture=/Game/Effects/Fort_Effects/Textures/Smoke/T_Puffy_Smoke_Trail_01}.G) {constExponent=4}
  n5 = ScalarParameter() {defaultValue=6}
  n6 = Subtract(B=ComponentMask(None=TextureCoordinate()) {g=true}) {constA=0.5}
  n7 = MaterialFunctionCall!(Side 1 Length=Multiply(A=Abs(None=@n6)) {constB=2}) {materialFunction=/Engine/Functions/Engine_MaterialFunctions02/Math/pythagoreanTheorum}
  n8 = Multiply(A=@n1.A, B=Multiply(A=Multiply(A=@n4, B=@n5), B=Multiply(A=@n7.Side 2 Length, B=@n7.Side 2 Length)))
  n9 = SphereMask(
   A=TextureCoordinate() {vTiling=0}
   B=ComponentMask(None=VectorParameter() {parameterName=StarrtendMask_Offset, defaultValue=<rgba 0.5 0 0 1>}.RGB) {g=true, r=true}
   Radius=Constant() {r=0.5}
   Hardness=Constant() {r=0.5}
  )
  n10 = ScalarParameter() {defaultValue=256, parameterName="Distance (S)"}
  n11 = DepthFade(
   Opacity=Multiply(
    A=Multiply(A=@n8, B=@n9)
    B=MaterialFunctionCall(Distance=@n10) {materialFunction=&Near_Camera_Fade}.Result
   )
   FadeDistance=ScalarParameter() {parameterName=FadeDistance}
  )
  n12 = Fresnel() {baseReflectFraction=0}
  LinearInterpolate(A=@n11, B=Multiply(A=@n11, B=OneMinus(None=@n12)), Alpha=ScalarParameter() {parameterName=Fresnel})
  MaterialFunctionCall(Emissive=@n1.RGB) {materialFunction=&MF_FX_EmissiveAdjustments}
 ]
}
