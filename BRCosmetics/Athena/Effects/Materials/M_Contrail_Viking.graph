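# Unlit translucent Viking contrail material for beam trails and Niagara ribbons, with emissive color derived from particle color.
# A scrolling Viking pattern combines with flare, noise and masks, scaled by time of day, to shape opacity.
# UseLocalSpaceUV switches between world-position UVs with object-centered sphere masks and dynamic-parameter motion, or panned mesh UVs.
# Non-mobile paths add depth and camera fades; an optional switch enables SplineThicken world position offset.
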Material /BRCosmetics/Athena/Effects/Materials/M_Contrail_Viking
graph {
 properties {
  bUsedWithBeamTrails=true
  bUsedWithNiagaraRibbons=true
  blendMode=BLEND_Translucent
  pixelDepthOffsetMode=PDOM_Legacy
  shadingModel=MSM_Unlit
  usedShadingModels=Unlit
 }
 nodes [
  n1 = WorldPosition()
  n2 = Divide(A=ComponentMask(None=@n1.XYZ) {b=true}) {constB=1000}
  Frac(None=@n2) {desc=Fixed JIRA FORT-100840 by break this node}
  n3 = ParticleColor()
  MaterialFunctionCall(Emissive=MaterialFunctionCall(LDR Input=Clamp(None=@n3.RGB), HDR Tint=@n3.RGB) {materialFunction=&DeriveHDRfromLDR}.RemappedRange) {materialFunction=&MF_FX_EmissiveAdjustments}
  n4 = TextureCoordinate()
  n5 = SphereMask(A=WorldPosition().XYZ, B=ObjectPositionWS(), Radius=Constant() {r=300}, Hardness=Constant())
  n6 = Multiply(A=Add(A=Time(), B=DynamicParameter() {defaultValue=<rgba 1 1 1 0>, paramNames [RandomOffset, BlendBlur, Param3, FrontEndMotion]}.RandomOffset)) {constB=-1}
  n7 = TextureCoordinate()
  n8 = MaterialFunctionCall(XY=AppendVector(A=Add(A=@n5, B=@n6), B=ComponentMask(None=@n7) {g=true})) {materialFunction=&Swizzle}
  n9 = TextureSample(UVs=TextureCoordinate() {coordinateIndex=1}) {samplerType=SAMPLERTYPE_Grayscale, texture=&LowResBlurredNoise}
  n10 = TextureSample() {texture=/Game/Effects/Fort_Effects/Textures/Mask/T_StormEdgeMask}
  n11 = Constant() {r=1}
  n12 = Constant() {r=0.5}
  n13 = Multiply(
   A=Multiply(
    A=TextureSample(UVs=@n8.YX) {samplerSource=SSM_Wrap_WorldGroupSettings, texture=/Game/Global/Textures/Default/Masks_and_Gradients/T_New_Soft_Flair}.R
    B=Multiply(A=@n9.R, B=@n10.R)
   )
   B=MaterialFunctionCall(Color_Morning=@n11, Color_Day=@n11, Color_Evening=@n12, Color_Night=@n12) {materialFunction=&Color_Time_of_Day}.Result
  )
  n14 = SphereMask(A=@n1.XYZ, B=ObjectPositionWS(), Radius=Constant() {r=380}, Hardness=Constant())
  n15 = Add(
   A=Add(A=Multiply(A=@n14) {constB=3}, B=Multiply(A=@n2) {constB=-1}) {constB=0}
   B=Multiply(A=Time(), B=DynamicParameter() {defaultValue=<rgba 1 1 1 0>, paramNames [RandomOffset, BlendBlur, Param3, FrontEndMotion]}.FrontEndMotion) {constB=0}
  )
  n16 = ComponentMask(None=TextureCoordinate()) {g=true}
  n17 = TextureCoordinate() {coordinateIndex=1}
  n18 = TextureSample(UVs=StaticSwitchParameter(True=AppendVector(A=@n15, B=@n16), False=Panner(Coordinate=@n17) {bFractionalPart=true, speedX=2}) {defaultValue=true, parameterName=UseLocalSpaceUV}) {samplerSource=SSM_Wrap_WorldGroupSettings, texture=/Game/Effects/Fort_Effects/Textures/Misc/T_Viking_Pattern02}
  n19 = OneMinus(None=SphereMask(A=WorldPosition().XYZ, B=ObjectPositionWS(), Radius=Constant() {r=300}, Hardness=Constant()))
  n20 = DynamicParameter() {defaultValue=<rgba 1 1 1 0>, paramNames [RandomOffset, BlendBlur, Param3, FrontEndMotion]}
  n21 = TextureCoordinate()
  n22 = LinearInterpolate(
   A=@n18.R
   B=@n18.R
   Alpha=StaticSwitchParameter(
    True=LinearInterpolate(B=@n19, Alpha=@n20.BlendBlur)
    False=MaterialFunctionCall(Float2=@n21) {materialFunction=&BreakOutFloat2Components}.G
   ) {defaultValue=true, parameterName=UseLocalSpaceUV}
  )
  n23 = Multiply(A=Multiply(A=@n22, B=@n13)) {constB=100}
  n24 = MaterialFunctionCall!(
   Blend=Divide(A=@n13) {constB=5}
   Base=MaterialFunctionCall!(Blend=@n13, Base=Add(A=@n22, B=@n23)) {materialFunction=/Engine/Functions/Engine_MaterialFunctions03/Blends/Blend_Lighten}.Result
  ) {materialFunction=/Engine/Functions/Engine_MaterialFunctions03/Blends/Blend_Screen}
  n25 = TextureCoordinate() {vTiling=0.5}
  n26 = TextureCoordinate() {vTiling=0.6}
  n27 = Constant() {r=0.05}
  n28 = Multiply(
   A=TextureSample(UVs=Panner(Coordinate=@n25) {speedY=-8}) {samplerType=SAMPLERTYPE_Grayscale, texture=&LowResBlurredNoise}.R
   B=TextureSample(UVs=Panner(Coordinate=@n26, Time=@n27) {speedY=5}) {samplerSource=SSM_Clamp_WorldGroupSettings, texture=/Game/Effects/Fort_Effects/Textures/Mask/T_HalfSphere_D}.R
  )
  n29 = Constant() {r=0.7}
  n30 = Constant() {r=0.3}
  n31 = MaterialFunctionCall(Color_Morning=@n29, Color_Day=@n29, Color_Evening=@n30, Color_Night=@n30) {materialFunction=&Color_Time_of_Day}
  n32 = OneMinus(None=MaterialFunctionCall(Float2=TextureCoordinate()) {materialFunction=&BreakOutFloat2Components}.G)
  n33 = Multiply(A=@n32, B=@n32)
  n34 = MaterialFunctionCall(Blend=Multiply(A=Multiply(A=@n28, B=@n31.Result), B=@n33), Base=Multiply(A=@n33, B=@n22)) {materialFunction=&Blend_Lighten}
  n35 = ScalarParameter() {defaultValue=50, parameterName=CameraDepthFasde}
  n36 = Multiply(
   A=Multiply(A=@n3.A, B=StaticSwitchParameter(True=@n24.Result, False=@n34.Result) {defaultValue=true, parameterName=UseLocalSpaceUV})
   B=ShadingPathSwitch(
    Default=MaterialFunctionCall(Fade Length=@n35) {materialFunction=&CameraDepthFade}.Result
    Mobile=Constant() {r=1}
   ) {desc=Migrated from FeatureLevelSwitch}
  )
  n37 = Constant() {r=50}
  Multiply(
   A=Power(Base=ComponentMask(None=@n4) {r=true}, Exp=ScalarParameter() {defaultValue=0.8, parameterName=Exponent})
   B=ShadingPathSwitch(Default=DepthFade(Opacity=@n36, FadeDistance=@n37), Mobile=@n36) {desc=Migrated from FeatureLevelSwitch}
  )
  n38 = ScalarParameter() {defaultValue=10, parameterName=Width}
  StaticSwitchParameter(
   True=MaterialFunctionCall(WidthBase=@n38, WidthTip=@n38) {materialFunction=&SplineThicken}.WorldPositionOffset
   False=Constant()
  ) {parameterName=UseSplineWPO}
 ]
}
