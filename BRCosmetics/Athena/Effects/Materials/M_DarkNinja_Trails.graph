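# Unlit, two-sided translucent material for Dark Ninja cosmetic beam trails, defaulting to purple.
# Dark and bright color parameters blend by inverted U times a ramp texture, multiply by particle color, then pass through MF_FX_EmissiveAdjustments.
# Opacity combines the ramp, particle alpha, a steep U falloff and a short clamped fade-in at the U start.
# An optional DepthFade softens intersections, but the mobile shading path bypasses it.
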Material /BRCosmetics/Athena/Effects/Materials/M_DarkNinja_Trails
graph {
 properties {
  bUsedWithBeamTrails=true
  blendMode=BLEND_Translucent
  pixelDepthOffsetMode=PDOM_Legacy
  shadingModel=MSM_Unlit
  twoSided=true
  usedShadingModels=Unlit
 }
 nodes [
  n1 = VectorParameter() {parameterName=EmissiveColorDark, defaultValue=<rgba 0.0119 0 0.0521 1>}
  n2 = VectorParameter() {parameterName=EmissiveColorBright, defaultValue=<rgba 0.3283 0 1 1>}
  n3 = ComponentMask(None=TextureCoordinate()) {r=true}
  n4 = OneMinus(None=@n3)
  n5 = TextureSample(UVs=TextureCoordinate()) {texture=/Game/Global/Textures/Default/Masks_and_Gradients/T_Water_Top_Ramp}
  n6 = Multiply(A=@n4, B=@n5.G)
  n7 = ParticleColor()
  MaterialFunctionCall(Emissive=Multiply(A=LinearInterpolate(A=@n1.RGB, B=@n2.RGB, Alpha=@n6), B=@n7.RGB)) {materialFunction=&MF_FX_EmissiveAdjustments}
  n8 = Power(Base=@n4) {constExponent=3}
  n9 = Clamp(None=LinearInterpolate(Alpha=@n3) {constB=12})
  n10 = Multiply(A=@n7.A, B=Multiply(A=@n5.G, B=Multiply(A=@n8, B=@n9)))
  n11 = ScalarParameter() {defaultValue=20, parameterName=DepthFadeAlpha}
  ShadingPathSwitch(Default=StaticSwitchParameter(True=DepthFade(Opacity=@n10, FadeDistance=@n11), False=@n10) {parameterName=UseDepthFade}, Mobile=@n10) {desc=Migrated from FeatureLevelSwitch}
 ]
}
