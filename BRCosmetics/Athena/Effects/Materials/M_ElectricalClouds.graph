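# Translucent unlit particle-sprite material for cloud puffs, with fake sun shading and time-of-day tinting.
# A normal map dotted with the atmospheric light vector blends from a shadow-strength floor, scaling morning, day, evening and night colors.
# That shading multiplies a smoke texture, randomly rotated per particle, and particle color, then passes through an emissive adjustment function.
# Opacity multiplies particle alpha by rotated and unrotated samples of a soft smoke mask.
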Material /BRCosmetics/Athena/Effects/Materials/M_ElectricalClouds
graph {
 properties {
  bUsedWithParticleSprites=true
  blendMode=BLEND_Translucent
  pixelDepthOffsetMode=PDOM_Legacy
  shadingModel=MSM_Unlit
  usedShadingModels=Unlit
 }
 nodes [
  n1 = DotProduct(
   A=AtmosphericLightVector()
   B=Transform(None=TextureSample() {samplerType=SAMPLERTYPE_Normal, texture=/Game/Effects/Fort_Effects/Textures/PointyNormal}.RGB)
  )
  n2 = Multiply(
   A=MaterialFunctionCall(
    Color_Morning=VectorParameter() {parameterName=Color_Morning, defaultValue=<rgba 0.6 0.3144 0.24 0>}.RGB
    Color_Day=VectorParameter() {parameterName=Color_Day, defaultValue=<rgba 0.9 0.9617 1 0>}.RGB
    Color_Evening=VectorParameter() {parameterName=Color_Evening, defaultValue=<rgba 0.8 0.5214 0.48 0>}.RGB
    Color_Night=VectorParameter() {parameterName=Color_Night, defaultValue=<rgba 0.25 0.2916 0.5 0>}.RGB
   ) {materialFunction=&Color_Time_of_Day}.Result
   B=LinearInterpolate(A=ScalarParameter() {defaultValue=0.2, parameterName=ShadowStrength}, Alpha=Clamp(None=@n1))
  )
  n3 = Rotator(Time=Multiply(A=ParticleRandom()) {constB=6.28})
  n4 = TextureSample(UVs=@n3) {texture=/Game/Global/Textures/Default/Tileable_Patterns/T_Smoke_Puff_Tile_Up}
  n5 = ParticleColor()
  MaterialFunctionCall(Emissive=Multiply(A=Multiply(A=@n2, B=@n4.G), B=@n5.RGB)) {materialFunction=&MF_FX_EmissiveAdjustments}
  Multiply(
   A=@n5.A
   B=Multiply(
    A=TextureSample(UVs=@n3) {texture=/Game/Effects/Fort_Effects/Textures/T_Smoke_Soft}.G
    B=TextureSample() {texture=&T_Smoke_Soft}.G
   )
  )
 ]
}
