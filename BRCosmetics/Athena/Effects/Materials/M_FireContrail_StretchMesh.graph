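# Unlit translucent material for a stretched fire contrail cosmetic effect on mesh particles.
# SplineThicken shapes the mesh width, using a particle dynamic parameter for the base and a scalar parameter for the tip.
# Three fire texture samples pan at different speeds, two with U offsets (one per particle), and sum to drive emissive and opacity.
# Particle color tints the result, while UV gradients fade opacity at the sides and toward one end.
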Material /BRCosmetics/Athena/Effects/Materials/M_FireContrail_StretchMesh
graph {
 properties {
  bUsedWithMeshParticles=true
  bUsedWithNiagaraMeshParticles=true
  blendMode=BLEND_Translucent
  pixelDepthOffsetMode=PDOM_Legacy
  shadingModel=MSM_Unlit
  usedShadingModels=Unlit
 }
 nodes [
  n1 = ComponentMask(None=TextureCoordinate()) {g=true}
  n2 = OneMinus(None=@n1)
  MaterialFunctionCall(
   WidthBase=DynamicParameter() {paramNames [RandomOffsetX, WidthBase, Param3, Param4]}.WidthBase
   WidthTip=ScalarParameter() {parameterName=Width_Tip}
   UVs for Thickness=Clamp(None=Multiply(A=@n2, B=@n1))
  ) {materialFunction=&SplineThicken}
  n3 = ParticleColor()
  n4 = TextureSample!(
   UVs=Panner(Coordinate=TextureCoordinate() {coordinateIndex=1, uTiling=0.25, vTiling=0.2}) {bFractionalPart=true, speedY=0.8}
   Tex
   Level
   Apply View MipBias
  ) {constMipValue=2, mipValueMode=TMVM_MipLevel, texture=/BRCosmetics/Effects/Fort_Effects/Textures/Fire/T_Fire_SubUV_4x2}
  n5 = TextureCoordinate() {coordinateIndex=1, uTiling=0.25, vTiling=0.2}
  n6 = Add(A=ComponentMask(None=@n5) {r=true}, B=ScalarParameter() {defaultValue=0.5, parameterName=OffsetX})
  n7 = ComponentMask(None=@n5) {g=true}
  n8 = TextureSample(UVs=Panner(Coordinate=AppendVector(A=@n6, B=@n7)) {bFractionalPart=true, speedY=0.6}) {constMipValue=2, mipValueMode=TMVM_MipLevel, texture=&T_Fire_SubUV_4x2}
  n9 = TextureCoordinate() {coordinateIndex=1, uTiling=0.25, vTiling=0.25}
  n10 = Add(A=ComponentMask(None=@n9) {r=true}, B=DynamicParameter() {paramNames [RandomOffsetX, WidthBase, Param3, Param4]}.RandomOffsetX)
  n11 = ComponentMask(None=@n9) {g=true}
  n12 = TextureSample(UVs=Panner(Coordinate=AppendVector(A=@n10, B=@n11)) {bFractionalPart=true, speedY=0.4}) {constMipValue=2, mipValueMode=TMVM_MipLevel, texture=/Game/Effects/Fort_Effects/Textures/Fire/T_FireTile}
  n13 = Add(A=Add(A=@n4.G, B=@n8.G), B=@n12.G)
  MaterialFunctionCall(Emissive=Multiply(A=@n3.RGB, B=LinearInterpolate(Alpha=@n13) {constA=0.1})) {materialFunction=&MF_FX_EmissiveAdjustments}
  n14 = Add(A=Add(A=@n4.A, B=@n8.A), B=@n12.A)
  n15 = ComponentMask(None=TextureCoordinate()) {r=true}
  n16 = Multiply(A=OneMinus(None=@n15), B=@n15)
  n17 = Add(A=@n16, B=@n16)
  n18 = ComponentMask(None=TextureCoordinate()) {g=true}
  Multiply(A=@n3.A, B=Multiply(A=Multiply(A=@n14, B=@n17), B=Power(Base=@n18) {constExponent=3}))
 ]
}
