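# Additive unlit material for the Brite Bomber glider contrail, usable on beam trails, Niagara ribbons and spline meshes.
# Floored sphere masks on V build three color stripes, blended with two multiplied panning sparkle-noise samples through a soft center mask.
# Particle alpha and a U-edge falloff scale that color into an emissive adjustment function.
# A fade chain combines the center mask, depth fades, particle alpha and noise scrolled by time and camera direction.
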Material /BRCosmetics/Athena/Effects/Materials/M_GliderContrail_BriteBomber
graph {
 properties {
  bUsedWithBeamTrails=true
  bUsedWithInstancedStaticMeshes=true
  bUsedWithNiagaraRibbons=true
  bUsedWithSplineMeshes=true
  blendMode=BLEND_Additive
  pixelDepthOffsetMode=PDOM_Legacy
  shadingModel=MSM_Unlit
  usedShadingModels=Unlit
 }
 nodes [
  n1 = VectorParameter() {parameterName=SparkleScale02, defaultValue=<rgba 8 5 5 5>}
  n2 = TextureCoordinate()
  n3 = Panner(Coordinate=Multiply(A=ComponentMask(None=@n1.RGB) {g=true, r=true}, B=@n2)) {speedX=0.1, speedY=1.25}
  n4 = VectorParameter() {parameterName=SparkleScale, defaultValue=<rgba 9 5 5 5>}
  n5 = Panner(Coordinate=Multiply(A=ComponentMask(None=@n4.RGB) {g=true, r=true}, B=@n2)) {speedX=-0.25, speedY=2}
  n6 = Power(
   Base=Multiply(
    A=TextureSample(UVs=@n3) {samplerType=SAMPLERTYPE_LinearColor, texture=/Game/Gadgets/Assets/Slow_Field/Textures/T_Slowfield_SparkleNoise_01}.R
    B=TextureSample(UVs=@n5) {samplerType=SAMPLERTYPE_LinearColor, texture=&T_Slowfield_SparkleNoise_01}.G
   )
   Exp=ScalarParameter() {defaultValue=2, parameterName=SparklesPower}
  ) {constExponent=0.2}
  n7 = ScalarParameter() {defaultValue=1500, parameterName=SparklesInstence}
  n8 = TextureCoordinate() {coordinateIndex=1, uTiling=0}
  n9 = Constant() {r=0.5}
  n10 = Floor(None=SphereMask(A=@n8, B=Constant2Vector() {g=0.5}, Radius=Constant() {r=0.25}, Hardness=@n9))
  n11 = SphereMask(A=@n8, B=Constant2Vector() {g=0.125}, Radius=Constant() {r=0.5}, Hardness=@n9)
  n12 = LinearInterpolate(
   A=LinearInterpolate(
    A=VectorParameter() {parameterName=Color_Stripes_Bot, defaultValue=<rgba 0.7708 0 0.2058 1>}.RGB
    B=VectorParameter() {parameterName=Color_Stripes_Mid, defaultValue=<rgba 0.8801 1 0 1>}.RGB
    Alpha=Clamp(None=@n10)
   )
   B=VectorParameter() {parameterName=Color_Stripes_Top, defaultValue=<rgba 0 0.305 0.1726 1>}.RGB
   Alpha=Clamp(None=Floor(None=@n11))
  )
  n13 = SphereMask(A=TextureCoordinate() {uTiling=0}, B=Constant2Vector() {g=0.5}, Radius=Constant() {r=0.5}, Hardness=Constant())
  n14 = LinearInterpolate(A=Multiply(A=Multiply(A=@n6, B=@n7) {constB=500}, B=@n12), B=@n12, Alpha=Clamp(None=@n13))
  n15 = ParticleColor()
  n16 = ComponentMask(None=TextureCoordinate()) {r=true}
  n17 = Multiply(A=@n16, B=OneMinus(None=@n16))
  MaterialFunctionCall(Emissive=Multiply(A=Multiply(A=@n14, B=@n15.A), B=Multiply(B=@n17) {constA=2})) {materialFunction=&MF_FX_EmissiveAdjustments}
  n18 = DepthFade(Opacity=@n13) {fadeDistanceDefault=5}
  n19 = MaterialFunctionCall(Fade Length=Constant() {r=25}) {materialFunction=&CameraDepthFade}
  n20 = Time()
  n21 = MaterialFunctionCall() {materialFunction=&CameraDirectionVector}
  n22 = Constant3Vector() {constant=<rgba 1 1 0 0>}
  n23 = DotProduct(A=Normalize(VectorInput=Multiply(A=@n21.Result, B=@n22)), B=Constant3Vector() {constant=<rgba 1 0 0 0>})
  n24 = AppendVector(A=Add(A=Multiply(A=@n20) {constB=0.1}, B=ConstantBiasScale(None=@n23)), B=@n20)
  Multiply(
   A=Multiply(A=@n15.A, B=Multiply(A=@n18, B=@n19.Result))
   B=LinearInterpolate(Alpha=TextureSample(UVs=@n24) {texture=&T_Smooth_tile}.G) {constA=0.2, constB=1.5}
  )
 ]
}
