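# Unlit, two-sided translucent material for the DarkViking glider contrail, drawing animated icy flames and streaks on particle beam trails.
# Panned cartoon flame and fire-mask textures, distorted by a dynamically flattened normal map, combine with masked light-ray streaks.
# These masks index two ice-flame curve atlas rows for emissive color, tinted by particle color and adjustable in hue, saturation and luminosity.
# Edge gradients, near-camera fade and particle alpha shape opacity; mobile skips depth fade.
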Material /BRCosmetics/Athena/Effects/Materials/M_GliderContrail_DarkViking
graph {
 properties {
  bUsedWithBeamTrails=true
  bUsedWithMeshDeformer=true
  bUsedWithSkeletalMesh=true
  blendMode=BLEND_Translucent
  pixelDepthOffsetMode=PDOM_Legacy
  refractionMethod=RM_IndexOfRefraction
  shadingModel=MSM_Unlit
  twoSided=true
  usedShadingModels=Unlit
 }
 nodes [
  n1 = ParticleColor()
  n2 = ScalarParameter() {defaultValue=0.5, parameterName=UScale}
  n3 = ScalarParameter() {defaultValue=1, parameterName=VScale}
  n4 = TextureSample(UVs=Panner(Coordinate=TextureCoordinate()) {speedX=-1, speedY=0.1}) {
   samplerSource=SSM_Wrap_WorldGroupSettings
   samplerType=SAMPLERTYPE_Normal
   texture=/Game/Effects/Fort_Effects/Gameplay/SmokeDustExplosions/Textures/T_Explosion_NM_01
  }
  n5 = ComponentMask(None=MaterialFunctionCall(Normal=@n4.RGB, Flatness=DynamicParameter() {defaultValue=<rgba 0.975 1 1 1>, paramNames [UVDistortion,Param2,Param3,Param4]}.UVDistortion) {materialFunction=&FlattenNormal}.Result) {g=true, r=true}
  n6 = Add(A=Multiply(A=TextureCoordinate() {coordinateIndex=1, vTiling=2}, B=AppendVector(A=@n2, B=@n3)), B=@n5)
  n7 = ScalarParameter() {defaultValue=2, parameterName=Bias}
  n8 = TextureSample!(UVs=Panner(Coordinate=Add(A=@n6) {constB=0.37}) {speedX=-0.8, speedY=-0.2}, Tex, Bias=@n7, Apply View MipBias) {
   mipValueMode=TMVM_MipBias
   samplerSource=SSM_Wrap_WorldGroupSettings
   texture=/Game/Effects/Fort_Effects/Textures/Tile/T_Tile_Texture_Cartoon_Flames_Layer
  }
  n9 = TextureSample(UVs=Panner(Coordinate=@n6) {speedX=-0.6, speedY=0.1}, Bias=@n7) {
   mipValueMode=TMVM_MipBias
   samplerSource=SSM_Wrap_WorldGroupSettings
   texture=&T_Tile_Texture_Cartoon_Flames_Layer
  }
  n10 = Add(A=Multiply(A=@n8.B) {constB=0.5}, B=Multiply(A=@n9.B) {constB=0.5})
  n11 = TextureCoordinate()
  n12 = Power(Base=OneMinus(None=ComponentMask(None=@n11) {g=true})) {constExponent=10}
  n13 = Add(A=TextureCoordinate(), B=@n5)
  n14 = ComponentMask(None=@n13) {r=true}
  n15 = AppendVector(A=ComponentMask(None=@n13) {g=true}, B=Add(A=OneMinus(None=@n14)) {constB=-0.15})
  n16 = ComponentMask(None=@n13) {r=true}
  n17 = AppendVector(A=ComponentMask(None=@n13) {g=true}, B=Add(A=OneMinus(None=@n16)) {constB=0.45})
  n18 = Sine(None=ComponentMask(None=TextureCoordinate()) {g=true}) {period=2}
  n19 = ScalarParameter() {defaultValue=1.5, parameterName=Bias_A}
  n20 = TextureSample(UVs=Panner(Coordinate=LinearInterpolate(A=@n15, B=@n17, Alpha=@n18)) {speedX=0.25}, Bias=@n19) {mipValueMode=TMVM_MipBias, texture=/Game/Effects/Fort_Effects/Textures/Fire/T_Fire_Mask}
  n21 = Add(A=TextureCoordinate(), B=@n5)
  n22 = ComponentMask(None=@n21) {r=true}
  n23 = AppendVector(A=ComponentMask(None=@n21) {g=true}, B=Add(A=OneMinus(None=@n22)) {constB=-0.15})
  n24 = ComponentMask(None=@n21) {r=true}
  n25 = AppendVector(A=ComponentMask(None=@n21) {g=true}, B=Add(A=OneMinus(None=@n24)) {constB=0.45})
  n26 = Sine(None=ComponentMask(None=TextureCoordinate()) {g=true}) {period=2}
  n27 = TextureSample(UVs=Panner(Coordinate=LinearInterpolate(A=@n23, B=@n25, Alpha=@n26)) {speedX=0.35}, Bias=@n19) {mipValueMode=TMVM_MipBias, texture=&T_Fire_Mask}
  n28 = Clamp(None=Multiply(A=Add(A=@n10, B=@n12), B=Multiply(A=@n20.R, B=@n27.R)))
  n29 = ComponentMask(None=TextureCoordinate()) {g=true}
  n30 = OneMinus(None=Power(Base=Sine(None=@n29) {period=2}, Exp=ScalarParameter() {defaultValue=0.2, parameterName=Gradient Control Exp}))
  n31 = Panner(Coordinate=MaterialFunctionCall(XY=TextureCoordinate()) {materialFunction=&Swizzle}.YX) {speedX=-0.976}
  n32 = Panner(Coordinate=MaterialFunctionCall(XY=TextureCoordinate() {uTiling=2, vTiling=2}) {materialFunction=&Swizzle}.YX) {speedX=-0.7616}
  n33 = Add(
   A=Power(Base=TextureSample!(UVs=@n31, Tex, Apply View MipBias) {texture=/Game/Effects/Fort_Effects/Textures/Outlander/T_InTheZone_LightRays_01}.R) {constExponent=1.5}
   B=Power(Base=TextureSample(UVs=@n32) {texture=&T_InTheZone_LightRays_01}.G)
  )
  n34 = TextureCoordinate() {uTiling=0.2}
  n35 = ConstantBiasScale(None=TextureSample(UVs=Panner(Coordinate=@n34) {bFractionalPart=true, speedX=-0.65}) {texture=&T_NoisePack_01}.R) {bias=-0.5, scale=0.2}
  n36 = OneMinus(None=ComponentMask(None=TextureCoordinate()) {r=true})
  n37 = LinearInterpolate(
   A=LinearInterpolate(
    A=ScalarParameter() {defaultValue=0.1, parameterName=Fine Streaks Opacity A}
    B=ScalarParameter() {defaultValue=0.3, parameterName=Fine Streaks Opacity B}
    Alpha=Divide(A=@n33)
   ) {constA=1}
   B=ScalarParameter() {defaultValue=1, parameterName=Hard Streaks Opacity}
   Alpha=Power(Base=Add(A=@n35, B=@n36)) {constExponent=4}
  ) {constAlpha=0}
  n38 = Constant() {r=0.5}
  n39 = TextureCoordinate()
  n40 = SphereMask(
   A=AppendVector(
    A=@n38
    B=MaterialFunctionCall(Float2=@n39) {materialFunction=&BreakOutFloat2Components}.G
   )
   B=@n38
   Radius=ScalarParameter() {defaultValue=0.5, parameterName=Mask Radius}
   Hardness=ScalarParameter() {defaultValue=0.9, parameterName=Mask Hardness}
  )
  n41 = VectorParameter() {parameterName=SphereMaskB, defaultValue=<rgba 0.5 0.5 0 1>}
  n42 = OneMinus(None=SphereMask(A=TextureCoordinate() {uTiling=0.5}, B=ComponentMask(None=@n41.RGB) {g=true, r=true}, Radius=Constant() {r=0.7}, Hardness=Constant() {r=0.3}))
  n43 = Multiply(
   A=Multiply(A=Multiply(A=@n37, B=@n40), B=Power(Base=@n42) {constExponent=5})
   B=ScalarParameter() {defaultValue=10, parameterName=Opacity Multi}
  ) {constB=10}
  n44 = Clamp(None=Add(A=@n28, B=Multiply(A=@n30, B=@n43)))
  n45 = ComponentMask(None=TextureCoordinate()) {r=true}
  n46 = ScalarParameter() {defaultValue=1.5, parameterName=Leading Edge Exponent}
  n47 = Saturate(None=Multiply(A=Power(Base=@n45, Exp=@n46), B=ScalarParameter() {defaultValue=5, parameterName=Leading Edge Intensity}))
  n48 = ComponentMask(None=TextureCoordinate()) {r=true}
  n49 = Multiply(
   A=Multiply(A=Multiply(A=@n44, B=@n47), B=OneMinus(None=@n48))
   B=MaterialFunctionCall(Distance=ScalarParameter() {defaultValue=64, parameterName=NearFadeDistance}) {materialFunction=&Near_Camera_Fade}.Result
  )
  n50 = DepthFade(Opacity=@n49, FadeDistance=ScalarParameter() {defaultValue=150, parameterName=Depth Fade Distance})
  LinearInterpolate(Alpha=Multiply(A=@n1.A, B=ShadingPathSwitch(Default=@n50, Mobile=@n49) {desc=Migrated from FeatureLevelSwitch})) {constA=1, constB=1.1}
  n51 = CurveAtlasRowParameter(CurveTime=@n44) {
   atlas=&AthenaCurveAtlas01
   curve=/Game/Athena/LinearColorCurves/IceFlamesCurve
   defaultValue=5
   parameterName=Curve
  }
  n52 = CurveAtlasRowParameter(CurveTime=@n28) {
   atlas=&AthenaCurveAtlas01
   curve=/Game/Athena/LinearColorCurves/IceFlamesCurve_Dragon
   defaultValue=17
   parameterName=Curve
  }
  n53 = Multiply(A=@n52) {constB=1.1}
  n54 = OneMinus(None=ComponentMask(None=TextureCoordinate()) {r=true})
  n55 = Multiply(A=LinearInterpolate(A=@n51, B=LinearInterpolate(A=@n52, B=@n53, Alpha=@n54) {constA=1, constB=10}, Alpha=@n28), B=@n1.RGB)
  n56 = Max(A=@n55, B=Constant3Vector() {constant=<rgba 0.001 0.001 0.001 0>})
  n57 = MaterialFunctionCall(
   Hue Shift Percentage=ScalarParameter() {group=Color, parameterName=HueShift}
   Texture=MaterialFunctionCall(LDR Input=Clamp(None=@n56) {minDefault=0.001}, HDR Tint=@n55) {materialFunction=&DeriveHDRfromLDR}.RemappedRange
  ) {materialFunction=&HueShift}
  n58 = ScalarParameter() {group=Color, parameterName=Desaturation}
  MaterialFunctionCall(Emissive=MaterialFunctionCall!(Color=Desaturation(None=@n57.Result, Fraction=@n58), Luminosity=ScalarParameter() {defaultValue=1, group=Color, parameterName=Luminosity}) {materialFunction=/Engine/Functions/Engine_MaterialFunctions01/ImageAdjustment/Luminosity_And_Color}.Result) {materialFunction=&MF_FX_EmissiveAdjustments}
 ]
}
