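# Unlit translucent mesh-particle material for the Diesel glider contrail's flame.
# World position offset sums sine flicker, curl noise (zeroed on Default and Low quality), pulsing width/length scaling and optional vertex-color flare driven by a dynamic parameter.
# Four panned caustic samples form the flame pattern, tinted by an orange HDR color and a Fresnel-driven rainbow thin-film layer.
# Opacity combines edge Fresnel, that pattern, a UV sphere mask and optional particle alpha.
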Material /BRCosmetics/Athena/Effects/Materials/M_GliderContrail_Diesel_Flame
graph {
 properties {
  bUsedWithMeshParticles=true
  bUsedWithNiagaraMeshParticles=true
  blendMode=BLEND_Translucent
  pixelDepthOffsetMode=PDOM_Legacy
  shadingModel=MSM_Unlit
  usedShadingModels=Unlit
 }
 nodes [
  n1 = Frac(None=Time())
  n2 = OneMinus(None=ComponentMask(None=TextureCoordinate()) {g=true})
  n3 = Multiply(A=@n2, B=@n2)
  n4 = Sine(None=Add(A=Multiply(A=@n1) {constB=50}, B=Multiply(A=@n3) {constB=-2}))
  n5 = Time()
  n6 = Add(A=ComponentMask(None=TextureCoordinate()) {g=true}, B=@n5)
  n7 = TextureCoordinate()
  n8 = Add(A=Multiply(A=@n5) {constB=40}, B=Multiply(A=ComponentMask(None=@n7) {g=true}) {constB=2})
  n9 = Add(A=Multiply(A=Sine(None=@n6) {period=1.31}) {constB=4, desc=Amplitude}, B=Multiply(A=Sine(None=@n8) {period=1.31}) {constB=4, desc=Amplitude})
  n10 = Multiply(A=VertexNormalWS(), B=Max(A=Add(A=@n4, B=@n9)) {constB=-2.5})
  n11 = Time()
  n12 = AppendVector(A=ComponentMask(None=WorldPosition().XYZ) {b=true}, B=Constant2Vector())
  n13 = Add(A=Multiply(A=Frac(None=@n11)) {constB=50}, B=Frac(None=Multiply(A=@n12) {constB=-0.02}))
  n14 = Sine(None=@n13)
  n15 = Cosine(None=@n13)
  n16 = OneMinus(None=ComponentMask(None=TextureCoordinate()) {g=true})
  n17 = Multiply(
   A=MaterialFunctionCall(X=ComponentMask(None=@n14) {r=true}, Y=ComponentMask(None=@n15) {r=true}, Z=Constant()) {materialFunction=&MakeFloat3}.Result
   B=Multiply(A=@n16, B=@n16)
  )
  n18 = Add(A=ComponentMask(None=TextureCoordinate()) {g=true}) {constB=0.5}
  n19 = Multiply(
   A=ScalarParameter() {defaultValue=0.5, group=WPO, parameterName=ScaleFlickerOffset_WPO}
   B=Multiply(A=Add(A=@n10, B=@n17) {constB=0}, B=Clamp(None=@n18))
  )
  n20 = Constant()
  n21 = Subtract(
   A=Subtract(A=WorldPosition().XYZ, B=ActorPositionWS())
   B=Multiply(A=Time(), B=ScalarParameter() {defaultValue=150, group=WPO, parameterName=Curl Speed Scale, sortPriority=4})
  )
  n22 = Multiply(
   A=ScalarParameter() {defaultValue=15, group=WPO, parameterName=Curl Scale, sortPriority=3}
   B=LinearInterpolate(Alpha=VertexColor().B) {constA=0.5}
  )
  n23 = Multiply(
   A=ScalarParameter() {defaultValue=1, group=WPO, parameterName=CurlIntensity, sortPriority=2}
   B=VectorNoise(World Position=Divide(A=@n21, B=@n22)) {noiseFunction=VNF_VectorALU}
  )
  n24 = LinearInterpolate(
   B=QualitySwitch(Default=@n20, Low=@n20, High=@n23, Medium=@n23)
   Alpha=ScalarParameter() {defaultValue=1, group=WPO, parameterName=Use Curl Noise, sortPriority=1}
  )
  n25 = ComponentMask(None=VertexNormalWS()) {g=true, r=true}
  n26 = Multiply(
   A=MaterialFunctionCall(Speed=ScalarParameter() {defaultValue=1.35, group=WPO, parameterName=Size_Pulse_Speed_01, sortPriority=11}) {materialFunction=&DebugTimeSine}.Result
   B=MaterialFunctionCall(Speed=ScalarParameter() {defaultValue=3, group=WPO, parameterName=Size_Pulse_Speed_02, sortPriority=12}) {materialFunction=&DebugTimeSine}.Result
  )
  n27 = TextureCoordinate() {uTiling=0}
  n28 = Constant() {r=0.5}
  n29 = Constant()
  n30 = SphereMask(A=@n27, B=Constant2Vector() {g=0.5}, Radius=@n28, Hardness=@n29)
  n31 = ComponentMask(None=LinearInterpolate(
    A=ScalarParameter() {defaultValue=1, group=WPO, parameterName=Pulse_Scale_Width_Min, sortPriority=6}
    B=ScalarParameter() {defaultValue=5, group=WPO, parameterName=Pulse_Scale_Width_Max, sortPriority=7}
    Alpha=Multiply(A=@n26, B=@n30)
   )) {g=true, r=true}
  n32 = Constant2Vector() {g=0.25}
  n33 = LinearInterpolate(
   A=ScalarParameter() {defaultValue=1, group=WPO, parameterName=Pulse_Scale_Length_Min, sortPriority=8}
   B=ScalarParameter() {defaultValue=5, group=WPO, parameterName=Pulse_Scale_Length_Max, sortPriority=9}
   Alpha=Multiply(A=@n26, B=SphereMask(A=@n27, B=@n32, Radius=@n28, Hardness=@n29))
  )
  n34 = ComponentMask(None=Transform(None=VectorParameter() {group=WPO, parameterName=UpwardsNormal, sortPriority=10, defaultValue=<rgba 0 0 1 1>}.RGB) {transformSourceType=TRANSFORMSOURCE_World, transformType=TRANSFORM_ParticleWorld}) {b=true}
  n35 = Add(A=AppendVector(A=Multiply(A=@n25, B=@n31), B=Constant()), B=AppendVector(A=Constant2Vector(), B=Multiply(A=@n33, B=@n34)))
  n36 = ScalarParameter() {defaultValue=1, group=WPO, parameterName=Use Pulse WPO, sortPriority=5}
  n37 = VertexColor()
  n38 = Multiply(A=VertexNormalWS(), B=DynamicParameter() {defaultValue=<rgba 1 1 1 0>, paramNames [Param1, Param2, Param3, Flare Amount]}.Flare Amount)
  Add(
   A=Add(A=Add(A=@n19, B=@n24), B=LinearInterpolate(B=@n35, Alpha=@n36))
   B=LinearInterpolate(B=Multiply(A=@n37.R, B=@n38), Alpha=ScalarParameter() {group=WPO, parameterName=Use Vertex Flare Out, sortPriority=13})
  )
  n39 = ScalarParameter() {defaultValue=0.75, group=ThinFilm, parameterName="Hue Shift Percentage (S)", sortPriority=4}
  n40 = ScalarParameter() {defaultValue=0.35, group=ThinFilm, parameterName=ThinFilm_Scale, sortPriority=1}
  n41 = Fresnel(ExponentIn=ScalarParameter() {defaultValue=0.5, group=ThinFilm, parameterName=ThinFilm_EX, sortPriority=2}, Normal=VertexNormalWS()) {baseReflectFraction=0}
  n42 = Multiply(
   A=TextureSampleParameter2D(UVs=Multiply(A=@n40, B=@n41)) {group=ThinFilm, parameterName=ThinFilmTexture, texture=/Game/Global/Textures/Default/Masks_and_Gradients/FortniteRainbow}.RGB
   B=ScalarParameter() {defaultValue=1, group=ThinFilm, parameterName=ThinFilm_Intensity, sortPriority=3}
  )
  n43 = Multiply(
   A=OneMinus(None=ScalarParameter() {defaultValue=0.7, group=ThinFilm, parameterName=Roughness, sortPriority=6})
   B=ScalarParameter() {defaultValue=1, group=ThinFilm, parameterName=ThinFilm_RoughnessScale, sortPriority=7}
  )
  n44 = ScalarParameter() {defaultValue=0.1, group=ThinFilm, parameterName=ThinFilm_Visibility, sortPriority=8}
  n45 = VertexNormalWS()
  n46 = LinearInterpolate(A=@n42, B=@n42, Alpha=Multiply(A=Clamp(None=@n43), B=Fresnel(ExponentIn=@n44, Normal=@n45) {baseReflectFraction=0})) {constAlpha=0, desc=Removed becuase values never deviates from 0 to 1}
  n47 = LinearInterpolate(
   B=Desaturation(
    None=MaterialFunctionCall(Hue Shift Percentage=@n39, Texture=@n46) {materialFunction=&HueShift}.Result
    Fraction=ScalarParameter() {defaultValue=0.75, group=ThinFilm, parameterName=Desat_Fraction, sortPriority=5}
   )
   Alpha=ScalarParameter() {defaultValue=1, group=ThinFilm, parameterName=Use Thin Film}
  ) {constA=1}
  n48 = OneMinus(None=TextureCoordinate())
  n49 = VectorParameter() {group=Flames, parameterName=Flame Texture UV Scale A, sortPriority=1, defaultValue=<rgba 2 0.25 0 0>}
  n50 = Panner(Coordinate=Multiply(A=TextureCoordinate(), B=ComponentMask(None=@n49.RGB) {g=true, r=true})) {speedX=-0.5, speedY=1}
  n51 = VectorParameter() {group=Flames, parameterName=Flame Texture UV Scale B, sortPriority=2, defaultValue=<rgba 2 0.5 0 0>}
  n52 = Panner(Coordinate=Multiply(A=TextureCoordinate(), B=ComponentMask(None=@n51.RGB) {g=true, r=true})) {speedX=1, speedY=2}
  n53 = VectorParameter() {group=Flames, parameterName=Flame Texture UV Scale C, sortPriority=3, defaultValue=<rgba 1 0.5 0 0>}
  n54 = Panner(Coordinate=Multiply(A=TextureCoordinate(), B=ComponentMask(None=@n53.RGB) {g=true, r=true})) {speedX=-0.5, speedY=1}
  n55 = VectorParameter() {group=Flames, parameterName=Flame Texture UV Scale D, sortPriority=4, defaultValue=<rgba 1.25 0.25 0 0>}
  n56 = Panner(Coordinate=Multiply(A=TextureCoordinate(), B=ComponentMask(None=@n55.RGB) {g=true, r=true})) {speedX=1, speedY=2}
  n57 = LinearInterpolate(
   A=LinearInterpolate(
    A=TextureSampleParameter2D(UVs=@n50) {group=Flames, parameterName=Flame Texture, texture=&T_ShieldCaustics}.G
    B=TextureSampleParameter2D(UVs=@n52) {parameterName=Flame Texture, texture=&T_ShieldCaustics}.G
   )
   B=LinearInterpolate(
    A=TextureSampleParameter2D(UVs=@n54) {parameterName=Flame Texture, texture=&T_ShieldCaustics}.G
    B=TextureSampleParameter2D(UVs=@n56) {parameterName=Flame Texture, texture=&T_ShieldCaustics}.G
   )
  )
  n58 = ScalarParameter() {defaultValue=0.1, group=Flames, parameterName=FlameAmount Contrast, sortPriority=5}
  n59 = TextureCoordinate()
  n60 = CameraVectorWS()
  n61 = Constant3Vector() {constant=<rgba 0 0 -0.5 0>}
  n62 = MaterialFunctionCall(
   Exponent=Multiply(A=ScalarParameter() {defaultValue=50, parameterName=FresnelExp}, B=ComponentMask(None=@n59) {g=true})
   Camera Vector=Normalize(VectorInput=Add(A=@n60, B=@n61))
  ) {materialFunction=&MF_Fresnel_FX}
  n63 = ScalarParameter() {parameterName=Remove Hard Edge}
  n64 = Clamp(None=Add(A=Power(Base=@n57, Exp=@n58), B=LinearInterpolate(A=@n62.Fresnel Inverted, Alpha=@n63) {constB=0}))
  n65 = TextureCoordinate()
  n66 = Constant() {r=2.5}
  n67 = Add(
   A=Multiply(A=ComponentMask(None=@n48) {g=true}, B=Multiply(A=@n64) {constB=10})
   B=Multiply(
    A=ComponentMask(None=@n65) {g=true}
    B=MaterialFunctionCall(Exponent=@n66) {materialFunction=&MF_Fresnel_FX}.Fresnel Inverted
   )
  )
  MaterialFunctionCall(Emissive=Multiply(A=Multiply(A=@n47, B=@n67), B=VectorParameter() {group=Color, parameterName=MainFlame_Color, defaultValue=<rgba 250 25 5 1>}.RGB)) {materialFunction=&MF_FX_EmissiveAdjustments}
  n68 = OneMinus(None=@n64)
  n69 = Multiply(
   A=MaterialFunctionCall(Exponent=ScalarParameter() {defaultValue=0.25, group=Opacity, parameterName=EdgeTrans, sortPriority=5}) {materialFunction=&MF_Fresnel_FX}.Fresnel
   B=Add(A=Power(Base=@n64) {constExponent=15}, B=Power(Base=@n68) {constExponent=8}) {constB=0}
  )
  n70 = VectorParameter() {group=Opacity, parameterName=SphereMask_A Scale, sortPriority=1, defaultValue=<rgba 0 1 0 0>}
  n71 = SphereMask(
   A=Multiply(A=TextureCoordinate(), B=ComponentMask(None=@n70.RGB) {g=true, r=true})
   B=ComponentMask(None=VectorParameter() {group=Opacity, parameterName=SphereMask_B, sortPriority=2, defaultValue=<rgba 0 0.525 0 0>}.RGB) {g=true, r=true}
   Radius=ScalarParameter() {defaultValue=0.5, group=Opacity, parameterName=RadiusEdge, sortPriority=3}
   Hardness=ScalarParameter() {group=Opacity, parameterName=HardnessEdge, sortPriority=4}
  )
  Multiply(
   A=LinearInterpolate(B=ParticleColor().A, Alpha=ScalarParameter() {group=Opacity, parameterName=Use Particle Alpha}) {constA=1}
   B=Clamp(None=Multiply(A=@n69, B=@n71))
  )
 ]
}
