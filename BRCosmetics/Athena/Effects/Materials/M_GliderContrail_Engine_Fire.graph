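# Translucent, unlit, two-sided engine-fire material for glider contrails, flagged for mesh particles and Niagara ribbons.
# Panned samples of a laser-sword trail noise texture blend dark and light colors and break up opacity, with per-particle tiling, offset and brightness from dynamic parameters.
# Opacity also combines particle alpha, depth fade, a V-centered sphere mask and a Fresnel whose exponent varies with view direction.
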Material /BRCosmetics/Athena/Effects/Materials/M_GliderContrail_Engine_Fire
graph {
 properties {
  bUsedWithMeshParticles=true
  bUsedWithNiagaraMeshParticles=true
  bUsedWithNiagaraRibbons=true
  blendMode=BLEND_Translucent
  pixelDepthOffsetMode=PDOM_Legacy
  shadingModel=MSM_Unlit
  twoSided=true
  usedShadingModels=Unlit
 }
 nodes [
  n1 = DynamicParameter() {defaultValue=<rgba 1 5 1 1>, paramNames [RandTile,ColorMult,TileMult,Param4]}
  n2 = VectorParameter() {parameterName=TileTex02, defaultValue=<rgba 1 0.1 0 1>}
  n3 = Multiply(
   A=Multiply(A=DynamicParameter() {defaultValue=<rgba 1 5 1 1>, paramNames [RandTile,ColorMult,TileMult,Param4]}.TileMult, B=TextureCoordinate())
   B=Add(A=@n1.RandTile, B=ComponentMask(None=@n2.RGB) {g=true, r=true})
  ) {constB=1.7}
  n4 = Panner(Coordinate=@n3) {speedY=2}
  n5 = Panner(Coordinate=@n3) {speedY=1}
  n6 = Power(Base=Multiply(
    A=TextureSample(UVs=@n4) {texture=/Game/Weapons/FORT_Melee/Effects/LaserSwordFounders/T_LaserSwordTrailNoise}.G
    B=TextureSample(UVs=@n5) {texture=&T_LaserSwordTrailNoise}.G
   )) {constExponent=0.2}
  n7 = ParticleColor()
  n8 = ScalarParameter() {defaultValue=2, parameterName=Fresnel}
  n9 = MaterialFunctionCall(Float3=Transform(None=CameraVectorWS()) {transformSourceType=TRANSFORMSOURCE_World, transformType=TRANSFORM_Local}) {materialFunction=&BreakOutFloat3Components}
  n10 = MaterialFunctionCall(X=@n9.R, Y=@n9.R) {materialFunction=&MakeFloat3}
  n11 = CameraVectorWS()
  n12 = LinearInterpolate(A=Constant() {r=1}, B=Constant() {r=8}, Alpha=Power(Base=DotProduct(A=@n10.Result, B=@n11)))
  n13 = Multiply(A=Multiply(
    A=@n7.A
    B=MaterialFunctionCall(Brightness=@n8, Exponent=@n12) {materialFunction=&MF_Fresnel_FX}.Fresnel
   ))
  n14 = DepthFade(FadeDistance=ScalarParameter() {defaultValue=2, parameterName=FadeDepth})
  Multiply(
   A=Multiply(A=@n6, B=Multiply(A=@n13, B=@n14))
   B=SphereMask(
    A=TextureCoordinate() {uTiling=0}
    B=Constant2Vector() {g=0.5}
    Radius=ScalarParameter() {defaultValue=0.45, parameterName=Radius}
    Hardness=ScalarParameter() {parameterName=Hardness}
   )
  )
  n15 = Multiply(
   A=DynamicParameter() {defaultValue=<rgba 1 5 1 1>, paramNames [RandTile,ColorMult,TileMult,Param4]}.TileMult
   B=TextureCoordinate() {uTiling=2}
  )
  n16 = Add(A=ComponentMask(None=VectorParameter() {parameterName=TileTex01, defaultValue=<rgba 1 0.2 0 1>}.RGB) {g=true, r=true}, B=@n1.RandTile)
  n17 = TextureSample(UVs=Panner(Coordinate=Multiply(A=@n15, B=@n16) {constB=4}) {bFractionalPart=true, speedY=1}) {texture=&T_LaserSwordTrailNoise}
  n18 = TextureSample(UVs=Panner(Coordinate=@n3) {bFractionalPart=true, speedY=1}) {texture=&T_LaserSwordTrailNoise}
  n19 = LinearInterpolate(
   A=VectorParameter() {parameterName=DarkCol, defaultValue=<rgba 0.5 0.5 0.5 0>}.RGB
   B=VectorParameter() {parameterName=LightCol, defaultValue=<rgba 1.5 1.5 1.5 0>}.RGB
   Alpha=Power(Base=Multiply(A=@n17.G, B=@n18.G)) {constExponent=0.5}
  )
  n20 = DynamicParameter() {defaultValue=<rgba 1 5 1 1>, paramNames [RandTile,ColorMult,TileMult,Param4]}
  Multiply(A=Multiply(A=Multiply(A=@n19, B=@n20.ColorMult), B=@n6), B=@n7.RGB)
 ]
}
