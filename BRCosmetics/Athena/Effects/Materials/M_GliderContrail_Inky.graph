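# Translucent, unlit, two-sided beam-trail material for an inky glider contrail with rainbow-tinted smoke.
# Two counter-panning smoke samples are summed, edge-masked and scaled by particle alpha and camera depth fade (skipped on mobile) for opacity.
# Dynamic parameters blend in noise-driven erosion of that alpha.
# Emissive color comes from a rainbow texture using reflection-vector UVs distorted by the smoke, then hue-shifted, faded along U and tinted by particle color.
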Material /BRCosmetics/Athena/Effects/Materials/M_GliderContrail_Inky
graph {
 properties {
  bUsedWithBeamTrails=true
  blendMode=BLEND_Translucent
  pixelDepthOffsetMode=PDOM_Legacy
  shadingModel=MSM_Unlit
  translucencyLightingMode=TLM_VolumetricPerVertexNonDirectional
  twoSided=true
  usedShadingModels=Unlit
 }
 nodes [
  n1 = ScalarParameter() {defaultValue=0.33, group=Smoke A, parameterName=InkSmokeA_TileU}
  n2 = ScalarParameter() {defaultValue=1, group=Smoke A, parameterName=InkSmokeA_TileV}
  n3 = Panner(
   Coordinate=Multiply(A=TextureCoordinate() {coordinateIndex=1}, B=AppendVector(A=@n1, B=@n2))
   Speed=AppendVector(
    A=ScalarParameter() {defaultValue=0.1, group=Smoke A, parameterName=InkSmokeA_PanU}
    B=ScalarParameter() {group=Smoke A, parameterName=InkSmokeA_PanV}
   )
  )
  n4 = ScalarParameter() {defaultValue=0.25, group=Smoke B, parameterName=InkSmokeB_TileU}
  n5 = ScalarParameter() {defaultValue=1, group=Smoke B, parameterName=InkSmokeB_TileV}
  n6 = Panner(
   Coordinate=Multiply(A=TextureCoordinate() {coordinateIndex=1}, B=AppendVector(A=@n4, B=@n5))
   Speed=AppendVector(
    A=ScalarParameter() {defaultValue=-0.13, group=Smoke B, parameterName=InkSmokeB_PanU}
    B=ScalarParameter() {group=Smoke B, parameterName=InkSmokeB_PanV}
   )
  )
  n7 = Add(
   A=Power(
    Base=TextureSample(UVs=@n3) {texture=/BRCosmetics/Athena/Effects/Textures/T_InkySmokeTrail}.G
    Exp=ScalarParameter() {defaultValue=0.25, group=Smoke A, parameterName=SmokeAOpaPow}
   )
   B=Power(
    Base=TextureSample(UVs=@n6) {texture=&T_InkySmokeTrail}.R
    Exp=ScalarParameter() {defaultValue=0.25, group=Smoke B, parameterName=SmokeBOpaPow}
   )
  )
  n8 = Panner() {speedY=1}
  n9 = MaterialFunctionCall() {materialFunction=&Swizzle}
  n10 = Multiply(
   A=Multiply(A=@n7) {constB=0.5}
   B=Multiply(
    A=TextureSample(UVs=@n8) {texture=/Game/Effects/Fort_Effects/Textures/Mask/T_StormEdgeMask}.G
    B=TextureSample(UVs=@n9.YX) {texture=&T_StormEdgeMask}.G
   )
  )
  n11 = ScalarParameter() {defaultValue=10, parameterName=OverallOpacity}
  n12 = ScalarParameter() {defaultValue=500, parameterName=FadeLength}
  n13 = ScalarParameter() {parameterName=FadeOffset}
  n14 = Multiply(
   A=Multiply(A=Multiply(A=@n10, B=@n11), B=ParticleColor().A)
   B=ShadingPathSwitch(
    Default=MaterialFunctionCall(Fade Length=@n12, Fade Offset=@n13) {materialFunction=&CameraDepthFade}.Result
    Mobile=Constant() {r=1}
   ) {desc=Migrated from FeatureLevelSwitch}
  )
  n15 = DynamicParameter() {defaultValue=<rgba 0 1 0 1>, paramNames [Erode, ErodeMaskScale, ErodeMaskTime, ColorStart]}
  n16 = DynamicParameter() {defaultValue=<rgba 0 1 0 1>, paramNames [Erode, ErodeMaskScale, ErodeMaskTime, ColorStart]}
  n17 = ScalarParameter() {defaultValue=2, parameterName=ErodeMaskScale}
  n18 = TextureCoordinate() {vTiling=0.25}
  n19 = AppendVector(A=Multiply(A=Multiply(A=@n16.ErodeMaskScale, B=@n17), B=ComponentMask(None=@n18) {r=true}), B=ComponentMask(None=@n18) {g=true})
  n20 = OneMinus(None=TextureSample(UVs=Panner(Coordinate=@n19) {speedY=0.1}) {samplerType=SAMPLERTYPE_LinearColor, texture=/Game/Environments/Foliage/Textures/T_NoiseMask02_M}.G)
  LinearInterpolate(
   A=@n14
   B=Multiply(
    A=@n14
    B=MaterialFunctionCall!(Dynamic Param=@n15.Erode, Alpha=@n20) {materialFunction=/Game/Effects/Fort_Effects/Materials/_Functions/Erode_Alpha}.Result
   )
   Alpha=DynamicParameter() {defaultValue=<rgba 0 1 0 1>, paramNames [Erode, ErodeMaskScale, ErodeMaskTime, ColorStart]}.ErodeMaskTime
  )
  n21 = Multiply(A=TextureCoordinate() {uTiling=3, vTiling=3}, B=ComponentMask(None=ReflectionVectorWS()) {g=true, r=true})
  n22 = Multiply(A=ScalarParameter() {defaultValue=0.25, parameterName=DistortAmount}, B=@n7)
  n23 = MaterialFunctionCall(
   Hue Shift Percentage=ScalarParameter() {defaultValue=0.9, parameterName="Hue Shift Percentage (S)"}
   Texture=TextureSample(UVs=Add(A=@n21, B=@n22)) {texture=/Game/Global/Textures/Default/Masks_and_Gradients/FortniteRainbow}.RGB
  ) {materialFunction=&HueShift}
  n24 = TextureCoordinate()
  n25 = OneMinus(None=Saturate(None=ComponentMask(None=@n24) {r=true}))
  MaterialFunctionCall(Emissive=Multiply(A=Multiply(A=@n23.Result, B=@n25), B=ParticleColor().RGB)) {materialFunction=&MF_FX_EmissiveAdjustments}
 ]
}
