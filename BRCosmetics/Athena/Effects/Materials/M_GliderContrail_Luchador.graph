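# Unlit, two-sided translucent material for the Luchador glider contrail, usable on Niagara ribbons and beam trails.
# Color samples the LuchadorVFX_Glider curve atlas row along V, tinted by particle color, with optional HDR remapping.
# Opacity blends two panning light-ray layers into hard streaks using a noise-perturbed U gradient, shaped by sphere masks and squared sine, then near-camera fade.
# Static switches, mostly off by default, add sine gradient, gradient-driven opacity, depth fade and refraction.
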Material /BRCosmetics/Athena/Effects/Materials/M_GliderContrail_Luchador
graph {
 properties {
  bUsedWithBeamTrails=true
  bUsedWithMeshDeformer=true
  bUsedWithNiagaraRibbons=true
  bUsedWithSkeletalMesh=true
  blendMode=BLEND_Translucent
  pixelDepthOffsetMode=PDOM_Legacy
  refractionMethod=RM_IndexOfRefraction
  shadingModel=MSM_Unlit
  twoSided=true
  usedShadingModels=Unlit
 }
 nodes [
  n1 = ComponentMask(None=TextureCoordinate()) {g=true}
  n2 = Power(
   Base=StaticSwitchParameter(True=Sine(None=@n1) {period=2}, False=@n1) {group=Control Switches, parameterName=Use Sine on Gradient Curve Control, sortPriority=2}
   Exp=ScalarParameter() {defaultValue=1, parameterName=Gradient Control Exp}
  )
  n3 = CurveAtlasRowParameter(CurveTime=@n2) {
   atlas=&AthenaCurveAtlas01
   curve=/Game/Athena/LinearColorCurves/LuchadorVFX_Glider
   defaultValue=7
   parameterName=Curve
  }
  n4 = ParticleColor()
  n5 = StaticSwitchParameter(True=Multiply(A=@n3, B=@n4.RGB), False=@n4.RGB) {defaultValue=true, group=Control Switches, parameterName=Use Curve or Just Particle Color, sortPriority=1}
  n6 = Clamp(None=@n5)
  n7 = StaticSwitchParameter(True=LinearInterpolate(Alpha=@n2) {constA=2}, False=Constant() {r=1}) {group=Control Switches, parameterName=Modulate HDR Intensity by Gradient Control, sortPriority=3}
  MaterialFunctionCall(Emissive=StaticSwitchParameter(
    True=MaterialFunctionCall(LDR Input=@n6, HDR Tint=@n5, HDR Intensity Multiplier=@n7) {materialFunction=&DeriveHDRfromLDR}.RemappedRange
    False=@n5
   ) {group=Control Switches, parameterName=Use HDR from LDR}) {materialFunction=&MF_FX_EmissiveAdjustments}
  n8 = OneMinus(None=@n2)
  n9 = ScalarParameter() {defaultValue=0.1, parameterName=Fine Streaks Opacity A}
  n10 = ScalarParameter() {defaultValue=0.3, parameterName=Fine Streaks Opacity B}
  n11 = TextureCoordinate()
  n12 = TextureSample(UVs=Panner(Coordinate=MaterialFunctionCall(XY=@n11) {materialFunction=&Swizzle}.YX) {speedX=-0.976}) {texture=/Game/Effects/Fort_Effects/Textures/Outlander/T_InTheZone_LightRays_01}
  n13 = TextureCoordinate() {uTiling=2, vTiling=2}
  n14 = TextureSample(UVs=Panner(Coordinate=MaterialFunctionCall(XY=@n13) {materialFunction=&Swizzle}.YX) {speedX=-0.7616}) {texture=&T_InTheZone_LightRays_01}
  n15 = Divide(A=Add(A=Power(Base=@n12.R) {constExponent=1.5}, B=Power(Base=@n14.G)))
  n16 = Panner(Coordinate=TextureCoordinate() {uTiling=0.2}) {bFractionalPart=true, speedX=-0.65}
  n17 = TextureCoordinate()
  n18 = Add(
   A=ConstantBiasScale(None=TextureSample(UVs=@n16) {texture=&T_NoisePack_01}.R) {bias=-0.5, scale=0.2}
   B=OneMinus(None=ComponentMask(None=@n17) {r=true})
  )
  n19 = Constant() {r=0.5}
  n20 = MaterialFunctionCall(Float2=TextureCoordinate()) {materialFunction=&BreakOutFloat2Components}
  n21 = Multiply(
   A=LinearInterpolate(
    A=LinearInterpolate(A=@n9, B=@n10, Alpha=@n15) {constA=1}
    B=ScalarParameter() {defaultValue=1, parameterName=Hard Streaks Opacity}
    Alpha=Power(Base=@n18) {constExponent=4}
   ) {constAlpha=0}
   B=SphereMask(
    A=AppendVector(A=@n19, B=@n20.G)
    B=@n19
    Radius=ScalarParameter() {defaultValue=0.5, parameterName=Mask Radius}
    Hardness=ScalarParameter() {defaultValue=0.9, parameterName=Mask Hardness}
   )
  )
  n22 = TextureCoordinate() {uTiling=0.5}
  n23 = ComponentMask(None=VectorParameter() {parameterName=SphereMaskB, defaultValue=<rgba 0.85 0.15 0 1>}.RGB) {g=true, r=true}
  n24 = Constant() {r=0.7}
  n25 = Constant() {r=0.3}
  n26 = Power(Base=OneMinus(None=SphereMask(A=@n22, B=@n23, Radius=@n24, Hardness=@n25))) {constExponent=5}
  n27 = Sine(None=ComponentMask(None=TextureCoordinate()) {r=true}) {period=2}
  n28 = Multiply(A=Multiply(A=Multiply(A=@n21, B=@n26), B=Multiply(A=@n27, B=@n27)), B=ScalarParameter() {defaultValue=10, parameterName=Opacity Multi}) {constB=10}
  n29 = Multiply(
   A=StaticSwitchParameter(True=Multiply(A=@n8, B=@n28), False=@n28) {group=Control Switches, parameterName=Multi Gradient Control into Opacity, sortPriority=4}
   B=MaterialFunctionCall(Distance=ScalarParameter() {defaultValue=64, parameterName=NearFadeDistance}) {materialFunction=&Near_Camera_Fade}.Result
  )
  n30 = ScalarParameter() {parameterName=Depth Fade Distance}
  n31 = StaticSwitchParameter(True=ShadingPathSwitch(Default=DepthFade(Opacity=@n29, FadeDistance=@n30), Mobile=@n29) {desc=Migrated from FeatureLevelSwitch}, False=@n29) {group=Control Switches, parameterName=Use Depth Fade, sortPriority=6}
  StaticSwitchParameter(True=LinearInterpolate(Alpha=Multiply(A=@n4.A, B=@n31)) {constA=1, constB=1.1}, False=Constant() {r=1}) {group=Control Switches, parameterName=Refraction?, sortPriority=5}
 ]
}
