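# Translucent, two-sided unlit material for the smoke trail of the RedRiding glider contrail, usable on Niagara ribbons and beam trails.
# Panned UVs, built from squared distance falloff around the object position and texture V, feed two multiplied sword-trail noise samples.
# That noise blends DarkCol to LightCol, shaped by a V gradient and particle color, for emissive.
# Opacity combines softened noise, U and V edge masks and particle alpha.
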Material /BRCosmetics/Athena/Effects/Materials/M_GliderContrail_RedRiding_SmokeTrail
graph {
 properties {
  bUsedWithBeamTrails=true
  bUsedWithNiagaraRibbons=true
  blendMode=BLEND_Translucent
  pixelDepthOffsetMode=PDOM_Legacy
  shadingModel=MSM_Unlit
  twoSided=true
  usedShadingModels=Unlit
 }
 nodes [
  n1 = MaterialFunctionCall(UV Channel=TextureCoordinate()) {materialFunction=&LinearGradient}
  n2 = Multiply(A=OneMinus(None=@n1.VGradient), B=@n1.VGradient)
  n3 = SphereMask(A=WorldPosition().XYZ, B=ObjectPositionWS(), Radius=Constant() {r=2200}, Hardness=Constant())
  n4 = Multiply(A=Multiply(A=@n3, B=@n3)) {constB=3}
  n5 = Multiply(A=ComponentMask(None=TextureCoordinate() {vTiling=1.5}) {g=true})
  n6 = Panner(Coordinate=Add(A=AppendVector(A=@n4, B=@n5)) {constB=0.3}) {bFractionalPart=true, speedX=3, speedY=-0.3}
  n7 = Multiply(
   A=TextureSample(UVs=Panner(Coordinate=@n6) {speedX=-0.5}) {texture=/Game/Weapons/FORT_Melee/Effects/LaserSwordFounders/T_LaserSwordTrailNoise}.G
   B=TextureSample(UVs=Panner(Coordinate=@n6) {speedX=-0.7}) {texture=&T_LaserSwordTrailNoise}.G
  )
  n8 = LinearInterpolate(
   A=VectorParameter() {parameterName=DarkCol, defaultValue=<rgba 0.5 0.5 0.5 0>}.RGB
   B=VectorParameter() {parameterName=LightCol, defaultValue=<rgba 1.5 1.5 1.5 0>}.RGB
   Alpha=@n7
  )
  MaterialFunctionCall(Emissive=Multiply(A=ParticleColor().RGB, B=Multiply(A=@n2, B=@n8))) {materialFunction=&MF_FX_EmissiveAdjustments}
  n9 = SphereMask(
   A=TextureCoordinate() {uTiling=0}
   B=Constant2Vector() {g=0.5}
   Radius=ScalarParameter() {defaultValue=0.45, parameterName=Radius}
   Hardness=ScalarParameter() {parameterName=Hardness}
  )
  n10 = SphereMask(
   A=TextureCoordinate() {vTiling=0}
   B=Constant2Vector() {r=0.5}
   Radius=ScalarParameter() {defaultValue=0.45, parameterName=Radius_1}
   Hardness=ScalarParameter() {parameterName=Hardness_1}
  )
  Multiply(A=Multiply(A=Power(Base=@n7) {constExponent=0.2}, B=Multiply(A=@n9, B=@n10)), B=ParticleColor().A)
 ]
}
