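# Translucent unlit mesh-particle material for the Space Shuttle glider contrail flame, tinted by particle color.
# Four panning caustic samples blend with an inverted Fresnel into a flame mask, which drives the V-gradient glow and, with edge Fresnel, sphere mask and particle alpha, the opacity.
# A desaturated Fresnel-indexed rainbow thin-film lookup modulates the emissive.
# Time-driven sine waves along normals and a circular wobble displace vertices, scaled by the ScaleOffset dynamic parameter.
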Material /BRCosmetics/Athena/Effects/Materials/M_GliderContrail_SpaceShuttle_Flame_SingleColor
graph {
 properties {
  bUsedWithMeshParticles=true
  bUsedWithNiagaraMeshParticles=true
  blendMode=BLEND_Translucent
  pixelDepthOffsetMode=PDOM_Legacy
  shadingModel=MSM_Unlit
  usedShadingModels=Unlit
 }
 nodes [
  n1 = TextureCoordinate()
  n2 = CameraVectorWS()
  n3 = Constant3Vector() {constant=<rgba 0 0 -0.5 0>}
  n4 = MaterialFunctionCall(
   Exponent=Multiply(A=ScalarParameter() {defaultValue=50, parameterName=FresnelExp}, B=ComponentMask(None=@n1) {g=true})
   Camera Vector=Normalize(VectorInput=Add(A=@n2, B=@n3))
  ) {materialFunction=&MF_Fresnel_FX}
  n5 = Panner(Coordinate=TextureCoordinate() {uTiling=2, vTiling=0.25}) {speedX=-0.5, speedY=1}
  n6 = Panner(Coordinate=TextureCoordinate() {uTiling=4, vTiling=0.5}) {speedX=1, speedY=2}
  n7 = Panner(Coordinate=TextureCoordinate() {vTiling=0.5}) {speedX=-0.5, speedY=1}
  n8 = Panner(Coordinate=TextureCoordinate() {uTiling=2}) {speedX=1, speedY=2}
  n9 = LinearInterpolate(
   A=LinearInterpolate(
    A=TextureSample(UVs=@n5) {texture=&T_ShieldCaustics}.G
    B=TextureSample(UVs=@n6) {texture=&T_ShieldCaustics}.G
   )
   B=LinearInterpolate(
    A=TextureSample(UVs=@n7) {texture=&T_ShieldCaustics}.G
    B=TextureSample(UVs=@n8) {texture=&T_ShieldCaustics}.G
   )
  )
  n10 = ScalarParameter() {defaultValue=0.5, parameterName=FlameAmount}
  n11 = Clamp(None=Add(A=Multiply(A=@n4.Fresnel Inverted), B=Power(Base=@n9, Exp=@n10)))
  n12 = OneMinus(None=@n11)
  n13 = Multiply(
   A=MaterialFunctionCall(Exponent=ScalarParameter() {defaultValue=1.5, parameterName=EdgeTrans}) {materialFunction=&MF_Fresnel_FX}.Fresnel
   B=Add(A=Power(Base=@n11) {constExponent=15}, B=Power(Base=@n12) {constExponent=8}) {constB=0}
  )
  n14 = SphereMask(
   A=TextureCoordinate() {uTiling=0}
   B=Constant2Vector() {g=0.525}
   Radius=ScalarParameter() {defaultValue=0.5, parameterName=RadiusEdge}
   Hardness=ScalarParameter() {parameterName=HardnessEdge}
  )
  Clamp(None=Multiply(A=Multiply(A=@n13, B=@n14), B=ParticleColor().A))
  n15 = ScalarParameter() {defaultValue=0.35, group=ThinFilm, parameterName=ThinFilm_Scale}
  n16 = VertexNormalWS()
  n17 = Fresnel(ExponentIn=ScalarParameter() {defaultValue=0.5, group=ThinFilm, parameterName=ThinFilm_EX}, Normal=@n16) {baseReflectFraction=0}
  n18 = Multiply(
   A=TextureSample(UVs=Multiply(A=@n15, B=@n17)) {texture=/Game/Global/Textures/Default/Masks_and_Gradients/FortniteRainbow}.RGB
   B=ScalarParameter() {defaultValue=1, group=ThinFilm, parameterName=ThinFilm_Intensity}
  )
  n19 = Multiply(
   A=OneMinus(None=ScalarParameter() {defaultValue=1, parameterName=Roughness})
   B=ScalarParameter() {defaultValue=1, group=ThinFilm, parameterName=ThinFilm_RoughnessScale}
  )
  n20 = ScalarParameter() {defaultValue=1.2164, group=ThinFilm, parameterName=ThinFilm_Visibility}
  n21 = LinearInterpolate(A=@n18, B=@n18, Alpha=Multiply(A=Clamp(None=@n19), B=Fresnel(ExponentIn=@n20, Normal=@n16) {baseReflectFraction=0}))
  n22 = ScalarParameter() {defaultValue=0.25, parameterName=Desat_Fraction}
  n23 = TextureCoordinate()
  n24 = Constant() {r=2.5}
  n25 = OneMinus(None=TextureCoordinate())
  n26 = Add(
   A=Multiply(
    A=ComponentMask(None=@n23) {g=true}
    B=MaterialFunctionCall(Exponent=@n24) {materialFunction=&MF_Fresnel_FX}.Fresnel Inverted
   )
   B=Multiply(A=ComponentMask(None=@n25) {g=true}, B=Multiply(A=@n11) {constB=10})
  )
  n27 = ParticleColor()
  MaterialFunctionCall(Emissive=Multiply(A=Desaturation(None=@n21, Fraction=@n22), B=Multiply(A=@n26, B=@n27.RGB))) {materialFunction=&MF_FX_EmissiveAdjustments}
  n28 = Multiply(A=Frac(None=Time())) {constB=50}
  n29 = OneMinus(None=ComponentMask(None=TextureCoordinate()) {g=true})
  n30 = Multiply(A=Multiply(A=@n29, B=@n29)) {constB=-2}
  n31 = Multiply(A=Sine(None=Add(A=@n28, B=@n30))) {desc=Amplitude}
  n32 = Add(A=Time(), B=ComponentMask(None=TextureCoordinate()) {g=true})
  n33 = TextureCoordinate()
  n34 = Add(A=Multiply(A=Time()) {constB=40}, B=Multiply(A=ComponentMask(None=@n33) {g=true}) {constB=2})
  n35 = Add(
   A=Multiply(A=Sine(None=@n32) {period=1.31}) {constB=4, desc=Amplitude}
   B=Multiply(A=Sine(None=@n34) {period=1.31}) {constB=4, desc=Amplitude}
  )
  n36 = Multiply(A=VertexNormalWS(), B=Max(A=Add(A=@n31, B=@n35)) {constB=-2.5})
  n37 = Time()
  n38 = AppendVector(A=ComponentMask(None=WorldPosition().XYZ) {b=true}, B=Constant2Vector())
  n39 = Add(A=Multiply(A=Frac(None=@n37)) {constB=50}, B=Frac(None=Multiply(A=@n38) {constB=-0.02}))
  n40 = Constant() {r=1}
  n41 = Multiply(A=Sine(None=@n39), B=@n40) {desc=Amplitude}
  n42 = Multiply(A=Cosine(None=@n39), B=@n40)
  n43 = OneMinus(None=ComponentMask(None=TextureCoordinate()) {g=true})
  n44 = Multiply(
   A=MaterialFunctionCall(X=ComponentMask(None=@n41) {r=true}, Y=ComponentMask(None=@n42) {r=true}, Z=Constant()) {materialFunction=&MakeFloat3}.Result
   B=Multiply(A=@n43, B=@n43)
  )
  n45 = ComponentMask(None=TextureCoordinate()) {g=true}
  Multiply(
   A=Multiply(A=DynamicParameter() {paramNames [ScaleOffset,Param2,Param3,Param4]}.ScaleOffset, B=Add(A=@n36, B=@n44) {constB=0})
   B=Clamp(None=Add(A=@n45) {constB=0.5})
  )
  ParticleDirection()
 ]
}
