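# Translucent, unlit, two-sided mesh-particle material for wing smoke in the Luchador glider deploy effect.
# A camera-vector triplanar dust texture, randomly offset per particle off mobile, is eroded by particle red, then masked by UV edge falloff and squared vertex green.
# That mask drives BaseColor-tinted emissive scaled by particle green, and opacity using particle alpha, Erode and depth fade.
# Dynamic parameters also control a vector-noise curl offset, replaced by a constant on mobile.
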Material /BRCosmetics/Athena/Effects/Materials/M_GliderDeploy_Luchador_WingSmoke
graph {
 properties {
  bEnableResponsiveAA=true
  bUsedWithMeshParticles=true
  blendMode=BLEND_Translucent
  pixelDepthOffsetMode=PDOM_Legacy
  shadingModel=MSM_Unlit
  twoSided=true
  usedShadingModels=Unlit
 }
 nodes [
  n1 = VectorParameter() {parameterName=BaseColor, defaultValue=<rgba 0.3231 0.0307 0.0284 1>}
  n2 = MaterialFunctionCall(X=ParticleRandom(), Y=ParticleRandom(), Z=ParticleRandom()) {materialFunction=&MakeFloat3}
  n3 = VectorParameter() {parameterName=Offset}
  n4 = MaterialFunctionCall!(
   Dynamic Param=ParticleColor().R
   Alpha=MaterialFunctionCall!(
    CameraVector=CameraVectorWS()
    Texture=TextureObject() {texture=/Game/Global/Textures/Default/PointSplatter/T_Tiling_Dust}
    Tiling=VectorParameter() {parameterName=Tiling, defaultValue=<rgba 1 1 1 0>}.RGB
    Offset=ShadingPathSwitch(Default=@n2.Result, Mobile=@n3.RGB) {desc=Migrated from FeatureLevelSwitch}
    AxisFadeContrast
   ) {materialFunction=/Engine/Functions/Engine_MaterialFunctions01/Texturing/TriplanarCameraVector}.Result
  ) {materialFunction=/Game/Effects/Fort_Effects/Materials/_Functions/Erode_Alpha}
  n5 = TextureCoordinate()
  n6 = ComponentMask(None=@n5) {r=true}
  n7 = Multiply(A=@n6, B=OneMinus(None=@n6))
  n8 = ComponentMask(None=@n5) {g=true}
  n9 = Multiply(A=@n8, B=OneMinus(None=@n8))
  n10 = VertexColor()
  n11 = StaticSwitchParameter(True=OneMinus(None=@n10.G), False=@n10.G) {parameterName=Invert Vertex Color}
  n12 = Multiply(A=Power(Base=Clamp(None=@n4.Result)) {constExponent=4}, B=Multiply(A=Multiply(A=@n7, B=@n9), B=Multiply(A=@n11, B=@n11)))
  MaterialFunctionCall(Emissive=Multiply(A=ParticleColor().G, B=Multiply(A=@n1.RGB, B=@n12))) {materialFunction=&MF_FX_EmissiveAdjustments}
  n13 = DynamicParameter() {defaultValue=<rgba 1 50 15 1>, paramNames [Curl_Speed, Curl_Scale, Curl_Intensity, Erode]}
  Multiply(A=@n13.Curl_Speed, B=Time())
  n14 = Multiply(A=@n12, B=ParticleColor().A)
  n15 = DynamicParameter() {defaultValue=<rgba 1 50 15 1>, paramNames [Curl_Speed, Curl_Scale, Curl_Intensity, Erode]}
  Multiply(
   A=DepthFade(
    Opacity=MaterialFunctionCall(Dynamic Param=@n14, Alpha=@n15.Erode) {materialFunction=&Erode_Alpha}.Result
    FadeDistance=ScalarParameter() {defaultValue=10, parameterName=FadeDistance}
   )
   B=ScalarParameter() {defaultValue=0.4, parameterName=MultOpacity}
  )
  n16 = WorldPosition()
  n17 = ActorPositionWS()
  n18 = Divide(
   A=Subtract(A=Subtract(A=@n16.XYZ, B=@n17), B=@n13.Curl_Speed)
   B=DynamicParameter() {defaultValue=<rgba 1 50 15 1>, paramNames [Curl_Speed, Curl_Scale, Curl_Intensity, Erode]}.Curl_Scale
  )
  ShadingPathSwitch(
   Default=Multiply(
    A=DynamicParameter() {defaultValue=<rgba 1 50 15 1>, paramNames [Curl_Speed, Curl_Scale, Curl_Intensity, Erode]}.Curl_Intensity
    B=VectorNoise(World Position=@n18) {noiseFunction=VNF_VectorALU}
   )
   Mobile=Constant() {r=1}
  ) {desc=Migrated from FeatureLevelSwitch}
 ]
}
