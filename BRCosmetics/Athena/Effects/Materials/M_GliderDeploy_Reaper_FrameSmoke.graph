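# Translucent, unlit, two-sided mesh-particle material for smoke on the Reaper glider deploy effect.
# Three panning smoke/noise textures, U-offset by particle red, combine with inverted Fresnel and squared inverse vertex green into a mask for erodable, depth-faded opacity.
# That mask also shapes emissive: base color plus Fresnel-driven rainbow thin-film, scaled by particle green.
# Dynamic parameters control erosion and actor-relative vector-noise displacement, which mobile replaces with a constant.
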Material /BRCosmetics/Athena/Effects/Materials/M_GliderDeploy_Reaper_FrameSmoke
graph {
 properties {
  bUsedWithMeshParticles=true
  blendMode=BLEND_Translucent
  pixelDepthOffsetMode=PDOM_Legacy
  shadingModel=MSM_Unlit
  twoSided=true
  usedShadingModels=Unlit
 }
 nodes [
  n1 = Power(Base=Fresnel(ExponentIn=ScalarParameter() {defaultValue=0.05, parameterName=SmokeFrensnel})) {constExponent=4}
  n2 = ParticleColor()
  n3 = TextureCoordinate() {uTiling=8}
  n4 = ComponentMask(None=@n3) {r=true}
  n5 = Panner(Coordinate=AppendVector(A=Add(A=@n2.R, B=@n4), B=ComponentMask(None=@n3) {g=true})) {speedX=1}
  n6 = ParticleColor()
  n7 = TextureCoordinate() {uTiling=4}
  n8 = ComponentMask(None=@n7) {r=true}
  n9 = Panner(Coordinate=AppendVector(A=Add(A=@n6.R, B=@n8), B=ComponentMask(None=@n7) {g=true})) {speedY=0.5}
  n10 = TextureCoordinate() {uTiling=6, vTiling=0.5}
  n11 = AppendVector(A=Add(A=ParticleColor().R, B=ComponentMask(None=@n10) {r=true}), B=ComponentMask(None=@n10) {g=true})
  n12 = Add(
   A=Add(
    A=TextureSample(UVs=@n5) {samplerType=SAMPLERTYPE_Grayscale, texture=&T_Smoke_Trail_Soft}.G
    B=TextureSample(UVs=@n9) {texture=/Game/Weapons/FORT_Melee/Effects/LaserSwordFounders/T_LaserSwordTrailNoise}.G
   )
   B=TextureSample(UVs=Panner(Coordinate=@n11) {speedY=0.5}) {texture=&T_Inky_Smoke_Tile}.G
  )
  n13 = OneMinus(None=VertexColor().G)
  n14 = Multiply(A=Multiply(A=OneMinus(None=@n1), B=Power(Base=@n12) {constExponent=1.25}), B=Multiply(A=@n13, B=@n13))
  n15 = ParticleColor()
  DepthFade(
   Opacity=MaterialFunctionCall!(
    Dynamic Param=Multiply(A=@n14, B=@n15.A)
    Alpha=DynamicParameter() {defaultValue=<rgba 1 50 15 0.1>, paramNames [Curl_Speed, Curl_Scale, Curl_Intensity, Erode]}.Erode
   ) {materialFunction=/Game/Effects/Fort_Effects/Materials/_Functions/Erode_Alpha}.Result
   FadeDistance=ScalarParameter() {defaultValue=10, parameterName=FadeDistance}
  )
  n16 = VectorParameter() {parameterName=BaseColor, defaultValue=<rgba 0.0058 0.0182 0.0382 1>}
  n17 = ScalarParameter() {defaultValue=0.5, group=ThinFilm, parameterName=ThinFilm_EX}
  n18 = VertexNormalWS()
  n19 = TextureSample(UVs=Multiply(
    A=ScalarParameter() {defaultValue=0.7, group=ThinFilm, parameterName=ThinFilm_Scale}
    B=Fresnel(ExponentIn=@n17, Normal=@n18) {baseReflectFraction=0}
   )) {texture=/Game/Global/Textures/Default/Masks_and_Gradients/FortniteRainbow}
  n20 = ScalarParameter() {defaultValue=1, group=ThinFilm, parameterName=ThinFilm_Intensity}
  n21 = Multiply(
   A=OneMinus(None=ScalarParameter() {defaultValue=0.7, parameterName=Roughness})
   B=ScalarParameter() {defaultValue=1, group=ThinFilm, parameterName=ThinFilm_RoughnessScale}
  )
  n22 = ScalarParameter() {defaultValue=1.2164, group=ThinFilm, parameterName=ThinFilm_Visibility}
  n23 = LinearInterpolate(
   A=@n16.RGB
   B=Add(A=@n16.RGB, B=Multiply(A=@n19.RGB, B=@n20))
   Alpha=Multiply(A=Clamp(None=@n21), B=Fresnel(ExponentIn=@n22, Normal=@n18) {baseReflectFraction=0})
  )
  MaterialFunctionCall(Emissive=Multiply(A=ParticleColor().G, B=Multiply(A=@n23, B=@n14))) {materialFunction=&MF_FX_EmissiveAdjustments}
  n24 = WorldPosition()
  n25 = ActorPositionWS()
  n26 = Divide(
   A=Subtract(
    A=Subtract(A=@n24.XYZ, B=@n25)
    B=DynamicParameter() {defaultValue=<rgba 1 50 15 0.1>, paramNames [Curl_Speed, Curl_Scale, Curl_Intensity, Erode]}.Curl_Speed
   )
   B=DynamicParameter() {defaultValue=<rgba 1 50 15 0.1>, paramNames [Curl_Speed, Curl_Scale, Curl_Intensity, Erode]}.Curl_Scale
  )
  ShadingPathSwitch(
   Default=Multiply(
    A=DynamicParameter() {defaultValue=<rgba 1 50 15 0.1>, paramNames [Curl_Speed, Curl_Scale, Curl_Intensity, Erode]}.Curl_Intensity
    B=VectorNoise(World Position=@n26) {noiseFunction=VNF_VectorALU}
   )
   Mobile=Constant() {r=1}
  ) {desc=Migrated from FeatureLevelSwitch}
 ]
}
