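# Translucent, unlit, two-sided mesh-particle material for the BoneWasp glider's exhaust effect.
# Emissive adds a flickering, panned-noise glow to a vertex-red color gradient whose exponents pulse from layered sines, tinted by particle color.
# Opacity combines particle alpha, time of day, depth fade and optional two-sided fresnel edge softening, with cheaper low-quality and mobile paths.
# A vertex-normal offset, masked by vertex green and scaled by dynamic TailSize, shapes the tail.
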Material /BRCosmetics/Athena/Effects/Materials/M_Glider_BoneWasp_Exhaust
graph {
 properties {
  bUsedWithMeshParticles=true
  blendMode=BLEND_Translucent
  pixelDepthOffsetMode=PDOM_Legacy
  previewMesh=/BRCosmetics/Gadgets/Assets/VinderTech_GliderChute/Glider_BoneWasp/FX/S_BoneWasp_Center
  shadingModel=MSM_Unlit
  twoSided=true
  usedShadingModels=Unlit
 }
 nodes [
  n1 = Constant() {r=1}
  n2 = ScalarParameter() {defaultValue=1, parameterName=OpacityNight}
  n3 = MaterialFunctionCall(Color_Morning=@n1, Color_Day=@n1, Color_Evening=@n2, Color_Night=@n2) {materialFunction=&Color_Time_of_Day}
  n4 = ParticleColor()
  n5 = Power(Base=VertexColor().R) {constExponent=3}
  n6 = ScalarParameter() {defaultValue=2, parameterName=FresnelOuter}
  n7 = ScalarParameter() {parameterName=FresnelInner}
  n8 = Fresnel(ExponentIn=@n6, BaseReflectFractionIn=@n7, Normal=Multiply(A=VertexNormalWS()) {constB=-1})
  n9 = Fresnel(ExponentIn=@n6, BaseReflectFractionIn=@n7)
  n10 = TwoSidedSign()
  n11 = Constant() {r=1}
  n12 = QualitySwitch(
   Default=ShadingPathSwitch(Default=LinearInterpolate(A=@n8, B=@n9, Alpha=@n10), Mobile=@n11) {desc=Migrated from FeatureLevelSwitch}
   Low=@n11
   Medium=@n11
  )
  n13 = Multiply(
   A=@n4.A
   B=StaticSwitchParameter(True=Multiply(A=@n5, B=@n12), False=Constant() {r=1}) {defaultValue=true, parameterName=UseFresneltoSoftEdge}
  )
  DepthFade(
   Opacity=Multiply(A=ScalarParameter() {defaultValue=1, parameterName=OpacityOverall}, B=Multiply(A=@n3.Result, B=@n13))
   FadeDistance=ScalarParameter() {defaultValue=20, parameterName=DepthFadeDist}
  )
  n14 = Sine(None=Multiply(A=Time(), B=ScalarParameter() {defaultValue=8, parameterName=FlickerSpd}))
  n15 = Constant() {r=0.5}
  n16 = LinearInterpolate(A=ScalarParameter() {defaultValue=0.4, parameterName=FlickerLow}, Alpha=Add(A=Multiply(A=@n14, B=@n15), B=@n15))
  n17 = Multiply(
   A=TextureCoordinate()
   B=AppendVector(
    A=ScalarParameter() {defaultValue=1.2, group=EmisColor, parameterName=U_Tile}
    B=ScalarParameter() {defaultValue=1, group=EmisColor, parameterName=V_Tile}
   )
  )
  n18 = LinearInterpolate(Alpha=TextureSample(UVs=Panner(Coordinate=@n17) {speedX=0.1, speedY=-1}) {texture=/Game/Global/Textures/Default/Masks_and_Gradients/T_BlasterBeamDistort}.R) {constA=0.3, constB=3}
  n19 = VectorParameter() {parameterName=ColorHigh, defaultValue=<rgba 1 1 1 0>}
  n20 = VectorParameter() {parameterName=ColorMid, defaultValue=<rgba 1 1 1 0>}
  n21 = Saturate(None=Power(Base=VertexColor().R, Exp=ScalarParameter() {defaultValue=3, parameterName=ColorLerpExp}) {constExponent=3})
  n22 = VectorParameter() {parameterName=ColorLow, defaultValue=<rgba 1 1 1 0>}
  n23 = VertexColor()
  n24 = Multiply(A=Time())
  n25 = Multiply(A=Sine(None=@n24) {period=0.155}, B=Sine(None=@n24) {period=0.2})
  n26 = Sine(None=@n24) {period=0.43}
  n27 = Saturate(None=Multiply(A=Sine(None=@n24) {period=0.075}, B=Multiply(A=@n25, B=@n26)))
  n28 = LinearInterpolate(Alpha=@n27) {constA=1.5, constB=2.5}
  n29 = Constant() {r=2}
  n30 = Power(Base=@n23.R, Exp=QualitySwitch(Default=ShadingPathSwitch(Default=@n28, Mobile=@n29) {desc=Migrated from FeatureLevelSwitch}, Low=@n29))
  n31 = LinearInterpolate(Alpha=@n27) {constA=2, constB=3.5}
  n32 = Constant() {r=3}
  n33 = Multiply(
   A=QualitySwitch(Default=ShadingPathSwitch(Default=@n31, Mobile=@n32) {desc=Migrated from FeatureLevelSwitch}, Low=@n32)
   B=DynamicParameter() {paramNames [TailSize, GradientContrast, RingShapeSpeed, Param4]}.GradientContrast
  )
  n34 = Add(
   A=Multiply(A=Multiply(A=@n16, B=@n18), B=LinearInterpolate(A=@n19.RGB, B=@n20.RGB, Alpha=@n21)) {constB=10}
   B=Multiply(A=LinearInterpolate(A=@n22.RGB, B=@n20.RGB, Alpha=@n30), B=Power(Base=@n23.R, Exp=@n33) {constExponent=3})
  )
  MaterialFunctionCall(Emissive=Desaturation(None=Multiply(A=@n34, B=@n4.RGB), Fraction=ScalarParameter() {parameterName=Desaturation})) {materialFunction=&MF_FX_EmissiveAdjustments}
  n35 = VertexNormalWS()
  Multiply(
   A=Multiply(A=Multiply(A=@n35) {constB=10}, B=VertexColor().G)
   B=LinearInterpolate(Alpha=DynamicParameter() {paramNames [TailSize, GradientContrast, RingShapeSpeed, Param4]}.TailSize) {constA=1, constB=0}
  )
 ]
}
