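# Translucent unlit, two-sided material for a glider exhaust plume, usable on skeletal meshes or mesh particles via a static switch.
# Panned distortion noise and vertex-red gradients, with sine-based flicker, blend three colors into emissive, tinted by particle color or a parameter.
# Vertex green drives tail-size offset along the normal or a local axis, optionally stretch-masked.
# Opacity uses time-of-day values and Fresnel soft edges; noise-driven refraction and other costly features simplify on mobile and low quality.
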Material /BRCosmetics/Athena/Effects/Materials/M_Glider_SciOps_Exhaust
graph {
 properties {
  bUsedWithMeshDeformer=true
  bUsedWithMeshParticles=true
  bUsedWithSkeletalMesh=true
  blendMode=BLEND_Translucent
  pixelDepthOffsetMode=PDOM_Legacy
  refractionMethod=RM_IndexOfRefraction
  shadingModel=MSM_Unlit
  twoSided=true
  usedShadingModels=Unlit
 }
 nodes [
  Constant3Vector() {constant=<rgba 0 0 1 0>}
  n1 = TextureCoordinate() {uTiling=0.1, vTiling=2.5}
  n2 = Power(Base=TextureSample(UVs=Panner(Coordinate=@n1) {speedY=-4}) {texture=/Game/Global/Textures/Default/Masks_and_Gradients/T_BlasterBeamDistort}.R) {constExponent=2.3}
  n3 = TextureCoordinate() {uTiling=0.1, vTiling=2}
  n4 = Power(Base=TextureSample(UVs=Panner(Coordinate=@n3) {speedY=-1}) {texture=&T_BlasterBeamDistort}.R) {constExponent=2.3}
  n5 = StaticSwitchParameter(
   True=DynamicParameter() {paramNames [TailSize, GradientContrast, RingShapeSpeed, Param4]}.RingShapeSpeed
   False=ScalarParameter() {defaultValue=1, parameterName=RingShapeSpeed}
  ) {defaultValue=true, parameterName=Use with Particles}
  n6 = LinearInterpolate(Alpha=@n5) {constA=1}
  n7 = ParticleColor()
  n8 = Constant() {r=1}
  n9 = Multiply(
   A=ShadingPathSwitch(Default=LinearInterpolate(A=@n2, B=@n4, Alpha=@n5) {constAlpha=0}, Mobile=Multiply(A=@n2, B=@n6)) {desc=Migrated from FeatureLevelSwitch}
   B=Multiply(
    A=ScalarParameter() {defaultValue=2.5, parameterName=RingShapeIntensity}
    B=StaticSwitchParameter(True=@n7.A, False=@n8) {defaultValue=true, parameterName=Use with Particles}
   )
  ) {constB=2.5}
  n10 = Multiply(
   A=TextureCoordinate()
   B=AppendVector(
    A=ScalarParameter() {defaultValue=2.5, group=EmisColor, parameterName=U_Tile}
    B=ScalarParameter() {defaultValue=0.6, group=EmisColor, parameterName=V_Tile}
   )
  )
  n11 = LinearInterpolate(Alpha=TextureSample(UVs=Panner(Coordinate=@n10) {speedX=0.1, speedY=-8}) {texture=&T_BlasterBeamDistort}.R) {constA=0.3, constB=3}
  n12 = VectorParameter() {parameterName=ColorHigh, defaultValue=<rgba 1 1 1 0>}
  n13 = VectorParameter() {parameterName=ColorMid, defaultValue=<rgba 1 1 1 0>}
  n14 = Saturate(None=Power(Base=VertexColor().R, Exp=ScalarParameter() {defaultValue=3, parameterName=ColorLerpExp}) {constExponent=3})
  n15 = VectorParameter() {parameterName=ColorLow, defaultValue=<rgba 1 1 1 0>}
  n16 = VertexColor()
  n17 = Multiply(A=Time())
  n18 = Multiply(A=Sine(None=@n17) {period=0.155}, B=Sine(None=@n17) {period=0.2})
  n19 = Sine(None=@n17) {period=0.43}
  n20 = Saturate(None=Multiply(A=Sine(None=@n17) {period=0.075}, B=Multiply(A=@n18, B=@n19)))
  n21 = LinearInterpolate(Alpha=@n20) {constA=1.5, constB=2.5}
  n22 = Constant() {r=2}
  n23 = Power(Base=@n16.R, Exp=QualitySwitch(Default=ShadingPathSwitch(Default=@n21, Mobile=@n22) {desc=Migrated from FeatureLevelSwitch}, Low=@n22))
  n24 = LinearInterpolate(Alpha=@n20) {constA=2, constB=3.5}
  n25 = Constant() {r=3}
  n26 = Multiply(
   A=StaticSwitchParameter(
    True=DynamicParameter() {paramNames [TailSize, GradientContrast, RingShapeSpeed, Param4]}.GradientContrast
    False=ScalarParameter() {defaultValue=1, parameterName=GradientContrast}
   ) {defaultValue=true, parameterName=Use with Particles}
   B=QualitySwitch(Default=ShadingPathSwitch(Default=@n24, Mobile=@n25) {desc=Migrated from FeatureLevelSwitch}, Low=@n25)
  )
  n27 = Add(
   A=Multiply(A=Add(A=@n9, B=@n11), B=LinearInterpolate(A=@n12.RGB, B=@n13.RGB, Alpha=@n14)) {constB=10}
   B=Multiply(A=LinearInterpolate(A=@n15.RGB, B=@n13.RGB, Alpha=@n23), B=Power(Base=@n16.R, Exp=@n26) {constExponent=3})
  )
  n28 = ParticleColor()
  n29 = StaticSwitchParameter(True=@n28.RGB, False=VectorParameter() {parameterName=Additive Color, defaultValue=<rgba 1 1 1 1>}.RGB) {defaultValue=true, parameterName=Use with Particles}
  MaterialFunctionCall(Emissive=Desaturation(None=Multiply(A=@n27, B=@n29), Fraction=ScalarParameter() {parameterName=Desaturation})) {materialFunction=&MF_FX_EmissiveAdjustments}
  n30 = Transform(None=VectorParameter() {parameterName=Local Normal for Tail, defaultValue=<rgba 0 1 0 0>}.RGB) {transformSourceType=TRANSFORMSOURCE_Local}
  n31 = VertexNormalWS()
  n32 = Multiply(A=Multiply(A=StaticSwitchParameter(True=@n30, False=@n31) {parameterName=Use Local Axis Normal}) {constB=10}, B=VertexColor().G)
  n33 = Multiply(A=ComponentMask(None=TextureCoordinate()) {g=true}) {constB=15}
  n34 = Saturate(None=Multiply(A=@n33, B=@n33))
  n35 = DynamicParameter() {paramNames [TailSize, GradientContrast, RingShapeSpeed, Param4]}
  n36 = ScalarParameter() {defaultValue=1, parameterName=TailSize}
  Multiply(
   A=LinearInterpolate(A=@n32, B=Multiply(A=@n32, B=@n34), Alpha=ScalarParameter() {parameterName=Use Stretch Mask})
   B=StaticSwitchParameter(True=LinearInterpolate(Alpha=@n35.TailSize) {constA=1, constB=0}, False=Multiply(A=@n36) {constB=-1}) {defaultValue=true, parameterName=Use with Particles}
  )
  n37 = ScalarParameter() {defaultValue=1, parameterName=OpacityDay}
  n38 = ScalarParameter() {defaultValue=1, parameterName=OpacityNight}
  n39 = StaticSwitchParameter(True=@n28.A, False=Constant() {r=1}) {defaultValue=true, parameterName=Use with Particles}
  n40 = VertexColor()
  n41 = ScalarParameter() {defaultValue=3, parameterName=VertexEdgePower}
  n42 = ScalarParameter() {defaultValue=2, parameterName=FresnelOuter}
  n43 = ScalarParameter() {parameterName=FresnelInner}
  n44 = Fresnel(ExponentIn=@n42, BaseReflectFractionIn=@n43, Normal=Multiply(A=VertexNormalWS()) {constB=-1})
  n45 = OneMinus(None=@n44)
  n46 = Fresnel(ExponentIn=@n42, BaseReflectFractionIn=@n43)
  n47 = OneMinus(None=@n46)
  n48 = Constant() {r=1}
  n49 = ShadingPathSwitch(
   Default=LinearInterpolate(
    A=StaticSwitchParameter(True=@n45, False=@n44) {parameterName=Invert SoftEdge Fresnel}
    B=StaticSwitchParameter(True=@n47, False=@n46) {parameterName=Invert SoftEdge Fresnel}
    Alpha=TwoSidedSign()
   )
   Mobile=@n48
  ) {desc=Migrated from FeatureLevelSwitch}
  n50 = StaticSwitchParameter(
   True=Multiply(A=Power(Base=@n40.R, Exp=@n41) {constExponent=3}, B=QualitySwitch(Default=@n49, Low=@n48, Medium=@n48))
   False=Constant() {r=1}
  ) {defaultValue=true, parameterName=UseFresneltoSoftEdge}
  Multiply(
   A=ScalarParameter() {defaultValue=1, parameterName=OpacityOverall}
   B=Multiply(
    A=MaterialFunctionCall(Color_Morning=@n37, Color_Day=@n37, Color_Evening=@n38, Color_Night=@n38) {materialFunction=&Color_Time_of_Day}.Result
    B=Multiply(A=@n39, B=@n50)
   )
  )
  n51 = LinearInterpolate(
   B=LinearInterpolate(Alpha=ScalarParameter() {parameterName=Separate Trans Check}) {constA=1.1, constB=1.035}
   Alpha=ScalarParameter() {defaultValue=1, parameterName=RefIntensity}
  ) {constA=1, constB=1.1}
  n52 = VertexColor()
  n53 = ScalarParameter() {defaultValue=2, parameterName=RefFadeOut}
  n54 = ScalarParameter() {defaultValue=2.5, group=Normal, parameterName=U_Tile_N}
  n55 = ScalarParameter() {defaultValue=0.6, group=Normal, parameterName=V_Tile_N}
  n56 = Panner(Coordinate=Multiply(A=TextureCoordinate(), B=AppendVector(A=@n54, B=@n55))) {speedY=-10}
  n57 = Multiply(
   A=Multiply(
    A=Power(Base=@n52.R, Exp=@n53)
    B=TextureSampleParameter2D(UVs=@n56) {parameterName=RefTexture, texture=/Game/Effects/Fort_Effects/Textures/Tile/T_TileNoise_03}.R
   )
   B=DotProduct(A=CameraVectorWS(), B=VertexNormalWS())
  )
  n58 = Constant() {r=1}
  QualitySwitch(
   Default=ShadingPathSwitch(Default=LinearInterpolate(B=@n51, Alpha=@n57) {constA=1, constB=1.1}, Mobile=@n58) {desc=Migrated from FeatureLevelSwitch}
   Low=@n58
   Medium=@n58
  )
 ]
}
