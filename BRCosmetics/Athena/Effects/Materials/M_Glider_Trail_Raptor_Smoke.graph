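# Unlit, two-sided translucent material for the Raptor glider's smoke trail, built for mesh particles.
# A panning normal map distorts the UVs of scrolling tile-pattern and blurred-noise textures, which multiply into streaky smoke.
# Sphere masks on U and V soften the edges, and particle alpha scales the result.
# Particle color feeds an emissive adjustment function, while depth-fade and camera-to-normal fade branches process the opacity.
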Material /BRCosmetics/Athena/Effects/Materials/M_Glider_Trail_Raptor_Smoke
graph {
 properties {
  bAutomaticallySetUsageInEditor=false
  bUsedWithMeshParticles=true
  bUsedWithStaticMesh=false
  blendMode=BLEND_Translucent
  pixelDepthOffsetMode=PDOM_Legacy
  shadingModel=MSM_Unlit
  twoSided=true
  usedShadingModels=Unlit
 }
 nodes [
  n1 = ParticleColor()
  n2 = TextureCoordinate() {uTiling=0.5, vTiling=0.6}
  n3 = ComponentMask(None=TextureSample(UVs=Panner(Coordinate=@n2) {speedX=0.25, speedY=0.52}) {samplerType=SAMPLERTYPE_Normal, texture=&T_Weap_Distort_N_02}.RGB) {g=true, r=true}
  n4 = Panner(Coordinate=Add(A=TextureCoordinate() {vTiling=0.1}, B=Multiply(A=@n3) {constB=0.08})) {speedY=0.55}
  n5 = MaterialFunctionCall(Float2=TextureCoordinate() {uTiling=2, vTiling=0.5}) {materialFunction=&BreakOutFloat2Components}
  n6 = Power(Base=@n5.G)
  n7 = Panner(Coordinate=Add(A=AppendVector(A=@n5.R, B=@n6), B=Multiply(A=@n3) {constB=0.14})) {speedY=1.1}
  n8 = Multiply(A=Multiply(
    A=TextureSample(UVs=@n4) {texture=/Game/Effects/Fort_Effects/Textures/T_Tile_Pattern_01}.R
    B=TextureSample(UVs=@n7) {samplerType=SAMPLERTYPE_Grayscale, texture=&LowResBlurredNoise}.R
   )) {constB=3.5}
  n9 = Constant() {r=0.5}
  n10 = SphereMask(
   A=MaterialFunctionCall(Float2=TextureCoordinate()) {materialFunction=&BreakOutFloat2Components}.R
   B=@n9
   Radius=@n9
  ) {hardnessPercent=0.25}
  n11 = SphereMask(
   A=MaterialFunctionCall(Float2=TextureCoordinate()) {materialFunction=&BreakOutFloat2Components}.G
   B=@n9
   Radius=@n9
  ) {hardnessPercent=0.25}
  n12 = Multiply(A=@n1.A, B=Multiply(A=Power(Base=@n8) {constExponent=1.2}, B=Multiply(A=@n10, B=@n11)))
  DepthFade(Opacity=@n12) {fadeDistanceDefault=6}
  MaterialFunctionCall(Emissive=@n1.RGB) {materialFunction=&MF_FX_EmissiveAdjustments}
  n13 = CameraVectorWS()
  n14 = VertexNormalWS()
  Multiply(A=@n12, B=Power(Base=DotProduct(A=@n13, B=@n14)))
 ]
}
