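# Translucent, unlit, two-sided mesh-particle material for the smoke trail of the Raptor glider.
# Panning noise, scaled by a dynamic parameter and U, combines with SplineThicken to offset vertices.
# Panning smoke textures and a normal-distorted trail mask build the emissive, blended in from a constant color along U.
# Opacity multiplies that mask by a squared sphere-mask falloff and a second dynamic parameter, then applies depth fade and particle alpha.
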Material /BRCosmetics/Athena/Effects/Materials/M_Glider_Trail_Raptor_Smoke_2
graph {
 properties {
  bAutomaticallySetUsageInEditor=false
  bUsedWithMeshParticles=true
  bUsedWithStaticMesh=false
  blendMode=BLEND_Translucent
  pixelDepthOffsetMode=PDOM_Legacy
  shadingModel=MSM_Unlit
  twoSided=true
  usedShadingModels=Unlit
 }
 nodes [
  n1 = DynamicParameter()
  n2 = Multiply(A=@n1.Param1, B=ScalarParameter() {defaultValue=200})
  n3 = MaterialFunctionCall(Float2=TextureCoordinate() {uTiling=0.44, vTiling=0.28}) {materialFunction=&BreakOutFloat2Components}
  n4 = Constant() {r=-0.5}
  n5 = Constant() {r=-0.2}
  n6 = TextureSample(UVs=Panner(Coordinate=AppendVector(A=@n3.R, B=@n3.R), Speed=AppendVector(A=@n4, B=@n5))) {samplerType=SAMPLERTYPE_LinearColor, texture=&T_NoisePack_LinearColor}
  Add(
   A=Multiply(A=Multiply(A=@n2, B=@n6.RGB), B=@n3.R)
   B=MaterialFunctionCall(WidthBase=ScalarParameter() {defaultValue=2, parameterName=Param_4}, WidthTip=ScalarParameter() {defaultValue=1, parameterName=Param_5}) {materialFunction=&SplineThicken}.WorldPositionOffset
  )
  n7 = Panner(Coordinate=TextureCoordinate() {uTiling=0.8}) {speedX=-2.2}
  n8 = TextureSample(UVs=@n7) {samplerType=SAMPLERTYPE_LinearColor, texture=/Game/Effects/Fort_Effects/Gameplay/SmokeDustExplosions/Textures/T_RibbonSmokeMasked_03}
  n9 = Power(Base=@n8.RGB)
  n10 = TextureCoordinate() {uTiling=1.55}
  n11 = Panner(Coordinate=TextureCoordinate() {uTiling=0.4, vTiling=0.55}) {speedX=0.3, speedY=0.12}
  n12 = Multiply(A=ComponentMask(None=TextureSample(UVs=@n11) {samplerType=SAMPLERTYPE_Normal, texture=&T_Weap_Distort_N_02}.RGB) {g=true, r=true}) {constB=0.12}
  n13 = TextureSample(UVs=Panner(Coordinate=Add(A=@n10, B=@n12)) {speedX=-3}) {texture=/Game/Effects/Fort_Effects/Textures/Smoke/T_Lobber_Trail_Mask}
  n14 = Multiply(
   A=Power(Base=TextureSample(UVs=@n7) {samplerType=SAMPLERTYPE_Grayscale, texture=&LowResBlurredNoise}.RGB)
   B=@n13.RGB
  )
  n15 = ScalarParameter() {defaultValue=0.92, parameterName=Param_1}
  MaterialFunctionCall(Emissive=LinearInterpolate(A=Constant3Vector(), B=Add(A=@n9, B=@n14), Alpha=Power(Base=@n3.R, Exp=@n15))) {materialFunction=&MF_FX_EmissiveAdjustments}
  n16 = SphereMask(
   A=MaterialFunctionCall(Float2=TextureCoordinate()) {materialFunction=&BreakOutFloat2Components}.R
   B=ScalarParameter() {defaultValue=0.1, parameterName=Param_3}
   Radius=ScalarParameter() {defaultValue=0.67, parameterName=Param_2}
  ) {hardnessPercent=0}
  n17 = Multiply(A=@n13.RGB, B=Multiply(A=@n16, B=@n16))
  Multiply(A=ParticleColor().A, B=DepthFade(Opacity=Multiply(A=@n17, B=@n1.Param2) {constB=0.5}) {fadeDistanceDefault=25})
 ]
}
