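# M_Opacity0 is an unlit, translucent effects material that, going by its name, renders fully transparent at zero opacity.
# Its graph holds a single Constant node left at its default of zero, with no textures, parameters or other logic.
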Material /BRCosmetics/Athena/Effects/Materials/M_Opacity0
graph {
 properties {blendMode=BLEND_Translucent, pixelDepthOffsetMode=PDOM_Legacy, shadingModel=MSM_Unlit, usedShadingModels=Unlit}
 nodes [Constant()]
}
