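# Additive, unlit, two-sided mesh-particle material for the Barbarian pickaxe effect, drawing a soft animated glow.
# Two soft sphere masks on U form a gradient fading from the U=0 edge.
# A panning tile pattern is subtracted from the wider mask and added to the narrower one to break up the glow.
# Color, light parameter and particle color tint it before an emissive adjustment function; particle alpha scales the wide mask.
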Material /BRCosmetics/Athena/Effects/Materials/M_Pickaxe_001_Barbarian
graph {
 properties {
  bUsedWithMeshParticles=true
  blendMode=BLEND_Additive
  pixelDepthOffsetMode=PDOM_Legacy
  previewMesh=/Game/Effects/Fort_Effects/Mesh/SM_Pickaxe_Barbarian
  shadingModel=MSM_Unlit
  twoSided=true
  usedShadingModels=Unlit
 }
 nodes [
  n1 = Multiply(
   A=ScalarParameter() {defaultValue=1, parameterName=light}
   B=VectorParameter() {parameterName=color, defaultValue=<rgba 0.0699 0.7583 1 1>}.RGB
  )
  n2 = TextureCoordinate()
  n3 = SphereMask(A=ComponentMask(None=TextureCoordinate()) {r=true}, B=Constant()) {attenuationRadius=0.32, hardnessPercent=0}
  n4 = TextureSample(UVs=Panner(Coordinate=TextureCoordinate() {vTiling=2}) {speedX=-0.25}) {texture=/Game/Effects/Fort_Effects/Textures/T_Tile_Pattern_01}
  n5 = Add(
   A=SphereMask(A=ComponentMask(None=@n2) {r=true}, B=Constant()) {attenuationRadius=0.16, hardnessPercent=0}
   B=Multiply(A=Subtract(A=@n3, B=@n4.R)) {constB=0.5}
  )
  n6 = ParticleColor()
  MaterialFunctionCall(Emissive=Multiply(A=Multiply(A=@n1, B=@n5), B=@n6.RGB)) {materialFunction=&MF_FX_EmissiveAdjustments}
  Multiply(A=@n3, B=@n6.A)
 ]
}
