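# Unlit, two-sided translucent particle material for pickaxe flame effects, usable on mesh and Niagara mesh particles.
# Color blends two parameters along V, tinted by particle color and EmissiveColor, then passed through an emissive-adjustment function.
# Opacity sums two panning noises, blends in a plasma edge, and masks by a vertical ramp and particle alpha, with dynamic parameters controlling edge, brightness and depth fade.
# Mobile substitutes three offset plasma samples; low quality skips depth fade.
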Material /BRCosmetics/Athena/Effects/Materials/M_Pickaxe_Flame
graph {
 properties {
  allowTranslucentCustomDepthWrites=true
  bUsedWithMeshParticles=true
  bUsedWithNiagaraMeshParticles=true
  blendMode=BLEND_Translucent
  pixelDepthOffsetMode=PDOM_Legacy
  shadingModel=MSM_Unlit
  twoSided=true
  usedShadingModels=Unlit
 }
 nodes [
  n1 = VectorParameter() {parameterName=Color_a, defaultValue=<rgba 1 0.4545 0 1>}
  n2 = VectorParameter() {parameterName=Color_b, defaultValue=<rgba 1 0.0067 0 1>}
  n3 = ComponentMask(None=TextureCoordinate()) {g=true}
  n4 = ParticleColor()
  n5 = Multiply(A=@n4.RGB, B=VectorParameter() {parameterName=EmissiveColor, defaultValue=<rgba 1 1 1 1>}.RGB)
  MaterialFunctionCall(Emissive=Multiply(A=LinearInterpolate(A=@n1.RGB, B=@n2.RGB, Alpha=@n3), B=ShadingPathSwitch(Default=@n5) {desc=Migrated from FeatureLevelSwitch})) {materialFunction=&MF_FX_EmissiveAdjustments}
  n6 = Multiply(A=TextureCoordinate() {vTiling=0.01}, B=ScalarParameter() {defaultValue=0.1, parameterName=Tile_Tex_0})
  n7 = Multiply(A=TextureCoordinate() {uTiling=4, vTiling=0.025}, B=ScalarParameter() {defaultValue=0.12, parameterName=Tile_Tex_2})
  n8 = Add(
   A=TextureSample(UVs=Panner(Coordinate=@n6) {speedX=0.012, speedY=-0.125}) {texture=/Game/Global/Textures/Default/Noise/T_TileNoise_02}.R
   B=TextureSample(UVs=Panner(Coordinate=@n7) {speedX=0.025, speedY=-0.1}) {texture=&T_TileNoise_02}.R
  )
  n9 = Panner(Coordinate=Multiply(A=TextureCoordinate() {uTiling=6}, B=ScalarParameter() {defaultValue=1, parameterName=Tile_Tex_1})) {speedX=0.125}
  n10 = DynamicParameter() {defaultValue=<rgba 0 0 0 1>, paramNames [LerpEdgeMask,BrightenMask,FadeDistance,Param4]}
  n11 = LinearInterpolate(
   A=@n8
   B=Add(A=TextureSample(UVs=@n9) {texture=/Game/Effects/Fort_Effects/Textures/Tile/T_Plasma_BallEdge}.R, B=@n8)
   Alpha=@n10.LerpEdgeMask
  ) {constAlpha=1}
  n12 = Multiply(A=TextureCoordinate() {uTiling=6}, B=ScalarParameter() {defaultValue=1, parameterName=Tile_Tex_1})
  n13 = AppendVector(A=ScalarParameter() {defaultValue=0.3, parameterName=U_Offset}, B=ScalarParameter() {parameterName=V_Offset})
  n14 = Panner(Coordinate=Add(A=@n12, B=@n13)) {speedX=0.125}
  n15 = Panner(Coordinate=@n12) {speedX=0.125}
  n16 = Add(A=@n12, B=Multiply(A=@n13) {constB=-1})
  n17 = AppendVector(
   A=AppendVector(
    A=TextureSample(UVs=@n14) {texture=&T_Plasma_BallEdge}.R
    B=TextureSample(UVs=@n15) {texture=&T_Plasma_BallEdge}.R
   )
   B=TextureSample(UVs=Panner(Coordinate=@n16) {speedX=0.125}) {texture=&T_Plasma_BallEdge}.R
  )
  n18 = Constant()
  n19 = Add(A=Constant() {r=1}, B=@n10.BrightenMask)
  n20 = TextureSample() {texture=/Game/Global/Textures/Default/Masks_and_Gradients/T_Soft_Ramp_Vert_03}
  n21 = Multiply(
   A=@n4.A
   B=Multiply(A=ShadingPathSwitch(Default=@n11, Mobile=@n17) {desc=Migrated from FeatureLevelSwitch}, B=LinearInterpolate(A=@n18, B=@n19, Alpha=@n20.R))
  )
  QualitySwitch(Default=DepthFade(Opacity=@n21, FadeDistance=@n10.FadeDistance), Low=@n21)
 ]
}
