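# Masked, two-sided energy material for the GlowBro pickaxe effect on Niagara ribbons and mesh particles, with a Trail switch reorienting UVs and masks.
# Distorted, panning cloud normals are rounded into bands or swapped for a noise mask, blending four glow colors into HDR emissive, plus optional view-space fake reflections.
# Opacity combines scrolling smoke, sphere masks and dissolution driven by dynamic parameters or flame texture thresholded by particle alpha, with non-mobile depth fade.
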Material /BRCosmetics/Athena/Effects/Materials/M_Pickaxe_GlowBro_Energy
graph {
 properties {
  bUsedWithNiagaraMeshParticles=true
  bUsedWithNiagaraRibbons=true
  blendMode=BLEND_Masked
  pixelDepthOffsetMode=PDOM_Legacy
  twoSided=true
  usedShadingModels=Default Lit
 }
 nodes [
  n1 = TextureSampleParameter2D(UVs=Panner(Coordinate=TextureCoordinate() {uTiling=0.4, vTiling=0.55}) {speedX=-0.1, speedY=0.8}) {parameterName=Normal Texture, samplerType=SAMPLERTYPE_Normal, texture=&T_Weap_Distort_N_02}
  n2 = Multiply(A=ComponentMask(None=@n1.RGB) {g=true, r=true}) {constB=0.2}
  n3 = ScalarParameter() {defaultValue=1, parameterName=Normal}
  LinearInterpolate(A=Multiply(A=AppendVector(A=@n2, B=@n1.B), B=@n3), B=Multiply(A=@n1.RGB, B=@n3), Alpha=ScalarParameter() {parameterName=Normal Change})
  n4 = TextureCoordinate()
  n5 = ComponentMask(None=@n4) {r=true}
  n6 = ComponentMask(None=@n4) {g=true}
  n7 = StaticSwitchParameter(True=OneMinus(None=@n5), False=@n6) {parameterName=Trail}
  n8 = ScalarParameter() {defaultValue=8.5, parameterName=Color contrast}
  n9 = LinearInterpolate(
   A=VectorParameter() {group=Color, parameterName=GlowColor3, defaultValue=<rgba 0.1667 0.0069 0.022 0>}.RGB
   B=VectorParameter() {group=Color, parameterName=GlowColor2, defaultValue=<rgba 0.4844 0.0562 0.1143 0>}.RGB
   Alpha=LinearInterpolate(A=Power(Base=@n7, Exp=@n8) {constExponent=8.5}, B=Power(Base=@n5, Exp=@n8), Alpha=ScalarParameter() {parameterName=Sharp type mask})
  )
  n10 = Constant3Vector() {constant=<rgba 0 0 1 0>}
  n11 = VectorParameter() {parameterName=UV, defaultValue=<rgba 0.4 1 1 1.5>}
  n12 = MaterialFunctionCall(Float2=TextureCoordinate() {coordinateIndex=1, uTiling=0.7, vTiling=0.9}) {materialFunction=&BreakOutFloat2Components}
  n13 = AppendVector(A=Multiply(A=@n11.B, B=@n12.R), B=Power(Base=Multiply(A=@n11.A, B=@n12.G)) {constExponent=0.7})
  n14 = MaterialFunctionCall(XY=@n13) {materialFunction=&Swizzle}
  n15 = TextureCoordinate() {uTiling=0.5, vTiling=0.64}
  n16 = ComponentMask(None=TextureSample(UVs=Panner(Coordinate=@n15) {speedX=-0.3, speedY=0.2}) {samplerType=SAMPLERTYPE_Normal, texture=&T_Weap_Distort_N_02}.RGB) {g=true, r=true}
  n17 = ScalarParameter() {defaultValue=0.09, parameterName=Noise}
  n18 = AppendVector(A=ScalarParameter() {defaultValue=-0.1, parameterName=SpeedX}, B=ScalarParameter() {defaultValue=1.2, parameterName=SpeedY})
  n19 = Panner(
   Coordinate=Add(A=StaticSwitchParameter(True=@n14.YX, False=@n13) {parameterName=Trail}, B=Multiply(A=@n16, B=@n17) {constB=0.09})
   Speed=@n18
  ) {speedX=-0.1, speedY=0.8}
  n20 = TextureSample(UVs=@n19) {samplerType=SAMPLERTYPE_Normal, texture=/Game/Global/Textures/Default/Tileable_Patterns/T_StormCloudDetail01_N}
  n21 = OneMinus(None=@n7)
  n22 = LinearInterpolate(A=@n10, B=@n20.RGB, Alpha=Multiply(A=@n20.RGB, B=Power(Base=@n21) {constExponent=0.5}))
  n23 = Round(None=LinearInterpolate(A=ScalarParameter() {defaultValue=7, parameterName=Dark}, Alpha=ComponentMask(None=@n22) {b=true}) {constA=11, constB=0})
  n24 = VectorParameter() {parameterName=UV, defaultValue=<rgba 0.4 1 1 1.5>}
  n25 = Multiply(A=TextureCoordinate() {coordinateIndex=1}, B=AppendVector(A=@n24.B, B=@n24.B))
  n26 = TextureSampleParameter2D(UVs=Panner(Coordinate=MaterialFunctionCall(XY=@n25) {materialFunction=&Swizzle}.YX, Speed=@n18)) {parameterName=Color Mask Texture, texture=/Game/Effects/Fort_Effects/Textures/Plasma/T_PlasmaNoise_A_01}
  n27 = OneMinus(None=Power(
    Base=StaticComponentMaskParameter(None=@n26.RGB) {defaultR=true, parameterName=Color Mask RGB}
    Exp=ScalarParameter() {defaultValue=1, parameterName=Color Mask Power}
   ))
  n28 = ScalarParameter() {parameterName=Color Mask}
  n29 = TextureSample(UVs=Panner(Coordinate=TextureCoordinate() {uTiling=0.4, vTiling=0.55}, Speed=@n18) {speedX=-0.1, speedY=0.8}) {samplerType=SAMPLERTYPE_Normal, texture=&T_StormCloudDetail01_N}
  n30 = ComponentMask(None=TextureCoordinate()) {r=true}
  n31 = TextureCoordinate() {coordinateIndex=1}
  n32 = Power(
   Base=StaticSwitchParameter(True=OneMinus(None=@n30), False=ComponentMask(None=@n31) {g=true}) {parameterName=Trail}
   Exp=ScalarParameter() {defaultValue=5, parameterName=Highlight mask}
  ) {constExponent=5}
  n33 = LinearInterpolate(
   A=ScalarParameter() {defaultValue=9, parameterName=Light}
   Alpha=ComponentMask(None=LinearInterpolate(A=@n10, B=@n29.RGB, Alpha=@n32)) {b=true}
  ) {constA=4.8, constB=0}
  n34 = OneMinus(None=StaticComponentMaskParameter(None=@n26.RGB) {defaultR=true, parameterName=Color Mask RGB_Side})
  n35 = LinearInterpolate(
   A=LinearInterpolate(
    A=@n9
    B=VectorParameter() {group=Color, parameterName=GlowColor4, defaultValue=<rgba 0.099 0 0.0216 0>}.RGB
    Alpha=LinearInterpolate(A=@n23, B=@n27, Alpha=@n28)
   )
   B=VectorParameter() {group=Color, parameterName=GlowColor1, defaultValue=<rgba 0.5469 0.046 0.2068 0>}.RGB
   Alpha=LinearInterpolate(A=Round(None=@n33), B=Power(Base=@n34) {constExponent=8.5}, Alpha=@n28)
  )
  n36 = Clamp(None=@n35)
  n37 = ShadingPathSwitch(
   Default=MaterialFunctionCall(LDR Input=@n36, HDR Tint=@n35, HDR Intensity Multiplier=ScalarParameter() {defaultValue=3, parameterName=HDRIntensity}) {materialFunction=&DeriveHDRfromLDR}.RemappedRange
   Mobile=@n36
  ) {desc=Migrated from FeatureLevelSwitch}
  n38 = ParticleColor()
  n39 = Multiply(
   A=TextureCoordinate()
   B=AppendVector(
    A=ScalarParameter() {defaultValue=1, parameterName=Reflection U Scale}
    B=ScalarParameter() {defaultValue=1, parameterName=Reflection V Scale}
   )
  )
  n40 = Transform(None=TextureSample(UVs=Panner(Coordinate=@n39) {speedY=0.1}) {samplerType=SAMPLERTYPE_Normal, texture=&T_Weap_Distort_N_02}.RGB)
  n41 = ScalarParameter() {defaultValue=0.5, parameterName=ReflectionNoiseSTR}
  n42 = Transform(None=Add(A=Multiply(A=@n40, B=@n41), B=VertexNormalWS())) {transformSourceType=TRANSFORMSOURCE_World, transformType=TRANSFORM_View}
  n43 = ComponentMask(None=Add(A=Multiply(A=@n42) {constB=0.5}) {constB=0.5}) {g=true, r=true}
  n44 = Multiply(
   A=MaterialFunctionCall(
    In=TextureSampleParameter2D(UVs=@n43) {parameterName=Fake Reflections Texture, texture=/Game/Packages/Fortress_SharedMaterials/Textures/T_Ravens_SpecularReflection}.G
    Contrast=ScalarParameter() {parameterName=ReflectionContrast}
   ) {materialFunction=&CheapContrast}.Result
   B=VectorParameter() {parameterName=ReflectionColor, defaultValue=<rgba 0 0.3 4 1>}.RGB
  )
  n45 = Constant()
  MaterialFunctionCall(Emissive=Add(A=Multiply(A=@n37, B=@n38.RGB), B=StaticSwitchParameter(True=@n44, False=@n45) {parameterName=Add Fake Reflections})) {materialFunction=&MF_FX_EmissiveAdjustments}
  Multiply(
   A=MaterialFunctionCall() {materialFunction=&DitherTemporalAA}.Result
   B=ScalarParameter()
  )
  n46 = TextureCoordinate() {uTiling=0.78}
  n47 = MaterialFunctionCall(XY=TextureCoordinate() {vTiling=0.78}) {materialFunction=&Swizzle}
  n48 = TextureSampleParameter2D(UVs=Panner(Coordinate=StaticSwitchParameter(True=@n46, False=@n47.YX) {parameterName=Trail}) {speedX=0.7}) {
   parameterName=Diffuse
   samplerType=SAMPLERTYPE_LinearColor
   texture=/Game/Effects/Fort_Effects/Gameplay/SmokeDustExplosions/Textures/T_RibbonSmokeMasked_03
  }
  Multiply(A=Multiply(A=@n9, B=@n48.R), B=ScalarParameter() {defaultValue=0.4, parameterName=Base})
  ScalarParameter() {defaultValue=1, parameterName=Roughness}
  n49 = SphereMask(A=TextureCoordinate(), B=Constant2Vector() {g=0.5, r=1}) {attenuationRadius=1, hardnessPercent=75}
  n50 = ScalarParameter() {parameterName=Sharp type mask}
  n51 = VectorParameter() {group=Trailing Edge Sphere Mask, parameterName=Trailing Edge Sphereical UV Scale, sortPriority=4, defaultValue=<rgba 0.5 1 0 0>}
  n52 = VectorParameter() {group=Trailing Edge Sphere Mask, parameterName="Trailing Edge Sphere Mask Coord (B)", sortPriority=3, defaultValue=<rgba 0.85 0.82 0 0>}
  n53 = SphereMask(
   A=Multiply(A=TextureCoordinate(), B=ComponentMask(None=@n51.RGB) {g=true, r=true})
   B=ComponentMask(None=@n52.RGB) {g=true, r=true}
   Radius=ScalarParameter() {defaultValue=0.8, group=Trailing Edge Sphere Mask, parameterName=Trailing Edge Radius, sortPriority=1}
   Hardness=ScalarParameter() {defaultValue=0.3, group=Trailing Edge Sphere Mask, parameterName=Trailing Edge Hardness, sortPriority=2}
  )
  n54 = LinearInterpolate(
   B=Power(Base=OneMinus(None=@n53), Exp=ScalarParameter() {defaultValue=5, group=Trailing Edge Sphere Mask, parameterName=Trailing Edge Contrast}) {constExponent=5}
   Alpha=ScalarParameter() {defaultValue=1, group=Alpha, parameterName=Multiply Trailing Edge Sphere Mask, sortPriority=6}
  ) {constA=1}
  n55 = VectorParameter() {parameterName=Mask_UV, defaultValue=<rgba 1 1 1 1>}
  n56 = Multiply(A=TextureCoordinate(), B=AppendVector(A=@n55.R, B=@n55.R))
  n57 = TextureCoordinate() {vTiling=2}
  n58 = MaterialFunctionCall(XY=@n57) {materialFunction=&Swizzle}
  n59 = ComponentMask(None=@n58.YX) {g=true}
  n60 = AppendVector(A=ComponentMask(None=@n58.YX) {r=true}, B=Clamp(None=LinearInterpolate(Alpha=@n59) {constA=-0.15, constB=1.15}))
  n61 = ScalarParameter() {defaultValue=0.28, parameterName=SpeedX_MaskTrail}
  n62 = ScalarParameter() {parameterName=SpeedY_MaskTrail}
  n63 = TextureSampleParameter2D(UVs=Panner(Coordinate=StaticSwitchParameter(True=@n56, False=@n60) {parameterName=Trail}, Speed=AppendVector(A=@n61, B=@n62)) {speedX=0.28}) {parameterName=Opacity Mask, texture=/Game/Effects/Fort_Effects/Gameplay/SmokeDustExplosions/Textures/T_RibbonSmokeMasked_02}
  n64 = TextureCoordinate()
  n65 = OneMinus(None=ComponentMask(None=@n64) {r=true})
  n66 = TextureSampleParameter2D(UVs=@n19) {
   parameterName=Dissolution
   samplerType=SAMPLERTYPE_LinearGrayscale
   texture=/Game/Global/Textures/Default/Tileable_Patterns/T_StormCloudDetail01_M
  }
  n67 = SphereMask(A=@n5, B=Constant(), Radius=DynamicParameter() {paramNames [Bend,WPOMult,Param3,Param4]}.WPOMult) {hardnessPercent=0}
  n68 = LinearInterpolate(
   A=ScalarParameter() {defaultValue=-3, parameterName=Dissolution strength}
   B=ScalarParameter() {parameterName=Dissolution range}
   Alpha=StaticSwitchParameter(True=@n67, False=@n6) {parameterName=Trail}
  ) {constA=-3}
  n69 = ScalarParameter() {defaultValue=1.5, parameterName=Strength}
  n70 = Multiply(A=@n69, B=DynamicParameter() {paramNames [Bend,WPOMult,Param3,Param4]}.Bend)
  n71 = TextureCoordinate() {coordinateIndex=1}
  n72 = VectorParameter() {parameterName=UV, defaultValue=<rgba 0.4 1 1 1.5>}
  n73 = AppendVector(A=@n72.B, B=@n72.B)
  n74 = Panner(
   Coordinate=MaterialFunctionCall(XY=Multiply(A=@n71, B=@n73)) {materialFunction=&Swizzle}.YX
   Speed=AppendVector(A=ScalarParameter() {defaultValue=-0.1, parameterName=SpeedX}, B=ScalarParameter() {defaultValue=1.2, parameterName=SpeedY})
  )
  n75 = Multiply(
   A=@n69
   B=Multiply(
    A=@n67
    B=TextureSampleParameter2D(UVs=@n74) {parameterName=Dissolution Texture, texture=/Game/Effects/Fort_Effects/Textures/Tile/T_Tile_Texture_Cartoon_Flames_Layer}.G
   )
  )
  n76 = Constant()
  n77 = If(
   A=TextureSample(UVs=@n74) {texture=/Game/Effects/Fort_Effects/Textures/T_Contrail_Lava2}.B
   B=ParticleColor().A
   "A > B"=Constant() {r=1}
   "A == B"=@n76
   "A < B"=@n76
  )
  n78 = Clamp(
   None=LinearInterpolate(
    A=LinearInterpolate(A=@n68, B=@n70, Alpha=@n66.R) {constA=-1.5, constB=1.3}
    B=Multiply(A=@n75, B=@n77)
    Alpha=ScalarParameter() {parameterName=Dissolution}
   )
   Max=ScalarParameter() {defaultValue=1, parameterName=Mask_Clamp_Max}
  )
  n79 = Constant() {r=1}
  n80 = Multiply(A=ScalarParameter() {defaultValue=1, parameterName=Height}, B=DynamicParameter() {paramNames [Bend,WPOMult,Param3,Param4]}.Param3)
  n81 = ComponentMask(None=@n64) {g=true}
  n82 = StaticSwitchParameter(
   True=Multiply(A=SquareRoot(None=@n65), B=@n78)
   False=Multiply(A=@n78, B=SphereMask(A=@n81, B=@n79, Radius=@n80) {attenuationRadius=1, hardnessPercent=0})
  ) {parameterName=Trail}
  n83 = VertexColor()
  n84 = SphereMask(
   A=MaterialFunctionCall(Float2=@n57) {materialFunction=&BreakOutFloat2Components}.R
   B=Constant() {r=0.5}
   Radius=ScalarParameter() {defaultValue=0.5, parameterName=Sphere Mask Radius}
  ) {hardnessPercent=0}
  n85 = Multiply(A=@n84, B=@n82)
  n86 = Multiply(A=@n63.R, B=@n84)
  n87 = StaticSwitchParameter(
   True=Multiply(A=LinearInterpolate(A=@n49, B=@n54, Alpha=@n50), B=Multiply(A=@n63.R, B=@n82))
   False=StaticSwitchParameter(True=Multiply(A=@n83.R, B=@n85), False=Multiply(A=@n86, B=@n82)) {parameterName=Mask}
  ) {parameterName=Trail}
  n88 = DepthFade(Opacity=@n87, FadeDistance=Constant() {r=50})
  ShadingPathSwitch(Default=@n88, Deferred=@n88, Mobile=@n87) {desc=Migrated from FeatureLevelSwitch}
 ]
}
