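# Additive unlit rainbow material for contrail ribbons on beam trails and Niagara ribbons.
# Three offset triangle gradients across UV V form RGB bands, hue-shifted by world height plus an optional time pan with a dynamic RainbowOffset, then fed through emissive adjustments.
# Opacity combines particle alpha, ramp, depth fade, noise flicker driven by time and horizontal camera direction, and time-of-day alphas.
# Mobile drops the depth fade and flicker.
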Material /BRCosmetics/Athena/Effects/Materials/M_RainbowContrailRibbon
graph {
 properties {
  bUsedWithBeamTrails=true
  bUsedWithNiagaraRibbons=true
  blendMode=BLEND_Additive
  pixelDepthOffsetMode=PDOM_Legacy
  shadingModel=MSM_Unlit
  usedShadingModels=Unlit
 }
 nodes [
  n1 = WorldPosition()
  n2 = ScalarParameter() {defaultValue=256, parameterName=Distance}
  n3 = Multiply(A=Time()) {constB=-1}
  n4 = DynamicParameter() {paramNames [RainbowOffset, Param2, Param3, Param4]}
  n5 = Add(
   A=ComponentMask(None=Divide(A=@n1.XYZ, B=@n2) {constB=256}) {b=true}
   B=LinearInterpolate(B=Add(A=@n3, B=@n4.RainbowOffset), Alpha=ScalarParameter() {parameterName=Use Dynamic Parameter Pan Over Time})
  )
  n6 = ComponentMask(None=TextureCoordinate() {uTiling=2, vTiling=2}) {g=true}
  n7 = OneMinus(None=Abs(None=@n6))
  n8 = OneMinus(None=Abs(None=Add(A=@n6) {constB=-1}))
  n9 = OneMinus(None=Abs(None=Add(A=@n6) {constB=-2}))
  MaterialFunctionCall(Emissive=MaterialFunctionCall(
    Hue Shift Percentage=Frac(None=@n5)
    Texture=MaterialFunctionCall(X=@n7, Y=@n8, Z=@n9) {materialFunction=&MakeFloat3}.Result
   ) {materialFunction=&HueShift}.Result) {materialFunction=&MF_FX_EmissiveAdjustments}
  n10 = ScalarParameter() {defaultValue=1, group=TOD Alpha, parameterName=Day Alpha, sortPriority=1}
  n11 = ScalarParameter() {defaultValue=1, group=TOD Alpha, parameterName=Night Alpha, sortPriority=2}
  n12 = ParticleColor()
  n13 = Power(Base=TextureSample() {texture=/Game/Effects/Fort_Effects/Textures/Ramps/T_Ramp_Smooth}.G)
  n14 = ScalarParameter() {defaultValue=5, parameterName=DepthFade Distance}
  n15 = Time()
  n16 = MaterialFunctionCall() {materialFunction=&CameraDirectionVector}
  n17 = Constant3Vector() {constant=<rgba 1 1 0 0>}
  n18 = DotProduct(A=Normalize(VectorInput=Multiply(A=@n16.Result, B=@n17)), B=Constant3Vector() {constant=<rgba 1 0 0 0>})
  n19 = AppendVector(A=Add(A=Multiply(A=@n15) {constB=0.1}, B=ConstantBiasScale(None=@n18)), B=@n15)
  n20 = Multiply(
   A=Multiply(A=@n12.A, B=DepthFade(Opacity=@n13, FadeDistance=@n14) {fadeDistanceDefault=5})
   B=LinearInterpolate(Alpha=TextureSample(UVs=@n19) {texture=&T_Smooth_tile}.G) {constA=0.2, constB=1.5}
  )
  Multiply(
   A=MaterialFunctionCall(Color_Morning=@n10, Color_Day=@n10, Color_Evening=@n11, Color_Night=@n11) {materialFunction=&Color_Time_of_Day}.Result
   B=ShadingPathSwitch(Default=@n20, Deferred=@n20, Mobile=Multiply(A=@n12.A, B=@n13)) {desc=Migrated from FeatureLevelSwitch}
  )
 ]
}
