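# M_SimpleContrailRibbon is an unlit translucent material for Niagara ribbon and beam trail contrails.
# A panned cloud texture, offset by normal-map distortion that grows as particle alpha drops, combines with a rounded cross-ribbon profile and HeightLerp to shape opacity.
# Dynamic parameters control trail alpha and end-cap falloff, with switchable depth and camera-distance fades.
# Emissive uses ambient color, optionally tinted by time of day or cycled through a rainbow gradient.
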Material /BRCosmetics/Athena/Effects/Materials/M_SimpleContrailRibbon
graph {
 properties {
  bUsedWithBeamTrails=true
  bUsedWithNanite=true
  bUsedWithNiagaraRibbons=true
  blendMode=BLEND_Translucent
  pixelDepthOffsetMode=PDOM_Legacy
  shadingModel=MSM_Unlit
  usedShadingModels=Unlit
 }
 nodes [
  n1 = Multiply(A=Time(), B=ScalarParameter() {defaultValue=1, parameterName=LightSpeed})
  n2 = VectorParameter() {parameterName=AmbientColor, defaultValue=<rgba 0.0938 0.0938 0.25 0>}
  n3 = Multiply(
   A=MaterialFunctionCall(
    Color_Morning=VectorParameter() {parameterName=Morning, defaultValue=<rgba 0.1771 0.0891 0.0577 1>}.RGB
    Color_Day=VectorParameter() {parameterName=Day, defaultValue=<rgba 0.4688 0.4181 0.3311 1>}.RGB
    Color_Evening=VectorParameter() {parameterName=Evening, defaultValue=<rgba 0.0781 0.0435 0.0375 1>}.RGB
    Color_Night=VectorParameter() {parameterName=Night, defaultValue=<rgba 0.1182 0.2527 0.3698 1>}.RGB
   ) {materialFunction=&Color_Time_of_Day}.Result
   B=@n2.RGB
  )
  n4 = Subtract(B=ComponentMask(None=TextureCoordinate()) {g=true}) {constA=0.5}
  n5 = MaterialFunctionCall!(Side 1 Length=Multiply(A=Abs(None=@n4)) {constB=2}) {materialFunction=/Engine/Functions/Engine_MaterialFunctions02/Math/pythagoreanTheorum}
  n6 = ParticleColor()
  n7 = VectorParameter() {parameterName=UV Tile_1, defaultValue=<rgba 0.7 0.75 0 0>}
  n8 = TextureSample(UVs=Multiply(A=TextureCoordinate() {coordinateIndex=1}, B=ComponentMask(None=@n7.RGB) {g=true, r=true})) {samplerType=SAMPLERTYPE_Normal, texture=/Game/Effects/Fort_Effects/BluGlow/testNormals}
  n9 = Multiply(
   A=Multiply(A=OneMinus(None=@n6.A), B=ComponentMask(None=@n8.RGB) {g=true, r=true}) {constA=1}
   B=ScalarParameter() {defaultValue=0.15, parameterName=UV Dist}
  )
  n10 = DynamicParameter() {paramNames [Param1, Trail Alpha, End Cap Fade, Param4]}
  n11 = ScalarParameter() {defaultValue=0.5, parameterName=XY Shift}
  n12 = Multiply(
   A=Add(A=TextureCoordinate() {coordinateIndex=1}, B=Multiply(A=@n10.Param1, B=@n11))
   B=ComponentMask(None=VectorParameter() {parameterName=UV Tile, defaultValue=<rgba 0.125 0.6 0 0>}.RGB) {g=true, r=true}
  )
  n13 = VectorParameter() {parameterName=PanningSpeed}
  n14 = TextureSample(UVs=Panner(Coordinate=Add(A=@n9, B=@n12), Speed=ComponentMask(None=@n13.RGB) {g=true, r=true})) {samplerType=SAMPLERTYPE_LinearColor, texture=/Game/Global/Textures/Default/Tileable_Patterns/Cloud_Diffuse}
  n15 = Multiply(A=Multiply(A=@n5.Side 2 Length, B=@n14.R), B=ComponentMask(None=TextureCoordinate()) {g=true})
  n16 = ScalarParameter() {defaultValue=1, parameterName=p}
  n17 = LinearInterpolate(
   A=LinearInterpolate(A=StaticSwitchParameter(True=@n3, False=@n2.RGB) {parameterName=Use Time of Day}, Alpha=Power(Base=@n15, Exp=@n16)) {constA=1}
   B=ParticleColor().RGB
   Alpha=ScalarParameter() {parameterName=ParticleColorStrength}
  )
  n18 = Multiply(
   A=TextureSample(UVs=Frac(None=@n1)) {texture=/Game/Athena/Events/NewYear/DiscoNY_Gradient}.RGB
   B=Desaturation(None=@n17, Fraction=Constant() {r=1})
  )
  n19 = ParticleColor()
  MaterialFunctionCall(Emissive=StaticSwitchParameter(True=Multiply(A=@n18, B=@n19.RGB), False=@n17) {parameterName=Use Rainbow}) {materialFunction=&MF_FX_EmissiveAdjustments}
  n20 = MaterialFunctionCall(Transition Phase=Multiply(A=@n5.Side 2 Length, B=@n5.Side 2 Length), Height Texture=LinearInterpolate(Alpha=@n14.R) {constA=-1}) {materialFunction=&HeightLerp}
  n21 = DepthFade(Opacity=Constant() {r=1}) {fadeDistanceDefault=15}
  n22 = Constant() {r=1}
  n23 = Multiply(
   A=ShadingPathSwitch(Default=QualitySwitch(Default=@n21, Low=@n22, High=@n21, Medium=@n21), Mobile=@n22) {desc=Migrated from FeatureLevelSwitch}
   B=MaterialFunctionCall(Fade Length=Constant() {r=25}) {materialFunction=&CameraDepthFade}.Result
  )
  n24 = Constant() {r=1}
  n25 = DynamicParameter() {paramNames [Param1, Trail Alpha, End Cap Fade, Param4]}
  n26 = Multiply(A=Multiply(A=ParticleColor().A, B=StaticSwitchParameter(True=@n23, False=@n24) {parameterName=UseDepthFade?}) {constA=1}, B=@n25.Trail Alpha)
  n27 = Sine(None=ComponentMask(None=TextureCoordinate()) {r=true}) {period=2}
  n28 = Multiply(
   A=Multiply(A=Multiply(A=@n20.Alpha, B=@n26), B=Power(Base=@n27, Exp=@n25.End Cap Fade))
   B=ScalarParameter() {defaultValue=0.675, parameterName=Opacity}
  )
  n29 = ScalarParameter() {defaultValue=2, parameterName=DistanceBrightness}
  n30 = MaterialFunctionCall(Distance=ScalarParameter() {defaultValue=1000, parameterName=BrightenDistance}) {materialFunction=&Near_Camera_Fade}
  n31 = StaticSwitchParameter(True=Multiply(A=@n28, B=LinearInterpolate(B=@n29, Alpha=@n30.Result) {constA=1}), False=@n28) {parameterName=UseCameraFadeBrightness}
  n32 = ScalarParameter() {defaultValue=500, parameterName="Fade_Start_Distance (S)"}
  n33 = ScalarParameter() {defaultValue=256, parameterName="Fade_End_Distance (S)"}
  n34 = StaticSwitchParameter(
   True=Multiply(
    A=@n31
    B=MaterialFunctionCall(Fade_Start_Distance=@n32, Fade_End_Distance=@n33) {materialFunction=&MF_Distance_SyncedFade_FX}.Result
   )
   False=@n31
  ) {parameterName=FadeCloseToTheCamera?}
  n35 = ScalarParameter() {parameterName=DepthFadeDistance}
  n36 = Constant() {r=1}
  n37 = ShadingPathSwitch(Default=QualitySwitch(Default=DepthFade(FadeDistance=@n35), Low=@n36), Mobile=@n36) {desc=Migrated from FeatureLevelSwitch}
  StaticSwitchParameter(True=Multiply(A=@n34, B=ShadingPathSwitch(Default=@n37, Forward=@n36)), False=@n34) {parameterName=UseDepthFade?}
 ]
}
