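# Unlit translucent sprite particle material for dragon smoke, tinted by particle color.
# Particle alpha slides a smoothstep over the inverted red channel of the dragon smoke texture, eroding the shape.
# A hard threshold of that mask, times blue channel and particle color, is modulated by tiling inky detail and fed to an emissive adjustment function.
# Per-particle random offsets and scales vary the detail UVs, except on mobile; doubled green times the soft mask forms an alpha-like term.
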Material /BRCosmetics/Athena/Effects/Materials/M_Smoke_Dragon
graph {
 properties {
  bUsedWithParticleSprites=true
  blendMode=BLEND_Translucent
  pixelDepthOffsetMode=PDOM_Legacy
  shadingModel=MSM_Unlit
  usedShadingModels=Unlit
 }
 nodes [
  n1 = ScalarParameter() {defaultValue=0.4238, parameterName=SmokeDetailLow}
  n2 = TextureCoordinate()
  n3 = ParticleRandom()
  n4 = Multiply(A=@n3) {constB=3}
  n5 = Multiply(A=@n3) {constB=7}
  n6 = Add(A=@n2, B=AppendVector(A=Frac(None=@n4), B=Frac(None=@n5)))
  n7 = LinearInterpolate(Alpha=@n3) {constA=0.4, constB=0.8}
  n8 = TextureSample(UVs=ShadingPathSwitch(Default=Multiply(A=@n6, B=@n7), Mobile=@n2) {desc=Migrated from FeatureLevelSwitch}) {texture=&T_Inky_Smoke_Tile}
  n9 = ParticleColor()
  n10 = TextureSample() {texture=/BRCosmetics/Athena/Effects/Textures/T_Smoke_Dragon}
  n11 = ParticleColor()
  n12 = MaterialFunctionCall(Alpha=OneMinus(None=@n10.R), Min=@n11.A, Max=Add(A=@n11.A, B=ScalarParameter() {defaultValue=0.7, parameterName=smooththick}) {constB=0.5}) {materialFunction=&SmoothStep}
  n13 = Constant() {r=1}
  n14 = If(A=@n12.Result, B=ScalarParameter() {defaultValue=0.75, parameterName=sTestBand}, "A > B"=@n13, "A == B"=@n13, "A < B"=Constant())
  n15 = Multiply(A=@n14, B=@n10.B)
  MaterialFunctionCall(Emissive=Multiply(A=LinearInterpolate(A=@n1, Alpha=@n8.G), B=Multiply(A=@n9.RGB, B=@n15))) {materialFunction=&MF_FX_EmissiveAdjustments}
  Multiply(A=Clamp(None=Add(A=@n10.G, B=@n10.G)), B=@n12.Result)
  Multiply(A=@n14)
 ]
}
