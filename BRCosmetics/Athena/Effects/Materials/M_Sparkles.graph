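# M_Sparkles is an unlit translucent particle material for Niagara and Cascade sprites that draws a five-point glowing star sparkle.
# Particle color RGB feeds the MF_FX_EmissiveAdjustments function for the glow color.
# Particle alpha is multiplied by the star texture's red channel and a CameraDepthFade term to form the transparency mask.
# The Use Camera Fade scalar blends that near-camera fade in or out, with CameraFadeLength setting its distance.
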Material /BRCosmetics/Athena/Effects/Materials/M_Sparkles
graph {
 properties {
  bUsedWithNiagaraSprites=true
  bUsedWithParticleSprites=true
  blendMode=BLEND_Translucent
  pixelDepthOffsetMode=PDOM_Legacy
  shadingModel=MSM_Unlit
  usedShadingModels=Unlit
 }
 nodes [
  n1 = ParticleColor()
  MaterialFunctionCall(Emissive=@n1.RGB) {materialFunction=&MF_FX_EmissiveAdjustments}
  n2 = ScalarParameter() {defaultValue=25, parameterName=CameraFadeLength}
  Multiply(
   A=Multiply(A=@n1.A, B=TextureSample() {texture=/Game/Effects/Fort_Effects/Textures/Flares/T_GlowStar_5pt}.R)
   B=LinearInterpolate(
    B=MaterialFunctionCall(Fade Length=@n2) {materialFunction=&CameraDepthFade}.Result
    Alpha=ScalarParameter() {defaultValue=1, parameterName=Use Camera Fade}
   ) {constA=1}
  )
 ]
}
