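# Additive, unlit, two-sided beam-trail material for the ring effect of the toy flying disc.
# ColorA and ColorB blend along the trail's U coordinate and feed the shared FX emissive adjustments function.
# A mask samples the middle band of a Kpop contrail texture via swizzled UVs, blending from normalized to raw red along U.
# Two opposing sphere masks on U fade the trail ends, with DepthFade softening intersections.
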Material /BRCosmetics/Athena/Effects/Materials/M_Toy_FlyingDisc_Ring
graph {
 properties {
  bUsedWithBeamTrails=true
  blendMode=BLEND_Additive
  pixelDepthOffsetMode=PDOM_Legacy
  shadingModel=MSM_Unlit
  twoSided=true
  usedShadingModels=Unlit
 }
 nodes [
  n1 = TextureCoordinate()
  MaterialFunctionCall(Float2=@n1) {materialFunction=&BreakOutFloat2Components}
  n2 = MaterialFunctionCall(Float2=TextureCoordinate()) {materialFunction=&BreakOutFloat2Components}
  n3 = Multiply(A=@n2.R)
  MaterialFunctionCall(Emissive=LinearInterpolate(
    A=VectorParameter() {parameterName=ColorA, defaultValue=<rgba 0 0.1883 1 1>}.RGB
    B=VectorParameter() {parameterName=ColorB, defaultValue=<rgba 0.2944 0.6341 1 1>}.RGB
    Alpha=@n3
   )) {materialFunction=&MF_FX_EmissiveAdjustments}
  n4 = MaterialFunctionCall(XY=@n1) {materialFunction=&Swizzle}
  n5 = TextureSample(UVs=LinearInterpolate(Alpha=ComponentMask(None=@n4.YX) {r=true}) {constA=0.3, constB=0.7}) {texture=/BRCosmetics/Athena/Effects/Textures/Skydive/P_Contrail_Kpop/T_Contrail_Kpop_elements}
  n6 = Power(Base=SphereMask(A=@n2.R, B=Constant(), Radius=Constant() {r=1}) {hardnessPercent=0}) {constExponent=3}
  n7 = SphereMask(A=@n2.R, B=Constant() {r=1}, Radius=Constant() {r=1}) {hardnessPercent=88}
  Multiply(
   A=LinearInterpolate(A=Normalize(VectorInput=@n5.R), B=@n5.R, Alpha=@n3)
   B=DepthFade(Opacity=Multiply(A=@n6, B=@n7), FadeDistance=ScalarParameter() {parameterName=DepthFade})
  )
 ]
}
