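# Additive unlit beam-trail material for a firefly streak cosmetic effect, tinted by particle color and dimmed at night.
# A squared world-space sphere mask around the offset object position supplies U for panning, noise-distorted edge, smoke and noise textures, with dynamic-parameter speeds.
# Those textures multiply and hard-light blend into opacity, shaped by UV masks and faded near the camera.
# Mobile skips noise erosion, depth fade and HDR remapping.
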Material /BRCosmetics/Athena/Effects/Materials/M_Trail_001_Firefliesstreak
graph {
 properties {bUsedWithBeamTrails=true, blendMode=BLEND_Additive, pixelDepthOffsetMode=PDOM_Legacy, shadingModel=MSM_Unlit, usedShadingModels=Unlit}
 nodes [
  n1 = TextureCoordinate()
  n2 = LinearInterpolate(B=Constant3Vector() {constant=<rgba 1 0.0893 0 0>}, Alpha=OneMinus(None=ComponentMask(None=@n1) {r=true})) {constA=1}
  n3 = SphereMask(
   A=WorldPosition().XYZ
   B=Add(A=ObjectPositionWS(), B=VectorParameter() {parameterName=InitialOffset, defaultValue=<rgba 0 0 -1000 1>}.RGB)
   Radius=Constant() {r=2550}
   Hardness=Constant()
  )
  n4 = Multiply(A=@n3, B=@n3)
  n5 = AppendVector(A=Multiply(A=@n4) {constB=20}, B=ComponentMask(None=TextureCoordinate()) {g=true})
  n6 = ComponentMask(None=TextureCoordinate()) {g=true}
  n7 = Panner(Coordinate=AppendVector(A=Multiply(A=@n4) {constB=15}, B=Multiply(A=@n6) {constB=0.4})) {speedX=0.8, speedY=0.1}
  n8 = Multiply(A=ComponentMask(None=TextureSample(UVs=@n7) {samplerType=SAMPLERTYPE_LinearColor, texture=&T_NoisePack_LinearColor}.RGB) {g=true, r=true}) {constB=0.31}
  n9 = Add(A=DynamicParameter() {paramNames [UV_flow,Param2,Param3,Param4]}.UV_flow, B=Constant() {r=5.1})
  n10 = Constant()
  n11 = TextureSample(UVs=Panner(Coordinate=Add(A=@n5, B=@n8), Speed=AppendVector(A=@n9, B=@n10))) {texture=/Game/Effects/Fort_Effects/Textures/Taker/T_PortalEdge_02_Packed}
  n12 = LinearInterpolate(
   A=Constant3Vector() {constant=<rgba 0.8594 0.7817 0.136 0>}
   B=Constant3Vector() {constant=<rgba 0.4062 0.3512 0.1517 0>}
   Alpha=Add(A=@n11.R, B=@n11.R)
  )
  n13 = ParticleColor()
  n14 = Multiply(A=Multiply(A=Multiply(A=@n2, B=@n12)) {constB=4}, B=@n13.RGB)
  n15 = MaterialFunctionCall(LDR Input=Clamp(None=@n14), HDR Tint=@n14) {materialFunction=&DeriveHDRfromLDR}
  n16 = Constant() {r=1}
  n17 = Constant() {r=0.5}
  MaterialFunctionCall(Emissive=Multiply(
    A=ShadingPathSwitch(Default=@n15.RemappedRange, Mobile=@n14) {desc=Migrated from FeatureLevelSwitch}
    B=MaterialFunctionCall(Color_Morning=@n16, Color_Day=@n16, Color_Evening=@n16, Color_Night=@n17) {materialFunction=&Color_Time_of_Day}.Result
   )) {materialFunction=&MF_FX_EmissiveAdjustments}
  n18 = ComponentMask(None=TextureCoordinate()) {g=true}
  n19 = AppendVector(A=Multiply(A=@n4) {constB=16}, B=Add(A=Multiply(A=@n18) {constB=0.9}) {constB=0.1})
  n20 = DynamicParameter() {paramNames [UV_flow,Param2,Param3,Param4]}
  n21 = AppendVector(A=Add(A=@n20.UV_flow, B=Constant() {r=3}), B=Constant())
  n22 = Multiply(A=TextureSample(UVs=Panner(Coordinate=@n19, Speed=@n21) {speedX=2.4}) {samplerType=SAMPLERTYPE_Grayscale, texture=&T_Smoke_Trail_Soft}.R) {constB=0.9}
  n23 = Add(A=Constant() {r=1.2}, B=@n20.Param2)
  n24 = Constant()
  n25 = TextureSample(UVs=Panner(Coordinate=@n19, Speed=AppendVector(A=@n23, B=@n24))) {samplerType=SAMPLERTYPE_LinearColor, texture=&T_NoisePack_LinearColor}
  n26 = Multiply(A=Multiply(
    A=MaterialFunctionCall!(Base=@n11.G, Blend=@n22) {materialFunction=/Engine/Functions/Engine_MaterialFunctions03/Blends/Blend_HardLight}.Result
    B=@n25.G
   )) {constB=3}
  n27 = SphereMask(
   A=MaterialFunctionCall(Float2=TextureCoordinate()) {materialFunction=&BreakOutFloat2Components}.G
   B=Constant() {r=0.48}
  ) {attenuationRadius=0.5, hardnessPercent=0}
  n28 = OneMinus(None=ComponentMask(None=TextureCoordinate()) {r=true})
  n29 = Multiply(
   A=Multiply(A=Multiply(A=@n27, B=@n28), B=@n28)
   B=SphereMask(A=TextureCoordinate(), B=Constant2Vector() {g=0.5, r=1.08}) {attenuationRadius=1.05, hardnessPercent=90}
  )
  n30 = ComponentMask(None=TextureCoordinate()) {r=true}
  n31 = Multiply(A=@n25.R, B=Multiply(A=@n30, B=@n30))
  n32 = Multiply(A=@n31, B=@n31)
  n33 = ShadingPathSwitch(Default=Subtract(A=@n29, B=Multiply(A=@n32) {constB=2}), Mobile=@n29) {desc=Migrated from FeatureLevelSwitch}
  n34 = SphereMask(A=WorldPosition().XYZ, B=CameraPositionWS(), Radius=Constant() {r=100}, Hardness=Constant() {r=0.2})
  n35 = Multiply(A=@n13.A, B=Multiply(A=Multiply(A=@n26, B=@n33), B=OneMinus(None=@n34)))
  ShadingPathSwitch(Default=DepthFade(Opacity=@n35), Mobile=@n35) {desc=Migrated from FeatureLevelSwitch}
 ]
}
