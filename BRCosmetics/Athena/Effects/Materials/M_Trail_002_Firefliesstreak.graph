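# Unlit additive particle sprite material for a cartoon firefly trail streak, with emissive from particle color through MF_FX_EmissiveAdjustments.
# A fast time sine pulses the horizontal UV scale around the center, and the clamped UVs sample T_CartoonFireflie.
# The texture's red channel times particle alpha drives opacity, with switchable depth fade and camera depth fade.
# Shading path switches bypass both fades on mobile.
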Material /BRCosmetics/Athena/Effects/Materials/M_Trail_002_Firefliesstreak
graph {
 properties {bUsedWithParticleSprites=true, blendMode=BLEND_Additive, pixelDepthOffsetMode=PDOM_Legacy, shadingModel=MSM_Unlit, usedShadingModels=Unlit}
 nodes [
  n1 = ParticleColor()
  MaterialFunctionCall(Emissive=Multiply(A=Constant() {r=1}, B=@n1.RGB)) {materialFunction=&MF_FX_EmissiveAdjustments}
  n2 = TextureCoordinate()
  n3 = TextureCoordinate()
  n4 = MaterialFunctionCall(Speed=Constant() {r=11}) {materialFunction=&DebugTimeSine}
  n5 = Add(A=@n2, B=Multiply(A=Subtract(A=@n3) {constB=0.5}, B=Multiply(A=@n4.Result) {constB=2}))
  n6 = Clamp(None=AppendVector(A=ComponentMask(None=@n5) {r=true}, B=ComponentMask(None=@n2) {g=true})) {maxDefault=0.9, minDefault=0.1}
  n7 = Constant() {r=0.5}
  n8 = Multiply(A=@n1.A, B=TextureSample(UVs=Panner(Coordinate=@n6, Time=@n7)) {texture=/Game/Effects/Fort_Effects/Textures/T_CartoonFireflie}.R)
  n9 = DepthFade(Opacity=@n8)
  n10 = ScalarParameter() {defaultValue=64, parameterName=Fade Length}
  n11 = ScalarParameter() {defaultValue=64, parameterName=Fade Offset}
  Multiply(
   A=StaticSwitchParameter(True=ShadingPathSwitch(Default=@n9, Mobile=@n8) {desc=Migrated from FeatureLevelSwitch}, False=@n8) {parameterName=DepthFade}
   B=ShadingPathSwitch(
    Default=MaterialFunctionCall(Fade Length=@n10, Fade Offset=@n11) {materialFunction=&CameraDepthFade}.Result
    Mobile=Constant() {r=1}
   ) {desc=Migrated from FeatureLevelSwitch}
  )
 ]
}
