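# Translucent unlit beam-trail material that draws a green flame streak, named for a swing effect.
# A noise-warped, time-frozen cartoon flame texture is eroded toward one end by two panning inky-smoke layers, whose object-relative UVs are normal-map distorted via dynamic parameters.
# The result blends two green color parameters into particle-tinted HDR emissive.
# Opacity multiplies it by particle alpha and a UV sphere mask, switching to zero under forward shading.
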Material /BRCosmetics/Athena/Effects/Materials/M_Trail_002_greenflamestreak_Swing
graph {
 properties {bUsedWithBeamTrails=true, blendMode=BLEND_Translucent, pixelDepthOffsetMode=PDOM_Legacy, shadingModel=MSM_Unlit, usedShadingModels=Unlit}
 nodes [
  n1 = Constant() {r=2}
  n2 = Constant() {r=1.6}
  n3 = TextureCoordinate() {uTiling=0.45, vTiling=0.3}
  n4 = ComponentMask(None=TextureSample(UVs=Panner(Coordinate=@n3) {speedX=-1.9, speedY=-0.2}) {samplerType=SAMPLERTYPE_LinearColor, texture=&T_NoisePack_LinearColor}.RGB) {g=true, r=true}
  n5 = ComponentMask(None=TextureCoordinate()) {r=true}
  n6 = Add(
   A=MaterialFunctionCall(Texture Scale=AppendVector(A=@n1, B=@n2)) {materialFunction=&ScaleUVsByCenter}.UVs
   B=Multiply(A=Multiply(B=@n4) {constA=0.65}, B=Power(Base=@n5) {constExponent=1.2})
  )
  n7 = Constant()
  n8 = ObjectPositionWS()
  n9 = Constant3Vector() {constant=<rgba 0 0 -2000 1>}
  n10 = Multiply(A=SphereMask(A=WorldPosition().XYZ, B=Add(A=@n8, B=@n9), Radius=Constant() {r=2550}, Hardness=Constant())) {constB=50}
  n11 = Multiply(A=ComponentMask(None=TextureCoordinate()) {g=true}) {constB=3}
  n12 = Constant() {r=0.32}
  n13 = Constant() {r=0.2}
  n14 = Constant() {r=1}
  n15 = DynamicParameter() {defaultValue=<rgba 0.975 4 0.5 1>, paramNames [UVDistortNormalFlatness, PannerSpeed, UVDistortScale, Param4]}
  n16 = TextureSample(UVs=Panner(Speed=AppendVector(A=@n14, B=@n15.PannerSpeed)) {speedX=0.1, speedY=0.5}) {
   samplerSource=SSM_Wrap_WorldGroupSettings
   samplerType=SAMPLERTYPE_Normal
   texture=/Game/Effects/Fort_Effects/Gameplay/SmokeDustExplosions/Textures/T_Explosion_NM_01
  }
  n17 = MaterialFunctionCall(Normal=@n16.RGB, Flatness=@n15.UVDistortNormalFlatness) {materialFunction=&FlattenNormal}
  n18 = Add(
   A=Multiply(A=AppendVector(A=@n10, B=@n11), B=AppendVector(A=@n12, B=@n13))
   B=Multiply(A=ComponentMask(None=@n17.Result) {g=true, r=true}, B=@n15.UVDistortScale) {constB=0.5}
  )
  n19 = Add(A=@n18) {constB=0.4}
  n20 = Add(
   A=TextureSample(UVs=Panner(Coordinate=@n19) {speedX=2.1}) {texture=&T_Inky_Smoke_Tile}.G
   B=TextureSample(UVs=Panner(Coordinate=@n18) {speedX=1.7}) {texture=&T_Inky_Smoke_Tile}.R
  )
  n21 = ComponentMask(None=TextureCoordinate()) {r=true}
  n22 = Subtract(
   A=TextureSample(UVs=Panner(Coordinate=@n6, Time=@n7) {speedY=1}) {samplerSource=SSM_Clamp_WorldGroupSettings, texture=/Game/Effects/Fort_Effects/Textures/Fire/T_CartoonFlame}.R
   B=Multiply(A=Multiply(A=@n20, B=@n21)) {constB=1.5}
  )
  n23 = LinearInterpolate(
   A=VectorParameter() {parameterName=Color 1, defaultValue=<rgba 0.6878 0.9323 0.2758 1>}.RGB
   B=VectorParameter() {parameterName=Color 2, defaultValue=<rgba 0.5743 0.9271 0.0738 1>}.RGB
   Alpha=@n22
  )
  n24 = Clamp(None=@n23)
  n25 = ParticleColor()
  MaterialFunctionCall(Emissive=Multiply(
    A=MaterialFunctionCall(LDR Input=@n24, HDR Tint=@n23) {materialFunction=&DeriveHDRfromLDR}.RemappedRange
    B=@n25.RGB
   )) {materialFunction=&MF_FX_EmissiveAdjustments}
  n26 = SphereMask(A=TextureCoordinate(), B=Constant2Vector() {g=0.5, r=2}) {attenuationRadius=2, hardnessPercent=80}
  ShadingPathSwitch(Default=Multiply(A=@n25.A, B=Multiply(A=@n22, B=@n26)), Forward=Constant())
 ]
}
