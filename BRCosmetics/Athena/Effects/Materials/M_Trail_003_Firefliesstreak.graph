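# Two-sided Default Lit material for mesh particles, apparently a streaking firefly cosmetic trail, though no material output connections appear.
# Three panning noise layers use square-rooted U to stretch the streaks and raw U to fade them, with Dynamic Parameter Param1 offsetting scroll speed.
# Two layers add panned distortion-normal and smooth textures to the noise color.
# Vertex color R, G and B weight the layers before summing; a separate branch tints those channels red.
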Material /BRCosmetics/Athena/Effects/Materials/M_Trail_003_Firefliesstreak
graph {
 properties {bUsedWithMeshParticles=true, pixelDepthOffsetMode=PDOM_Legacy, twoSided=true, usedShadingModels=Default Lit}
 nodes [
  n1 = MaterialFunctionCall(Float2=TextureCoordinate() {uTiling=0.28, vTiling=0.1}) {materialFunction=&BreakOutFloat2Components}
  n2 = AppendVector(A=SquareRoot(None=@n1.R), B=@n1.G)
  n3 = AppendVector(A=Add(A=Constant() {r=-4}, B=DynamicParameter().Param1), B=Constant() {r=-0.5})
  n4 = Multiply(
   A=Constant() {r=10}
   B=TextureSample(UVs=Panner(Coordinate=@n2, Speed=@n3)) {samplerType=SAMPLERTYPE_LinearColor, texture=&T_NoisePack_LinearColor}.RGB
  )
  n5 = MaterialFunctionCall(Float2=TextureCoordinate() {uTiling=2, vTiling=0.2}) {materialFunction=&BreakOutFloat2Components}
  n6 = SquareRoot(None=@n5.R)
  n7 = Add(A=Constant() {r=-2.5}, B=DynamicParameter().Param1)
  n8 = Constant() {r=-0.5}
  n9 = TextureSample(UVs=Panner(Coordinate=AppendVector(A=@n6, B=@n5.G), Speed=AppendVector(A=@n7, B=@n8))) {samplerType=SAMPLERTYPE_LinearColor, texture=&T_NoisePack_LinearColor}
  n10 = Panner(Coordinate=TextureCoordinate() {uTiling=0.4, vTiling=0}) {speedX=-2.9}
  n11 = TextureSample(UVs=@n10) {samplerType=SAMPLERTYPE_Normal, texture=&T_Weap_Distort_N_02}
  n12 = Multiply(A=AppendVector(
    A=ComponentMask(None=@n11.RGB) {g=true, r=true}
    B=TextureSample(UVs=@n10) {texture=&T_Smooth_tile}.R
   )) {constB=2}
  n13 = Multiply(A=Constant() {r=4}, B=Multiply(A=Add(A=@n9.RGB, B=@n12) {constB=0}, B=@n5.R))
  n14 = VertexColor()
  n15 = MaterialFunctionCall(Float2=TextureCoordinate() {uTiling=2, vTiling=0.2}) {materialFunction=&BreakOutFloat2Components}
  n16 = SquareRoot(None=@n15.R)
  n17 = Add(A=Constant() {r=-3.8}, B=DynamicParameter().Param1)
  n18 = Constant() {r=-0.5}
  n19 = TextureSample(UVs=Panner(Coordinate=AppendVector(A=@n16, B=@n15.G), Speed=AppendVector(A=@n17, B=@n18))) {samplerType=SAMPLERTYPE_LinearColor, texture=&T_NoisePack_LinearColor}
  n20 = Panner(Coordinate=TextureCoordinate() {uTiling=0.4, vTiling=0}) {speedX=-2.5}
  n21 = TextureSample(UVs=@n20) {samplerType=SAMPLERTYPE_Normal, texture=&T_Weap_Distort_N_02}
  n22 = Multiply(A=AppendVector(
    A=ComponentMask(None=@n21.RGB) {g=true, r=true}
    B=TextureSample(UVs=@n20) {texture=&T_Smooth_tile}.R
   )) {constB=1.5}
  n23 = Multiply(A=Constant() {r=8}, B=Multiply(A=Add(A=@n19.RGB, B=@n22) {constB=0}, B=@n15.R))
  n24 = VertexColor()
  Add(A=Multiply(A=Multiply(A=@n4, B=@n1.R), B=VertexColor().R), B=Add(A=Multiply(A=@n13, B=@n14.G), B=Multiply(A=@n23, B=@n24.B)) {constB=0})
  n25 = VertexColor()
  n26 = Constant3Vector() {constant=<rgba 1 0 0 0>}
  Add(A=Multiply(A=@n25.R, B=@n26), B=Add(A=Multiply(A=@n25.G, B=@n26), B=Multiply(A=@n25.B, B=@n26)))
  Multiply(A=@n1.R)
 ]
}
