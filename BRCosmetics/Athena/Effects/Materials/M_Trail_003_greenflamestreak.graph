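# Unlit Modulate-blend material for a flame streak trail on beam, Niagara ribbon and sprite effects, green by default.
# A panning, noise-distorted smoke texture, partly mapped by distance from the object, is shaped by sphere masks and eroded by rotating noise.
# That mask blends two colors, multiplied by a pulsing HDR gradient, particle color and time-of-day boost.
# Alpha fades toward white, Modulate's neutral value; switches offer gradient and rainbow variants.
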Material /BRCosmetics/Athena/Effects/Materials/M_Trail_003_greenflamestreak
graph {
 properties {
  bUsedWithBeamTrails=true
  bUsedWithNiagaraRibbons=true
  bUsedWithNiagaraSprites=true
  blendMode=BLEND_Modulate
  pixelDepthOffsetMode=PDOM_Legacy
  shadingModel=MSM_Unlit
  translucencyPass=MTP_BeforeDOF
  usedShadingModels=Unlit
 }
 nodes [
  n1 = Constant() {r=1}
  n2 = TextureCoordinate()
  n3 = Power(Base=OneMinus(None=ComponentMask(None=@n2) {r=true}), Exp=ScalarParameter() {defaultValue=1, parameterName=Gradient Contrast})
  n4 = TextureCoordinate()
  n5 = Power(Base=OneMinus(None=ComponentMask(None=@n4) {r=true})) {constExponent=5.7}
  n6 = Multiply(A=Constant() {r=4.6}, B=Time())
  n7 = Multiply(A=@n5, B=LinearInterpolate(Alpha=Sine(None=@n6)) {constA=0.5})
  n8 = SphereMask(A=WorldPosition().XYZ, B=ObjectPositionWS(), Radius=Constant() {r=2200}, Hardness=Constant())
  n9 = Multiply(A=@n8, B=@n8)
  n10 = ComponentMask(None=TextureCoordinate()) {g=true}
  n11 = TextureSample(UVs=Panner(Coordinate=TextureCoordinate() {uTiling=0.3, vTiling=0.45}) {speedX=-1}) {samplerType=SAMPLERTYPE_LinearColor, texture=&T_NoisePack_LinearColor}
  n12 = Add(
   A=AppendVector(A=Multiply(A=@n9) {constB=5}, B=Multiply(A=@n10))
   B=Multiply(A=ComponentMask(None=@n11.RGB) {g=true, r=true}, B=ScalarParameter() {defaultValue=0.45, parameterName=Smoke Trail Distortion Intensity}) {constB=0.45}
  )
  n13 = AppendVector(A=Constant(), B=ScalarParameter() {parameterName=Texture Offset})
  n14 = TextureSampleParameter2D(UVs=Panner(Coordinate=Add(A=@n12, B=@n13)) {speedX=1.9}) {parameterName=SmokeTrailTexture, samplerType=SAMPLERTYPE_Grayscale, texture=&T_Smoke_Trail_Soft}
  n15 = TextureCoordinate()
  n16 = SphereMask(
   A=Power(Base=MaterialFunctionCall(Float2=@n15) {materialFunction=&BreakOutFloat2Components}.R) {constExponent=0.15}
   B=Constant() {r=0.55}
  ) {attenuationRadius=0.5, hardnessPercent=0}
  n17 = Multiply(A=@n16, B=@n16)
  n18 = SphereMask(A=TextureCoordinate(), B=Constant2Vector() {g=0.5, r=1.05}) {attenuationRadius=1.05, hardnessPercent=85}
  n19 = SphereMask(A=WorldPosition().XYZ, B=ObjectPositionWS(), Radius=Constant() {r=2200}, Hardness=Constant())
  n20 = Multiply(A=@n19, B=@n19)
  n21 = ComponentMask(None=TextureCoordinate()) {g=true}
  n22 = Multiply(A=AppendVector(A=Multiply(A=@n20) {constB=18}, B=Multiply(B=@n21) {constA=1.8}), B=AppendVector(A=Constant() {r=0.58}, B=Constant() {r=0.4}))
  n23 = Panner(Coordinate=Rotator(Coordinate=Add(A=@n22) {constB=0.3}, Time=Constant() {r=1.6}) {speed=1}) {speedX=0.25, speedY=2.4}
  n24 = Panner(Coordinate=Rotator(Coordinate=@n22, Time=Constant() {r=1.6}) {speed=1}) {speedX=-0.18, speedY=1.5}
  n25 = Multiply(
   A=Multiply(
    A=TextureSample(UVs=@n23) {samplerType=SAMPLERTYPE_LinearColor, texture=&T_NoisePack_LinearColor}.R
    B=TextureSample(UVs=@n24) {samplerType=SAMPLERTYPE_LinearColor, texture=&T_NoisePack_LinearColor}.G
   )
   B=ComponentMask(None=TextureCoordinate()) {r=true}
  )
  n26 = Subtract(A=Multiply(A=Multiply(A=@n14.R, B=@n14.R), B=Multiply(A=@n17, B=@n18)), B=Multiply(A=@n25, B=@n25))
  n27 = Multiply(
   A=LinearInterpolate(
    A=Constant3Vector() {constant=<rgba 0.0342 0.099 0.0005 0>}
    B=Constant() {r=20}
    Alpha=StaticSwitchParameter(True=@n7, False=@n5) {defaultValue=true, parameterName=For Flames}
   )
   B=LinearInterpolate(
    A=VectorParameter() {parameterName=Color A, defaultValue=<rgba 0.1686 0.3177 0.091 1>}.RGB
    B=VectorParameter() {parameterName=Color B, defaultValue=<rgba 0.97 1 0.9115 1>}.RGB
    Alpha=@n26
   )
  )
  n28 = ComponentMask(None=TextureCoordinate()) {r=true}
  n29 = Multiply(
   A=TextureSample(UVs=Panner(Coordinate=@n28) {speedX=-1.6667}) {texture=/Game/Global/Textures/Default/Masks_and_Gradients/T_Grad_Rainbow}.RGB
   B=ParticleColor().RGB
  )
  n30 = MaterialFunctionCall(LDR Input=Clamp(None=@n29), HDR Tint=@n29) {materialFunction=&DeriveHDRfromLDR}
  n31 = ParticleColor()
  n32 = Multiply(
   A=MaterialFunctionCall(Color_Morning=@n1, Color_Day=@n1, Color_Evening=@n1, Color_Night=Constant() {r=10}) {materialFunction=&Color_Time_of_Day}.Result
   B=Multiply(
    A=StaticSwitchParameter(True=@n3, False=@n27) {parameterName="Use Gradient "}
    B=StaticSwitchParameter(True=@n30.RemappedRange, False=@n31.RGB) {parameterName=Use Rainbow Color}
   )
  )
  n33 = Multiply(A=Clamp(None=Multiply(A=@n26) {constB=40}), B=ScalarParameter() {defaultValue=1, parameterName=Alpha})
  StaticSwitchParameter(True=@n32, False=LinearInterpolate(B=@n32, Alpha=Multiply(A=@n31.A, B=@n33)) {constA=1}) {parameterName=Use Particle Color Only}
 ]
}
