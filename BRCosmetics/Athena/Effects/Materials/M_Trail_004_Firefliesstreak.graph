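# Two-sided Default Lit material for mesh particles, apparently for a firefly streak cosmetic trail.
# Vertex color channels mask tints, with red driving an orange glow that is boosted and passed through MF_FX_EmissiveAdjustments, and green adding a dark blue-grey.
# A separate branch pans a dust distortion normal map and mote noise, sums and amplifies them, then weights the result by vertex red and green.
# Material output connections are not shown.
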Material /BRCosmetics/Athena/Effects/Materials/M_Trail_004_Firefliesstreak
graph {
 properties {bUsedWithMeshParticles=true, pixelDepthOffsetMode=PDOM_Legacy, twoSided=true, usedShadingModels=Default Lit}
 nodes [
  n1 = VertexColor()
  n2 = Multiply(A=@n1.R, B=Constant3Vector() {constant=<rgba 1 0.4516 0 0>})
  Add(A=Multiply(A=@n1.G, B=Constant3Vector() {constant=<rgba 0.0346 0.0451 0.0573 0>}), B=@n2)
  n3 = Panner(Coordinate=TextureCoordinate() {uTiling=0, vTiling=0.1}) {speedX=0.2, speedY=1}
  n4 = TextureSample(UVs=@n3) {samplerType=SAMPLERTYPE_Normal, texture=/Game/Effects/Fort_Effects/Textures/T_DustDistortion}
  n5 = Multiply(A=Add(A=ComponentMask(None=@n4.RGB) {g=true, r=true}, B=TextureSample(UVs=@n3) {texture=/Game/Global/Textures/Default/Noise/TankMoteNoise}.B)) {constB=4}
  n6 = VertexColor()
  Add(A=Multiply(A=@n5, B=@n6.R), B=Multiply(A=@n5, B=@n6.G))
  Constant()
  MaterialFunctionCall(Emissive=Multiply(A=@n2) {constB=15}) {materialFunction=&MF_FX_EmissiveAdjustments}
  Multiply(A=@n1.B, B=Constant3Vector() {constant=<rgba 1 0 0 0>})
 ]
}
