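# Translucent, unlit, two-sided mesh-particle material for scrolling, smoky wind-speed streaks.
# Two panning soft smoke-trail samples and a randomly offset inky smoke tile form opacity and particle-tinted emissive.
# UV gradients shape the streak, while view-angle and depth fades soften it, with depth fade also modulating emissive boost.
# A dynamic parameter scales a local-space vertex offset, masked by vertex red.
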Material /BRCosmetics/Athena/Effects/Materials/M_WindSpeedy_Mesh
graph {
 properties {
  bUsedWithMeshParticles=true
  bUsedWithNiagaraMeshParticles=true
  blendMode=BLEND_Translucent
  pixelDepthOffsetMode=PDOM_Legacy
  shadingModel=MSM_Unlit
  twoSided=true
  usedShadingModels=Unlit
 }
 nodes [
  n1 = ScalarParameter() {defaultValue=0.8, parameterName=FadeSensitivity}
  n2 = ParticleColor()
  n3 = TextureCoordinate() {coordinateIndex=1, vTiling=0.2}
  n4 = TextureSample!(
   UVs=Panner(Coordinate=MaterialFunctionCall(XY=@n3) {materialFunction=&Swizzle}.YX) {bFractionalPart=true, speedX=1.8}
   Tex
   Level
   Apply View MipBias
  ) {
   constMipValue=2
   mipValueMode=TMVM_MipLevel
   samplerType=SAMPLERTYPE_Grayscale
   texture=&T_Smoke_Trail_Soft
  }
  n5 = MaterialFunctionCall(XY=TextureCoordinate() {coordinateIndex=1, vTiling=0.4}) {materialFunction=&Swizzle}
  n6 = Add(A=ComponentMask(None=@n5.YX) {r=true}, B=ScalarParameter() {defaultValue=0.133, parameterName=OffsetX})
  n7 = ComponentMask(None=@n5.YX) {g=true}
  n8 = TextureSample(UVs=Panner(Coordinate=AppendVector(A=@n6, B=@n7)) {bFractionalPart=true, speedX=1.2}) {
   constMipValue=2
   mipValueMode=TMVM_MipLevel
   samplerType=SAMPLERTYPE_Grayscale
   texture=&T_Smoke_Trail_Soft
  }
  n9 = Add(A=@n4.A, B=@n8.A)
  n10 = TextureCoordinate() {coordinateIndex=1, uTiling=8, vTiling=0.5}
  n11 = Add(
   A=ComponentMask(None=@n10) {r=true}
   B=DynamicParameter() {defaultValue=<rgba 1 1 60 1>, paramNames [RandomOffsetX, WidthBase, DepthFade, Param4]}.RandomOffsetX
  )
  n12 = ComponentMask(None=@n10) {g=true}
  n13 = TextureSample(UVs=Panner(Coordinate=AppendVector(A=@n11, B=@n12)) {bFractionalPart=true, speedY=1.6}) {constMipValue=2, mipValueMode=TMVM_MipLevel, texture=&T_Inky_Smoke_Tile}
  n14 = ComponentMask(None=TextureCoordinate()) {r=true}
  n15 = Multiply(A=OneMinus(None=@n14), B=@n14)
  n16 = TextureCoordinate()
  n17 = Multiply(A=Multiply(A=Add(A=@n9, B=@n13.A), B=Add(A=@n15, B=@n15)), B=Power(Base=ComponentMask(None=@n16) {g=true}) {constExponent=3})
  n18 = DynamicParameter() {defaultValue=<rgba 1 1 60 1>, paramNames [RandomOffsetX, WidthBase, DepthFade, Param4]}
  n19 = LinearInterpolate(A=@n18.DepthFade, B=Multiply(A=@n18.DepthFade) {constB=0.25}, Alpha=@n17)
  DepthFade(
   Opacity=Multiply(
    A=MaterialFunctionCall("FadeSensitivity (optional)"=@n1) {materialFunction=&FadeBasedOnViewAngle}.Result
    B=Multiply(A=@n2.A, B=@n17)
   )
   FadeDistance=@n19
  )
  n20 = Add(A=@n4.G, B=@n8.G)
  n21 = Multiply(A=@n2.RGB, B=LinearInterpolate(Alpha=Add(A=@n20, B=@n13.G)) {constA=0.1})
  MaterialFunctionCall(Emissive=LinearInterpolate(A=@n21, B=Multiply(A=@n21) {constB=6}, Alpha=DepthFade(Opacity=@n13.A, FadeDistance=@n19))) {materialFunction=&MF_FX_EmissiveAdjustments}
  n22 = Constant3Vector()
  n23 = VectorParameter() {parameterName=Offset, defaultValue=<rgba 0 270 0 0>}
  Transform(None=Multiply(A=LinearInterpolate(A=@n22, B=@n23.RGB, Alpha=@n18.Param4) {constB=150}, B=VertexColor().R)) {transformSourceType=TRANSFORMSOURCE_Local}
 ]
}
