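# Translucent, two-sided unlit mesh particle material for smoke in the EvilCowbow main body skin transition effect.
# Three panning smoke textures, shaped by Fresnel, drive emissive over UV-masked color gradients and feed erosion-based opacity with depth and vertex-color fades.
# Particle color channels control U offset, brightness, color blend and alpha.
# Dynamic parameters drive erosion and actor-relative vector noise for displacement, replaced by a constant on mobile.
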Material /BRCosmetics/Athena/Effects/Materials/Skins/M_EvilCowbow_Mainbody_Transition_Smoke
graph {
 properties {
  bUsedWithMeshParticles=true
  bUsedWithNiagaraMeshParticles=true
  blendMode=BLEND_Translucent
  pixelDepthOffsetMode=PDOM_Legacy
  shadingModel=MSM_Unlit
  twoSided=true
  usedShadingModels=Unlit
 }
 nodes [
  n1 = VectorParameter() {parameterName=BaseColor_2, defaultValue=<rgba 0.0235 0.05 0.003 1>}
  n2 = VectorParameter() {parameterName=BaseColor_3, defaultValue=<rgba 0.4912 1 0.0819 1>}
  n3 = TextureCoordinate() {coordinateIndex=1, vTiling=1.25}
  n4 = LinearInterpolate(
   A=VectorParameter() {parameterName=BaseColor_1, defaultValue=<rgba 0.2304 0.375 0.0259 1>}.RGB
   B=@n1.RGB
   Alpha=Clamp(None=ComponentMask(None=@n3) {g=true})
  )
  n5 = ComponentMask(None=TextureCoordinate() {coordinateIndex=2, vTiling=3}) {g=true}
  n6 = Clamp(None=Multiply(A=@n5, B=@n5))
  n7 = SphereMask(
   A=TextureCoordinate() {coordinateIndex=1, uTiling=1.2}
   B=Constant2Vector() {g=0.5, r=0.6}
   Radius=ScalarParameter() {defaultValue=0.9, parameterName=TopMask_Radius}
   Hardness=ScalarParameter() {defaultValue=0.25, parameterName=TopMask_Hardness}
  )
  n8 = LinearInterpolate(
   A=VectorParameter() {parameterName=BaseColor_5, defaultValue=<rgba 0.8944 0.0111 1 1>}.RGB
   B=LinearInterpolate(A=@n1.RGB, B=LinearInterpolate(A=@n2.RGB, B=@n4, Alpha=@n6), Alpha=Multiply(A=@n7, B=@n7))
   Alpha=ParticleColor().B
  )
  n9 = Fresnel(ExponentIn=ScalarParameter() {defaultValue=0.25, parameterName=SmokeFrensnel}) {baseReflectFraction=0.7}
  n10 = ScalarParameter() {defaultValue=1, parameterName=XTileMult_2}
  n11 = TextureCoordinate() {uTiling=3, vTiling=2}
  n12 = ComponentMask(None=@n11) {r=true}
  n13 = AppendVector(A=Add(A=ParticleColor().R, B=Multiply(A=@n10, B=@n12)), B=ComponentMask(None=@n11) {g=true})
  n14 = TextureCoordinate() {uTiling=2}
  n15 = ComponentMask(None=@n14) {r=true}
  n16 = ScalarParameter() {defaultValue=4, parameterName=XTileMult}
  n17 = AppendVector(A=Add(A=ParticleColor().R, B=Multiply(A=@n15, B=@n16)), B=ComponentMask(None=@n14) {g=true})
  n18 = ScalarParameter() {defaultValue=0.5, parameterName=PanSpeed}
  n19 = AppendVector(A=Constant(), B=@n18)
  n20 = Add(
   A=TextureSample(UVs=Panner(Coordinate=@n13) {speedY=0.35}) {samplerType=SAMPLERTYPE_Grayscale, texture=&T_Smoke_Trail_Soft}.G
   B=TextureSample(UVs=Panner(Coordinate=@n17, Speed=@n19) {speedY=0.5}) {texture=&T_Inky_Smoke_Tile}.G
  )
  n21 = TextureCoordinate() {uTiling=2, vTiling=2}
  n22 = Add(A=ParticleColor().R, B=Multiply(A=ScalarParameter() {defaultValue=3, parameterName=XTileMult_1}, B=ComponentMask(None=@n21) {r=true}))
  n23 = ComponentMask(None=@n21) {g=true}
  n24 = Constant()
  n25 = TextureSample(UVs=Panner(Coordinate=AppendVector(A=@n22, B=@n23), Speed=AppendVector(A=@n24, B=@n18)) {speedY=0.5}) {texture=&T_Inky_Smoke_Tile}
  n26 = Multiply(A=OneMinus(None=Power(Base=@n9) {constExponent=1}), B=Power(Base=Add(A=@n20, B=@n25.G)) {constExponent=1.25})
  MaterialFunctionCall(Emissive=Multiply(A=ParticleColor().G, B=Multiply(A=@n8, B=@n26))) {materialFunction=&MF_FX_EmissiveAdjustments}
  n27 = MaterialFunctionCall!(
   Dynamic Param=@n26
   Alpha=DynamicParameter() {defaultValue=<rgba 1 50 15 1>, paramNames [Curl_Speed, Curl_Scale, Curl_Intensity, Erode]}.Erode
  ) {materialFunction=/Game/Effects/Fort_Effects/Materials/_Functions/Erode_Alpha}
  n28 = ScalarParameter() {defaultValue=2, parameterName=FadeDistance}
  Multiply(A=Multiply(A=DepthFade(Opacity=@n27.Result, FadeDistance=@n28), B=ParticleColor().A), B=VertexColor().R)
  n29 = WorldPosition()
  n30 = ActorPositionWS()
  n31 = Divide(
   A=Subtract(
    A=Subtract(A=@n29.XYZ, B=@n30)
    B=DynamicParameter() {defaultValue=<rgba 1 50 15 1>, paramNames [Curl_Speed, Curl_Scale, Curl_Intensity, Erode]}.Curl_Speed
   )
   B=DynamicParameter() {defaultValue=<rgba 1 50 15 1>, paramNames [Curl_Speed, Curl_Scale, Curl_Intensity, Erode]}.Curl_Scale
  )
  ShadingPathSwitch(
   Default=Multiply(
    A=DynamicParameter() {defaultValue=<rgba 1 50 15 1>, paramNames [Curl_Speed, Curl_Scale, Curl_Intensity, Erode]}.Curl_Intensity
    B=VectorNoise(World Position=@n31) {noiseFunction=VNF_VectorALU}
   )
   Mobile=Constant() {r=1}
  ) {desc=Migrated from FeatureLevelSwitch}
 ]
}
