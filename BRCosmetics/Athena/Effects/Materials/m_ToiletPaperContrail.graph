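# Unlit, masked, two-sided material for toilet paper contrail cosmetic ribbons and beam trails.
# Emissive scrolls the paper texture under fake lighting from sun and camera vectors, tinted by time of day and particle red.
# World position offset adds panned or world-aligned noise plus sine/cosine twist, pinned at the U=0 end.
# Dynamic parameters vary pan phase and a rotated mask cut, with optional head erosion and quality-gated dithered pixel depth offset.
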Material /BRCosmetics/Athena/Effects/Materials/m_ToiletPaperContrail
graph {
 properties {
  bUsedWithBeamTrails=true
  bUsedWithNiagaraRibbons=true
  blendMode=BLEND_Masked
  pixelDepthOffsetMode=PDOM_Legacy
  shadingModel=MSM_Unlit
  translucencyLightingMode=TLM_VolumetricPerVertexDirectional
  twoSided=true
  usedShadingModels=Unlit
 }
 nodes [
  n1 = TextureCoordinate() {vTiling=0.2}
  n2 = Constant() {r=1.5}
  n3 = TextureSample(UVs=Panner(Coordinate=Panner(Coordinate=@n1, Time=@n2) {speedY=0.5}) {speedX=-3}) {texture=/BRCosmetics/Athena/Effects/Textures/T_toilet_paper_roll_contrail_D}
  n4 = CollectionParameter() {
   collection=&FortniteMaterialParameters
   parameterName=SunlightDirectionalVector
  }
  n5 = ComponentMask(None=Transform(None=@n4) {transformSourceType=TRANSFORMSOURCE_World, transformType=TRANSFORM_Tangent}) {b=true}
  n6 = MaterialFunctionCall(Float2=TextureCoordinate()) {materialFunction=&BreakOutFloat2Components}
  n7 = Multiply(
   A=LinearInterpolate(Alpha=Multiply(A=@n5) {constB=-1}) {constA=0.8}
   B=LinearInterpolate(A=LinearInterpolate(Alpha=@n6.G) {constA=1, constB=0.83}, B=@n6.R)
  )
  n8 = ComponentMask(None=CameraVectorWS()) {b=true}
  n9 = OneMinus(None=@n8)
  n10 = Multiply(A=Add(A=Divide(A=@n9) {constB=1.5}) {constB=-0.4}) {constB=5}
  n11 = LinearInterpolate(A=Multiply(A=@n10) {constB=0.6}, B=@n10, Alpha=TwoSidedSign()) {constA=0.8}
  n12 = Add(A=@n8) {constB=0.2}
  n13 = MaterialFunctionCall!(Blend=Multiply(A=@n7, B=LinearInterpolate(Alpha=@n12) {constA=1, constB=2}), Base=@n7) {materialFunction=/Engine/Functions/Engine_MaterialFunctions03/Blends/Blend_Lighten}
  n14 = MaterialFunctionCall(
   Blend=Multiply(A=@n5) {constB=1.3}
   Base=MaterialFunctionCall(Blend=Multiply(A=@n7, B=@n11), Base=@n13.Result) {materialFunction=&Blend_Lighten}.Result
  ) {materialFunction=&Blend_Lighten}
  n15 = Constant3Vector() {constant=<rgba 0.3707 0.5453 0.75 0>}
  n16 = Constant() {r=1}
  n17 = Constant3Vector() {constant=<rgba 0.2706 0.3992 0.6406 0>}
  n18 = Constant3Vector() {constant=<rgba 1.5 0.589 0.5292 0>}
  n19 = Constant() {r=1}
  n20 = Constant3Vector() {constant=<rgba 0.4917 0.6258 1 0>}
  n21 = Multiply(
   A=Multiply(A=Multiply(A=@n3.RGB, B=@n14.Result), B=ParticleColor().R)
   B=LinearInterpolate(
    A=MaterialFunctionCall(Color_Morning=@n15, Color_Day=@n16, Color_Evening=@n15, Color_Night=@n17) {materialFunction=&Color_Time_of_Day}.Result
    B=MaterialFunctionCall(Color_Morning=@n18, Color_Day=@n19, Color_Evening=@n18, Color_Night=@n20) {materialFunction=&Color_Time_of_Day}.Result
    Alpha=MaterialFunctionCall(Blend=@n5, Base=@n13.Result) {materialFunction=&Blend_Lighten}.Result
   )
  )
  MaterialFunctionCall(Emissive=LinearInterpolate(A=Multiply(A=@n21) {constB=0.73}, B=@n21, Alpha=TwoSidedSign()) {constA=0.8}) {materialFunction=&MF_FX_EmissiveAdjustments}
  n22 = MaterialFunctionCall(Float2=TextureCoordinate()) {materialFunction=&BreakOutFloat2Components}
  n23 = ScalarParameter() {defaultValue=60, group=WPO, parameterName=WPO_FalloffAlongEdge}
  n24 = WorldPosition()
  n25 = Multiply(A=Time()) {constB=500}
  n26 = Multiply(
   A=MaterialFunctionCall(
    TextureObject=TextureObject() {texture=/Game/Global/Textures/Default/Noise/3xPerlin_M}
    TextureSize=VectorParameter() {group=WPO, parameterName=WPO_Tile, defaultValue=<rgba 700 700 700 1>}.RGB
    WorldPosition=Add(A=@n24.XYZ, B=@n25)
   ) {materialFunction=&WorldAlignedTexture}.XYZ Texture
   B=ScalarParameter() {defaultValue=70, group=WPO, parameterName=WPO_Intensity}
  ) {constB=70}
  n27 = TextureCoordinate()
  n28 = MaterialFunctionCall(Float2=@n27) {materialFunction=&BreakOutFloat2Components}
  n29 = SquareRoot(None=@n28.R)
  n30 = ScalarParameter() {defaultValue=1, parameterName=U_Tile}
  n31 = Multiply(
   A=MaterialFunctionCall(Float2=@n27) {materialFunction=&BreakOutFloat2Components}.R
   B=@n30
  )
  n32 = Panner(
   Coordinate=StaticSwitchParameter(True=AppendVector(A=@n29, B=@n28.G), False=AppendVector(A=@n31, B=@n30)) {defaultValue=true, group=WPO, parameterName=UseStretchedTile}
   Time=Add(A=Time(), B=DynamicParameter() {paramNames [RandomInitailU, MaskCutUVRotator, RandomTwistOffset, OffsetTwistSpeed]}.RandomInitailU)
   Speed=AppendVector(A=ScalarParameter() {defaultValue=-3, parameterName=WPO_PanSpeed_U}, B=ScalarParameter() {defaultValue=0.2, parameterName=WPO_PanSpeed_V})
  ) {speedX=-3, speedY=0.2}
  n33 = AppendVector(
   A=ComponentMask(None=TextureSample(UVs=@n32) {samplerType=SAMPLERTYPE_Normal, texture=/Game/Effects/Fort_Effects/Gameplay/Water/Textures/T_WaterRipples_01}.RGB) {g=true, r=true}
   B=LinearInterpolate(Alpha=TextureSample(UVs=@n32) {samplerType=SAMPLERTYPE_Grayscale, texture=&LowResBlurredNoise}.R) {constA=-0.52}
  )
  n34 = ShadingPathSwitch(
   Default=LinearInterpolate(A=@n33, B=Multiply(A=@n33) {constB=50}, Alpha=ParticleColor() {desc=WPO Intensity}.B) {constB=50}
   Mobile=Constant()
  ) {desc=Migrated from FeatureLevelSwitch}
  n35 = MaterialFunctionCall(Float2=TextureCoordinate() {uTiling=0.9}) {materialFunction=&BreakOutFloat2Components}
  n36 = Subtract(A=ComponentMask(None=TextureCoordinate() {vTiling=2}) {g=true})
  n37 = MaterialFunctionCall!(MaxiumURange, RangeRampHardness, MirrorTexCoordU) {materialFunction=/BRCosmetics/Weapons/FORT_Melee/Pickaxe_Venus/MF_LocalSpaceTrailUV}
  n38 = ComponentMask(None=Multiply(A=ObjectPositionWS()) {constB=-1}) {b=true}
  n39 = Multiply(A=Add(A=ComponentMask(None=@n37."LocalSpaceCoordinate(UV)") {r=true}, B=Divide(A=@n38) {constB=500})) {constB=3}
  n40 = Multiply(A=Multiply(A=Multiply(A=@n35.R, B=@n35.R), B=Multiply(A=@n36) {constB=1.5}), B=AppendVector(A=Cosine(None=@n39), B=Sine(None=@n39))) {constB=8}
  n41 = MaterialFunctionCall(Float2=TextureCoordinate()) {materialFunction=&BreakOutFloat2Components}
  n42 = AppendVector(A=LinearInterpolate(B=Multiply(A=@n40) {constB=10}, Alpha=Multiply(A=@n41.R, B=@n41.R)), B=Constant())
  n43 = ScalarParameter() {defaultValue=1, parameterName=Twist}
  Multiply(
   A=Saturate(None=Multiply(A=@n22.R, B=@n23))
   B=Add(A=StaticSwitchParameter(True=@n26, False=@n34) {group=WPO, parameterName=UseWorldAlignedUV}, B=Multiply(A=@n42, B=@n43)) {constB=0}
  )
  n44 = MaterialFunctionCall(Float2=TextureCoordinate()) {materialFunction=&BreakOutFloat2Components}
  n45 = ScalarParameter() {defaultValue=5, parameterName=TrailHeadErode}
  n46 = Saturate(None=Add(A=TextureSample() {texture=&AtmosphericCloudNoise01}.R, B=Multiply(A=@n44.R, B=@n45)))
  n47 = Constant() {r=1}
  n48 = MaterialFunctionCall(Float2=Rotator(
    Coordinate=TextureCoordinate()
    Time=DynamicParameter() {paramNames [RandomInitailU, MaskCutUVRotator, RandomTwistOffset, OffsetTwistSpeed]}.MaskCutUVRotator
   ) {speed=0.8}) {materialFunction=&BreakOutFloat2Components}
  n49 = OneMinus(None=Ceil(None=Add(A=@n48.R) {constB=-0.9}))
  ShadingPathSwitch(
   Default=Multiply(A=StaticSwitchParameter(True=@n46, False=@n47) {parameterName=UseTrailHeadErode}, B=Clamp(None=@n49))
   Mobile=Constant() {r=1}
  ) {desc=Migrated from FeatureLevelSwitch}
  n50 = Constant()
  n51 = ShadingPathSwitch(
   Default=Multiply(A=MaterialFunctionCall() {materialFunction=&DitherTemporalAA}.Result) {constB=30}
   Mobile=@n50
  ) {desc=Migrated from FeatureLevelSwitch}
  StaticSwitchParameter(True=QualitySwitch(Default=@n51, Low=@n50, High=@n51, Medium=@n50), False=Constant()) {defaultValue=true, parameterName=UsePixelDepth}
 ]
}
